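# Control Rig for the female player body WriteWalrus, correcting deformation of skirt, necklace, strap, shoulder and boot bones.
# After the Fort deform rig runs, parent constraints blend skirt, necklace and support bones between body bones, and foot yaw drives boot corrections.
# Sequenced transform offsets are weighted by deform-rig attributes, spherical pose readers, pelvis height and crouch, jump, knee-slide and movement variables.
# Kalman filters and alpha interpolation smooth many weights.
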ControlRigBlueprint /BRCosmetics/Characters/Player/Female/Medium/Bodies/F_MED_WriteWalrus/Meshes/F_MED_WriteWalrus_CtrlRig
graphs [
 New Function_ContainedGraph {nodes [
   RigVMFunctionEntryNode()
   > &RigUnit_ParentConstraint(
    AdvancedSettings=(InterpolationType=Average,RotationOrderForFilter=XZY)
    Child=(Type=Bone,Name="dyn_skirt_fr_l_base")
    Parents=((Item=(Type=Bone,Name="thigh_l"),Weight=1),(Item=(Type=Bone,Name="pelvis"),Weight=0))
    Weight=1
   )
   > &RigUnit_ParentConstraint(
    AdvancedSettings=(InterpolationType=Average,RotationOrderForFilter=XZY)
    Child=(Type=Bone,Name="dyn_skirt_side_l_frnt_base")
    Parents=((Item=(Type=Bone,Name="thigh_l"),Weight=1),(Item=(Type=Bone,Name="pelvis"),Weight=0))
    Weight=1
   )
   > &RigUnit_ParentConstraint(
    AdvancedSettings=(InterpolationType=Average,RotationOrderForFilter=XZY)
    Child=(Type=Bone,Name="dyn_skirt_side_l_bk_base")
    Parents=((Item=(Type=Bone,Name="thigh_l"),Weight=1),(Item=(Type=Bone,Name="pelvis"),Weight=0))
    Weight=1
   )
   > &RigUnit_ParentConstraint(
    AdvancedSettings=(InterpolationType=Average,RotationOrderForFilter=XZY)
    Child=(Type=Bone,Name="dyn_skirt_fr_mid_base")
    Parents=((Item=(Type=Bone,Name="thigh_l"),Weight=0.6),(Item=(Type=Bone,Name="thigh_r"),Weight=0.4),(Item=(Type=Bone,Name="pelvis"),Weight=0))
    Weight=1
   )
   > &RigUnit_ParentConstraint(
    AdvancedSettings=(InterpolationType=Average,RotationOrderForFilter=XZY)
    Child=(Type=Bone,Name="support_shoulder_r_rotator")
    Parents=((Item=(Type=Bone,Name="clavicle_r"),Weight=0.1),(Item=(Type=Bone,Name="upperarm_twist_01_r"),Weight=0.9))
    Weight=1
   )
   > &RigUnit_ParentConstraint(
    AdvancedSettings=(InterpolationType=Average,RotationOrderForFilter=XZY)
    Child=(Type=Bone,Name="dyn_necklace_bottom_base")
    Parents=((Item=(Type=Bone,Name="spine_05"),Weight=0.5),(Item=(Type=Bone,Name="spine_04"),Weight=0.5))
    Weight=1
   )
   > &RigUnit_ParentConstraint(
    AdvancedSettings=(InterpolationType=Average,RotationOrderForFilter=XZY)
    Child=(Type=Bone,Name="dyn_skirt_arrow_01_base")
    Parents=((Item=(Type=Bone,Name="thigh_l"),Weight=0.95),(Item=(Type=Bone,Name="thigh_r"),Weight=0.05))
    Weight=1
   )
   > &RigUnit_ParentConstraint(
    AdvancedSettings=(InterpolationType=Average,RotationOrderForFilter=XZY)
    Child=(Type=Bone,Name="dyn_skirt_arrow_02_base")
    Parents=((Item=(Type=Bone,Name="thigh_l"),Weight=0.8),(Item=(Type=Bone,Name="thigh_r"),Weight=0.2))
    Weight=1
   )
   > &RigUnit_ParentConstraint(
    AdvancedSettings=(InterpolationType=Average,RotationOrderForFilter=XZY)
    Child=(Type=Bone,Name="dyn_skirt_arrow_03_base")
    Parents=((Item=(Type=Bone,Name="thigh_l"),Weight=0.7),(Item=(Type=Bone,Name="thigh_r"),Weight=0.3))
    Weight=1
   )
   > &RigUnit_ParentConstraint(
    AdvancedSettings=(InterpolationType=Average,RotationOrderForFilter=XZY)
    Child=(Type=Bone,Name="dyn_skirt_arrow_04_base")
    Parents=((Item=(Type=Bone,Name="thigh_l"),Weight=0.5),(Item=(Type=Bone,Name="thigh_r"),Weight=0.5))
    Weight=1
   )
   > &RigUnit_ParentConstraint(
    AdvancedSettings=(InterpolationType=Average,RotationOrderForFilter=XZY)
    Child=(Type=Bone,Name="dyn_skirt_arrow_05_base")
    Parents=((Item=(Type=Bone,Name="thigh_l"),Weight=0.4),(Item=(Type=Bone,Name="thigh_r"),Weight=0.6))
    Weight=1
   )
   > &RigUnit_ParentConstraint(
    AdvancedSettings=(InterpolationType=Average,RotationOrderForFilter=XZY)
    Child=(Type=Bone,Name="dyn_skirt_arrow_06_base")
    Parents=((Item=(Type=Bone,Name="thigh_l"),Weight=0.2),(Item=(Type=Bone,Name="thigh_r"),Weight=0.8))
    Weight=1
   )
   > &RigUnit_ParentConstraint(
    AdvancedSettings=(InterpolationType=Average,RotationOrderForFilter=XZY)
    Child=(Type=Bone,Name="dyn_skirt_arrow_07_base")
    Parents=((Item=(Type=Bone,Name="thigh_l"),Weight=0.05),(Item=(Type=Bone,Name="thigh_r"),Weight=0.95))
    Weight=1
   )
   > &RigUnit_ParentConstraint(
    AdvancedSettings=(InterpolationType=Average,RotationOrderForFilter=XZY)
    Child=(Type=Bone,Name="dyn_skirt_arrow_01_02")
    Parents=((Item=(Type=Bone,Name="thigh_l"),Weight=0.95),(Item=(Type=Bone,Name="thigh_r"),Weight=0.05))
    Weight=1
   )
   > &RigUnit_ParentConstraint(
    AdvancedSettings=(InterpolationType=Average,RotationOrderForFilter=XZY)
    Child=(Type=Bone,Name="dyn_skirt_arrow_02_02")
    Parents=((Item=(Type=Bone,Name="thigh_l"),Weight=0.8),(Item=(Type=Bone,Name="thigh_r"),Weight=0.2))
    Weight=1
   )
   > &RigUnit_ParentConstraint(
    AdvancedSettings=(InterpolationType=Average,RotationOrderForFilter=XZY)
    Child=(Type=Bone,Name="dyn_skirt_arrow_03_02")
    Parents=((Item=(Type=Bone,Name="thigh_l"),Weight=0.7),(Item=(Type=Bone,Name="thigh_r"),Weight=0.3))
    Weight=1
   )
   > &RigUnit_ParentConstraint(
    AdvancedSettings=(InterpolationType=Average,RotationOrderForFilter=XZY)
    Child=(Type=Bone,Name="dyn_skirt_arrow_04_02")
    Parents=((Item=(Type=Bone,Name="thigh_l"),Weight=0.5),(Item=(Type=Bone,Name="thigh_r"),Weight=0.5))
    Weight=1
   )
   > &RigUnit_ParentConstraint(
    AdvancedSettings=(InterpolationType=Average,RotationOrderForFilter=XZY)
    Child=(Type=Bone,Name="dyn_skirt_arrow_05_02")
    Parents=((Item=(Type=Bone,Name="thigh_l"),Weight=0.4),(Item=(Type=Bone,Name="thigh_r"),Weight=0.6))
    Weight=1
   )
   > &RigUnit_ParentConstraint(
    AdvancedSettings=(InterpolationType=Average,RotationOrderForFilter=XZY)
    Child=(Type=Bone,Name="dyn_skirt_arrow_06_02")
    Parents=((Item=(Type=Bone,Name="thigh_l"),Weight=0.2),(Item=(Type=Bone,Name="thigh_r"),Weight=0.8))
    Weight=1
   )
   > &RigUnit_ParentConstraint(
    AdvancedSettings=(InterpolationType=Average,RotationOrderForFilter=XZY)
    Child=(Type=Bone,Name="dyn_skirt_arrow_07_01")
    Parents=((Item=(Type=Bone,Name="thigh_l"),Weight=0.05),(Item=(Type=Bone,Name="thigh_r"),Weight=0.95))
    Weight=1
   )
   > &RigUnit_ParentConstraint(
    AdvancedSettings=(InterpolationType=Average,RotationOrderForFilter=XZY)
    Child=(Type=Bone,Name="skirt_side_l_ALT_support")
    Parents=((Item=(Type=Bone,Name="calf_l"),Weight=0.5),(Item=(Type=Bone,Name="thigh_twist_01_l"),Weight=0.5))
    Weight=1
   )
   > RigVMFunctionReturnNode()
  ]}
 New Function_ContainedGraph {nodes [
   RigVMFunctionEntryNode()
   > Get Delta Ref = RigVMFunctionReferenceNode(Item=(Type=Bone,Name="foot_r"))
   > &RigUnit_ModifyTransforms(
    ItemToModify=((Item=(Type=Bone,Name="support_foot_r"),Transform=(Rotation=(X=0,Y=0,Z=0,W=-1),Translation=(X=8.1,Y=-3.8,Z=0),Scale3D=(X=1,Y=1,Z=1))))
    WeightMinimum=0
    Weight=&RigVMFunction_MathDoubleRemap(SourceMaximum=50, SourceMinimum=25, TargetMaximum=1, TargetMinimum=0, Value=@"Get Delta Ref".Result.Rotation.Yaw).Result
   )
   > Get Delta Ref_1 = RigVMFunctionReferenceNode(Item=(Type=Bone,Name="foot_l"))
   > &RigUnit_ModifyTransforms(
    ItemToModify=((Item=(Type=Bone,Name="support_foot_l"),Transform=(Rotation=(X=0,Y=0,Z=0,W=-1),Translation=(X=-8.109,Y=-3.8112,Z=0),Scale3D=(X=1,Y=1,Z=1))))
    WeightMinimum=0
    Weight=&RigVMFunction_MathDoubleRemap(SourceMaximum=50, SourceMinimum=25, TargetMaximum=1, TargetMinimum=0, Value=@"Get Delta Ref_1".Result.Rotation.Yaw).Result
   )
   > RigVMFunctionReturnNode()
  ]}
 RigVMFunctionLibrary {nodes [
   Boots = RigVMCollapseNode(ExecuteContext=())
   Constraints = RigVMCollapseNode(ExecuteContext=())
  ]}
 RigVMFunction_Sequence_1_ContainedGraph {nodes [
   n1 = RigVMFunctionEntryNode()
   RigVMFunction_Sequence = &RigVMFunction_Sequence(ExecuteContext=@n1.ExecuteContext)
   n2 = &RigVMFunction_Sequence(ExecuteContext=@RigVMFunction_Sequence.B)
   n3 = &RigVMFunction_Sequence(ExecuteContext=@n2.B)
   RigVMFunctionReturnNode(A=@RigVMFunction_Sequence.A, B=@n2.A, C=@n3.A, D=@n3.B)
  ]}
 RigVMModel {nodes [
   RigUnit_BeginExecution = &RigUnit_BeginExecution()
   /Script/FortniteGame.RigUnit_FortDeformRig(
    Bool=(bUpperArm=true,bElbow=true,bWrist=true,bGroin=true,bGlute=true,bKnee=true,bPec=true,bBicep=true,bLat=true)
    Curve="(bLUpperArmUpAlpha=False,bRUpperArmUpAlpha=False,bLUpperArmFrAlpha=False,bRUpperArmFrAlpha=False,bLUpperArmBkAlpha=False,bRUpperArmBkAlpha=False,bLWristUpAlpha=False,bRWristUpAlpha=False,bLWristDnAlpha=False,bRWristDnAlpha=False,bLKneeAlpha=true,bRKneeAlph ...(truncated)"
    Debug=(bUpperArmDebug=False,bWristDebug=False,bLatDebug=False,bGroinGluteDebug=False)
    Value="(UpperArmUp=(Rotation=(X=0.0,Y=-0.0,Z=0.0,W=1.0),Translation=(X=0.0,Y=0.0,Z=8.0),Scale3D=(X=0.70,Y=1.0,Z=1.0)),UpperArmFwdFr=(Rotation=(X=0.0,Y=-0.0,Z=0.0,W=1.0),Translation=(X=0.0,Y=10.0,Z=0.0),Scale3D=(X=0.50,Y=1.0,Z=1.0)),UpperArmFwdBk=(Rotation=(X=0.0, ...(truncated)"
    ExecutePin=@RigUnit_BeginExecution.ExecutePin
   ) {is=struct RigVMUnitNode}
   > Constraints = RigVMFunctionReferenceNode()
   > Boots = RigVMFunctionReferenceNode()
   > RigVMFunction_Sequence = RigVMAggregateNode()
   n1 = &RigVMFunction_Sequence(ExecuteContext=@RigVMFunction_Sequence.D)
   n2 = RigVMDispatchNode(BoneName=Root, Default=0, Name=deformRig_r_thigh_up_alpha) {template="DISPATCH_RigDispatch_GetAnimAttribute(in Name,in BoneName,in Default,out Value,out Found)"}
   n3 = &RigVMFunction_MathDoubleRemap(
    SourceMaximum=85
    SourceMinimum=75
    TargetMaximum=1
    TargetMinimum=0
    Value=&RigUnit_GetRelativeTransformForItem(Child=(Type=Bone,Name="pelvis"), Parent=(Type=Bone,Name="root"), bChildInitial=False, bParentInitial=False).RelativeTransform.Translation.Z
   )
   &RigUnit_ModifyTransforms(
    ItemToModify="((Item=(Type=Bone,Name=\"dyn_skirt_arrow_03_base\"),Transform=(Rotation=(X=0,Y=0,Z=0.1736,W=0.9848),Translation=(X=1,Y=-1,Z=0),Scale3D=(X=1,Y=1,Z=1))),(Item=(Type=Bone,Name=\"dyn_skirt_arrow_04_base\"),Transform=(Rotation=(X=0.2301,Y=-0.028,Z=0.2635,W=0.9364), ...(truncated)"
    WeightMinimum=0
    Weight=&RigVMFunction_MathDoubleMul(A=@n2.Value, B=@n3.Result).Result
    ExecutePin=@n1.A
   )
   n4 = RigVMDispatchNode(BoneName=Root, Default=0, Name=deformRig_l_thigh_up_alpha) {template="DISPATCH_RigDispatch_GetAnimAttribute(in Name,in BoneName,in Default,out Value,out Found)"}
   > &RigUnit_ModifyTransforms(
    ItemToModify="((Item=(Type=Bone,Name=\"dyn_skirt_arrow_03_base\"),Transform=(Rotation=(X=-0.1787,Y=0.0044,Z=0.2203,W=0.9589),Translation=(X=1,Y=-1,Z=0),Scale3D=(X=1,Y=1,Z=1))),(Item=(Type=Bone,Name=\"dyn_skirt_arrow_04_base\"),Transform=(Rotation=(X=0,Y=0,Z=0.1736,W=0.9848) ...(truncated)"
    WeightMinimum=0
    Weight=&RigVMFunction_MathDoubleMul(A=@n4.Value, B=@n3.Result).Result
   )
   n5 = RigVMVariableNode() {variable=Crouching ON}
   n6 = &RigVMFunction_KalmanFloat(BufferSize=16, Value=&RigVMFunction_MathDoubleMul(A=@n2.Value, B=@n5.Value).Result) {template="KalmanFilter::Execute(in Value,in BufferSize,out Result)"}
   > &RigUnit_ModifyTransforms(
    ItemToModify=((Item=(Type=Bone,Name="dyn_skirt_arrow_03_base"),Transform=(Rotation=(X=0,Y=0,Z=0.0436,W=0.999),Translation=(X=0,Y=2,Z=0),Scale3D=(X=1,Y=1,Z=1))))
    WeightMinimum=0
    Weight=@n6.Result
   )
   > &RigUnit_ModifyTransforms(
    ItemToModify=((Item=(Type=Bone,Name="dyn_skirt_arrow_04_base"),Transform=(Rotation=(X=0,Y=0,Z=0.0872,W=0.9962),Translation=(X=0,Y=2,Z=0),Scale3D=(X=1,Y=1,Z=1))))
    WeightMinimum=0
    Weight=@n6.Result
   )
   > &RigUnit_ModifyTransforms(
    ItemToModify=((Item=(Type=Bone,Name="dyn_skirt_arrow_02_base"),Transform=(Rotation=(X=0,Y=0,Z=0.0436,W=0.999),Translation=(X=0,Y=0,Z=0),Scale3D=(X=1,Y=1,Z=1))))
    WeightMinimum=0
    Weight=@n6.Result
   )
   > &RigUnit_ModifyTransforms(
    ItemToModify=((Item=(Type=Bone,Name="dyn_skirt_arrow_05_base"),Transform=(Rotation=(X=0,Y=0,Z=0.1305,W=0.9914),Translation=(X=0,Y=2,Z=0),Scale3D=(X=1,Y=1,Z=1))))
    WeightMinimum=0
    Weight=@n6.Result
   )
   > &RigUnit_ModifyTransforms(
    ItemToModify=((Item=(Type=Bone,Name="dyn_skirt_arrow_06_base"),Transform=(Rotation=(X=0,Y=0,Z=0.0872,W=0.9962),Translation=(X=0,Y=0,Z=0),Scale3D=(X=1,Y=1,Z=1))))
    WeightMinimum=0
    Weight=@n6.Result
   )
   > &RigUnit_ModifyTransforms(
    ItemToModify=((Item=(Type=Bone,Name="dyn_skirt_arrow_01_base"),Transform=(Rotation=(X=0,Y=0,Z=0,W=1),Translation=(X=0,Y=0,Z=0),Scale3D=(X=1,Y=1,Z=1))))
    WeightMinimum=0
    Weight=@n6.Result
   )
   > &RigUnit_ModifyTransforms(
    ItemToModify="((Item=(Type=Bone,Name=\"dyn_skirt_arrow_07_base\"),Transform=(Rotation=(X=0,Y=0,Z=0.0436,W=0.999),Translation=(X=0,Y=0,Z=0),Scale3D=(X=1,Y=1,Z=1))),(Item=(Type=Bone,Name=\"dyn_skirt_fr_r_base\"),Transform=(Rotation=(X=0,Y=-0.1045,Z=0,W=0.9945),Translation=(X= ...(truncated)"
    WeightMinimum=0
    Weight=@n6.Result
   )
   > n7 = &RigUnit_SphericalPoseReader(
    ActiveRegionScaleFactors=(PositiveWidth=1.5,NegativeWidth=1.5,PositiveHeight=1.5,NegativeHeight=1.5)
    ActiveRegionSize=0.1
    Debug=(bDrawDebug=True,bDraw2D=False,bDrawLocalAxes=False,DebugScale=25,DebugSegments=20,DebugThickness=0.25)
    DriverAxis=(X=-1,Y=0,Z=0)
    DriverItem=(Type=Bone,Name="thigh_l")
    FalloffRegionScaleFactors=(PositiveWidth=2,NegativeWidth=2,PositiveHeight=2,NegativeHeight=2)
    FalloffSize=0.4
    FlipHeightScaling=False
    FlipWidthScaling=False
    OptionalParentItem=(Type=Bone,Name="None")
    RotationOffset=(X=110,Y=-20,Z=20)
   )
   > n8 = &RigUnit_SphericalPoseReader(
    ActiveRegionScaleFactors=(PositiveWidth=1.5,NegativeWidth=1.5,PositiveHeight=1.5,NegativeHeight=1.5)
    ActiveRegionSize=0.1
    Debug=(bDrawDebug=True,bDraw2D=False,bDrawLocalAxes=False,DebugScale=25,DebugSegments=20,DebugThickness=0.25)
    DriverAxis=(X=1,Y=0,Z=0)
    DriverItem=(Type=Bone,Name="thigh_r")
    FalloffRegionScaleFactors=(PositiveWidth=2,NegativeWidth=2,PositiveHeight=2,NegativeHeight=2)
    FalloffSize=0.4
    FlipHeightScaling=False
    FlipWidthScaling=False
    OptionalParentItem=(Type=Bone,Name="None")
    RotationOffset=(X=80,Y=0,Z=-60)
   )
   n9 = &RigVMFunction_MathDoubleMul(B=@n7.OutputParam, A=@n8.OutputParam)
   n10 = RigVMVariableNode() {variable=Jumping ON}
   > &RigUnit_ModifyTransforms(
    ItemToModify="((Item=(Type=Bone,Name=\"dyn_skirt_arrow_03_base\"),Transform=(Rotation=(X=0,Y=0,Z=0.0872,W=0.9962),Translation=(X=0,Y=0,Z=0),Scale3D=(X=1,Y=1,Z=1))),(Item=(Type=Bone,Name=\"dyn_skirt_arrow_04_base\"),Transform=(Rotation=(X=-0.4162,Y=-0.0734,Z=0.1574,W=0.8925) ...(truncated)"
    WeightMinimum=0
    Weight=&RigVMFunction_KalmanFloat(BufferSize=16, Value=&RigVMFunction_MathDoubleMul(A=@n9.Result, B=@n10.Value).Result) {template="KalmanFilter::Execute(in Value,in BufferSize,out Result)"}.Result
   )
   n11 = &RigVMFunction_KalmanFloat(BufferSize=16, Value=&RigVMFunction_MathDoubleMul(B=@n4.Value, A=@n5.Value).Result) {template="KalmanFilter::Execute(in Value,in BufferSize,out Result)"}
   &RigUnit_ModifyTransforms(
    ItemToModify=((Item=(Type=Bone,Name="dyn_skirt_arrow_03_base"),Transform=(Rotation=(X=-0.1735,Y=-0.0076,Z=0.043,W=0.9839),Translation=(X=0,Y=0,Z=0),Scale3D=(X=1,Y=1,Z=1))))
    WeightMinimum=0
    Weight=@n11.Result
    ExecutePin=@n1.B
   )
   > &RigUnit_ModifyTransforms(
    ItemToModify=((Item=(Type=Bone,Name="dyn_skirt_arrow_04_base"),Transform=(Rotation=(X=-0.1735,Y=-0.0076,Z=0.043,W=0.9839),Translation=(X=0,Y=0,Z=0),Scale3D=(X=1,Y=1,Z=1))))
    WeightMinimum=0
    Weight=@n11.Result
   )
   > &RigUnit_ModifyTransforms(
    ItemToModify=((Item=(Type=Bone,Name="dyn_skirt_arrow_02_base"),Transform=(Rotation=(X=0,Y=0,Z=0.0436,W=0.999),Translation=(X=0,Y=0,Z=0),Scale3D=(X=1,Y=1,Z=1))))
    WeightMinimum=0
    Weight=@n11.Result
   )
   > &RigUnit_ModifyTransforms(
    ItemToModify=((Item=(Type=Bone,Name="dyn_skirt_arrow_05_base"),Transform=(Rotation=(X=-0.1736,Y=0,Z=0,W=0.9848),Translation=(X=0,Y=0,Z=0),Scale3D=(X=1,Y=1,Z=1))))
    WeightMinimum=0
    Weight=@n11.Result
   )
   > &RigUnit_ModifyTransforms(
    ItemToModify=((Item=(Type=Bone,Name="dyn_skirt_arrow_06_base"),Transform=(Rotation=(X=0,Y=0,Z=0.0436,W=0.999),Translation=(X=0,Y=0,Z=0),Scale3D=(X=1,Y=1,Z=1))))
    WeightMinimum=0
    Weight=@n11.Result
   )
   > &RigUnit_ModifyTransforms(
    ItemToModify=((Item=(Type=Bone,Name="dyn_skirt_arrow_01_base"),Transform=(Rotation=(X=0,Y=0,Z=0,W=1),Translation=(X=0,Y=0,Z=0),Scale3D=(X=1,Y=1,Z=1))))
    WeightMinimum=0
    Weight=@n11.Result
   )
   > &RigUnit_ModifyTransforms(
    ItemToModify=((Item=(Type=Bone,Name="dyn_skirt_arrow_07_base"),Transform=(Rotation=(X=0,Y=0,Z=0,W=1),Translation=(X=0,Y=0,Z=0),Scale3D=(X=1,Y=1,Z=1))))
    WeightMinimum=0
    Weight=@n11.Result
   )
   > &RigUnit_ModifyTransforms(
    ItemToModify="((Item=(Type=Bone,Name=\"dyn_skirt_arrow_03_base\"),Transform=(Rotation=(X=0,Y=0,Z=0.0436,W=0.999),Translation=(X=0,Y=0,Z=0),Scale3D=(X=1,Y=1,Z=1))),(Item=(Type=Bone,Name=\"dyn_skirt_arrow_04_base\"),Transform=(Rotation=(X=0,Y=0,Z=0.0436,W=0.999),Translation=( ...(truncated)"
    WeightMinimum=0
    Weight=&RigVMFunction_KalmanFloat(BufferSize=16, Value=RigVMVariableNode() {variable=KneeSliding ON}.Value) {template="KalmanFilter::Execute(in Value,in BufferSize,out Result)"}.Result
   )
   > &RigUnit_ModifyTransforms(
    ItemToModify=((Item=(Type=Bone,Name="upperarm_twist_02_r"),Transform=(Rotation=(X=0,Y=0,Z=0,W=1),Translation=(X=0,Y=0,Z=0),Scale3D=(X=1,Y=1.2,Z=1.2))))
    Weight=1
    WeightMinimum=0
   )
   DISPATCH_RigDispatch_GetAnimAttribute = RigVMDispatchNode(BoneName=Root, Default=0, Name=deformRig_l_thigh_up_alpha) {template="DISPATCH_RigDispatch_GetAnimAttribute(in Name,in BoneName,in Default,out Value,out Found)"}
   &RigUnit_ModifyTransforms(
    ItemToModify="((Item=(Type=Bone,Name=\"dyn_skirt_fr_mid_base\"),Transform=(Rotation=(X=-0.6246,Y=0.1563,Z=-0.3136,W=0.6979),Translation=(X=0,Y=0,Z=0),Scale3D=(X=1,Y=1,Z=1))),(Item=(Type=Bone,Name=\"dyn_skirt_fr_l_base\"),Transform=(Rotation=(X=0,Y=0,Z=0,W=1),Translation=(X= ...(truncated)"
    WeightMinimum=0
    Weight=@DISPATCH_RigDispatch_GetAnimAttribute.Value
    ExecutePin=@RigVMFunction_Sequence.A
   )
   > &RigUnit_ModifyTransforms(
    ItemToModify="((Item=(Type=Bone,Name=\"dyn_skirt_fr_mid_base\"),Transform=(Rotation=(X=0.6166,Y=0.2923,Z=0.1933,W=0.705),Translation=(X=0,Y=-4,Z=0),Scale3D=(X=1,Y=1,Z=1))),(Item=(Type=Bone,Name=\"dyn_skirt_fr_r_base\"),Transform=(Rotation=(X=0,Y=0.1108,Z=0,W=0.9938),Transla ...(truncated)"
    WeightMinimum=0
    Weight=RigVMDispatchNode(BoneName=Root, Default=0, Name=deformRig_r_thigh_up_alpha) {template="DISPATCH_RigDispatch_GetAnimAttribute(in Name,in BoneName,in Default,out Value,out Found)"}.Value
   )
   > n12 = &RigUnit_SphericalPoseReader(
    ActiveRegionScaleFactors=(PositiveWidth=1,NegativeWidth=1,PositiveHeight=2,NegativeHeight=2)
    ActiveRegionSize=0.1
    Debug=(bDrawDebug=true,bDraw2D=False,bDrawLocalAxes=False,DebugScale=25,DebugSegments=20,DebugThickness=0.25)
    DriverAxis=(X=-1,Y=0,Z=0)
    DriverItem=(Type=Bone,Name="thigh_l")
    FalloffRegionScaleFactors=(PositiveWidth=1,NegativeWidth=1,PositiveHeight=1,NegativeHeight=1)
    FalloffSize=0.3
    FlipHeightScaling=False
    FlipWidthScaling=False
    OptionalParentItem=(Type=Bone,Name="spine_01")
    RotationOffset=(X=90,Y=0,Z=55)
   )
   > n13 = &RigUnit_SphericalPoseReader(
    ActiveRegionScaleFactors=(PositiveWidth=1,NegativeWidth=1,PositiveHeight=2,NegativeHeight=2)
    ActiveRegionSize=0.1
    Debug=(bDrawDebug=true,bDraw2D=False,bDrawLocalAxes=False,DebugScale=25,DebugSegments=20,DebugThickness=0.25)
    DriverAxis=(X=1,Y=0,Z=0)
    DriverItem=(Type=Bone,Name="thigh_r")
    FalloffRegionScaleFactors=(PositiveWidth=0.5,NegativeWidth=1,PositiveHeight=1,NegativeHeight=1)
    FalloffSize=0.3
    FlipHeightScaling=False
    FlipWidthScaling=False
    OptionalParentItem=(Type=Bone,Name="spine_01")
    RotationOffset=(X=-80,Y=0,Z=65)
   )
   Add = &RigVMFunction_MathDoubleAdd(A=RigVMVariableNode() {variable=Is Backward Forward Movement}.Value, B=RigVMVariableNode() {variable=Is Right Left Movement}.Value)
   Absolute = &RigVMFunction_MathDoubleAbs(Value=@Add.Result)
   KalmanFilter = &RigVMFunction_KalmanFloat(
    BufferSize=16
    Value=&RigVMFunction_MathDoubleRemap(SourceMaximum=1, SourceMinimum=0, TargetMaximum=1, TargetMinimum=0, Value=@Absolute.Result).Result
   ) {template="KalmanFilter::Execute(in Value,in BufferSize,out Result)"}
   Multiply = &RigVMFunction_MathDoubleMul(A=@n12.OutputParam, B=@n13.OutputParam)
   > &RigUnit_ModifyTransforms(
    ItemToModify=((Item=(Type=Bone,Name="dyn_skirt_fr_mid_base"),Transform=(Rotation=(X=-0.1183,Y=0,Z=0,W=0.993),Translation=(X=-1,Y=-3.5,Z=0.5082),Scale3D=(X=1,Y=1,Z=1))))
    WeightMinimum=0
    Weight=&RigVMFunction_MathDoubleMul(B=@KalmanFilter.Result, A=@Multiply.Result).Result
   )
   n14 = &RigUnit_GetTransform(Item=(Type=Bone,Name="root"), bInitial=False)
   n15 = &RigUnit_GetTransform(Item=(Type=Bone,Name="pelvis"), bInitial=False)
   n16 = &RigVMFunction_AlphaInterp(
    InterpSpeedDecreasing=8
    InterpSpeedIncreasing=8
    OutRange=(Max=1)
    Scale=1
    bClampResult=true
    bMapRange=true
    Bias=~
    ClampMin=~
    Value=&RigVMFunction_MathDoubleRemap(
     SourceMaximum=22
     SourceMinimum=32
     TargetMaximum=1
     TargetMinimum=0
     Value=&RigVMFunction_MathVectorDistance(A=(X=0,Y=0,Z=0), B=(X=0,Y=0,Z=0), B.Z=@n14.Transform.Translation.Z, A.Z=@n15.Transform.Translation.Z).Result
    ).Result
   )
   > &RigUnit_ModifyTransforms(
    ItemToModify="((Item=(Type=Bone,Name=\"dyn_skirt_bk_mid_base\"),Transform=(Rotation=(X=0,Y=0,Z=0,W=1),Translation=(X=0,Y=0,Z=0),Scale3D=(X=0.01,Y=1,Z=1))),(Item=(Type=Bone,Name=\"dyn_skirt_bk_l_base\"),Transform=(Rotation=(X=0,Y=0,Z=0,W=1),Translation=(X=0,Y=0,Z=0),Scale3D= ...(truncated)"
    WeightMinimum=0
    Weight=@n16.Result
   )
   n17 = &RigVMFunction_MathVectorDistance(
    A=(X=0,Y=0,Z=0)
    B=(X=0,Y=0,Z=0)
    B.Z=&RigUnit_GetTransform(Item=(Type=Bone,Name="root"), bInitial=False).Transform.Translation.Z
    A.Z=&RigUnit_GetTransform(Item=(Type=Bone,Name="pelvis"), bInitial=False).Transform.Translation.Z
   )
   n18 = &RigVMFunction_AlphaInterp(
    InterpSpeedDecreasing=8
    InterpSpeedIncreasing=8
    OutRange=(Max=1)
    Scale=1
    bClampResult=true
    bMapRange=true
    Bias=~
    ClampMin=~
    Value=&RigVMFunction_MathDoubleMul(
     B=1
     A=&RigVMFunction_MathDoubleRemap(SourceMaximum=22, SourceMinimum=32, TargetMaximum=1, TargetMinimum=0, Value=@n17.Result).Result
    ).Result
   )
   > &RigUnit_ModifyTransforms(
    ItemToModify="((Item=(Type=Bone,Name=\"dyn_skirt_bk_mid_base\"),Transform=(Rotation=(X=0,Y=0,Z=0,W=1),Translation=(X=0,Y=0,Z=0),Scale3D=(X=0.01,Y=1,Z=1))),(Item=(Type=Bone,Name=\"dyn_skirt_bk_l_base\"),Transform=(Rotation=(X=0,Y=0,Z=0,W=1),Translation=(X=0,Y=0,Z=0),Scale3D= ...(truncated)"
    WeightMinimum=0
    Weight=@n18.Result
   )
   > n19 = &RigUnit_SphericalPoseReader(
    ActiveRegionScaleFactors=(PositiveWidth=1.5,NegativeWidth=1.5,PositiveHeight=1.5,NegativeHeight=1.5)
    ActiveRegionSize=0.1
    Debug=(bDrawDebug=True,bDraw2D=False,bDrawLocalAxes=False,DebugScale=25,DebugSegments=20,DebugThickness=0.25)
    DriverAxis=(X=-1,Y=0,Z=0)
    DriverItem=(Type=Bone,Name="thigh_l")
    FalloffRegionScaleFactors=(PositiveWidth=2,NegativeWidth=2,PositiveHeight=2,NegativeHeight=2)
    FalloffSize=0.2
    FlipHeightScaling=False
    FlipWidthScaling=False
    OptionalParentItem=(Type=Bone,Name="None")
    RotationOffset=(X=110,Y=-20,Z=20)
   )
   > n20 = &RigUnit_SphericalPoseReader(
    ActiveRegionScaleFactors=(PositiveWidth=1.5,NegativeWidth=1.5,PositiveHeight=1.5,NegativeHeight=1.5)
    ActiveRegionSize=0.1
    Debug=(bDrawDebug=True,bDraw2D=False,bDrawLocalAxes=False,DebugScale=25,DebugSegments=20,DebugThickness=0.25)
    DriverAxis=(X=1,Y=0,Z=0)
    DriverItem=(Type=Bone,Name="thigh_r")
    FalloffRegionScaleFactors=(PositiveWidth=2,NegativeWidth=2,PositiveHeight=2,NegativeHeight=2)
    FalloffSize=0.2
    FlipHeightScaling=False
    FlipWidthScaling=False
    OptionalParentItem=(Type=Bone,Name="None")
    RotationOffset=(X=80,Y=0,Z=-50)
   )
   n21 = &RigVMFunction_AlphaInterp(
    InterpSpeedDecreasing=8
    InterpSpeedIncreasing=8
    OutRange=(Max=1)
    Scale=1
    bClampResult=true
    bMapRange=true
    Bias=~
    ClampMin=~
    Value=&RigVMFunction_MathDoubleMul(B=@n19.OutputParam, A=@n20.OutputParam).Result
   )
   > &RigUnit_ModifyTransforms(
    ItemToModify=((Item=(Type=Bone,Name="dyn_skirt_fr_mid_base"),Transform=(Rotation=(X=0,Y=0,Z=0,W=1),Translation=(X=0,Y=-13.7109,Z=0),Scale3D=(X=1,Y=1,Z=1))))
    WeightMinimum=0
    Weight=@n21.Result
   )
   n22 = RigVMVariableNode() {variable=Is Backward Forward Movement}
   RigVMVariableNode(
    Value=&RigVMFunction_MathDoubleRemap(SourceMaximum=1, SourceMinimum=0, TargetMaximum=1, TargetMinimum=0, Value=@n22.Value).Result
    ExecuteContext=@RigVMFunction_Sequence.B
   ) {variable=Is Forward Movement}
   > RigVMVariableNode(Value=&RigVMFunction_MathDoubleRemap(SourceMaximum=0, SourceMinimum=-1, TargetMaximum=0, TargetMinimum=1, Value=@n22.Value).Result) {variable=Is Backward Movement}
   > &RigUnit_ModifyTransforms(
    ItemToModify=((Item=(Type=Bone,Name="dyn_leg_strap_r_base"),Transform=(Rotation=(X=-0.0391,Y=-0.1193,Z=-0.3659,W=0.9222),Translation=(X=0,Y=0,Z=0),Scale3D=(X=1,Y=1,Z=1))))
    WeightMinimum=0
    Weight=RigVMDispatchNode(BoneName=Root, Default=0, Name=deformRig_r_knee_alpha) {template="DISPATCH_RigDispatch_GetAnimAttribute(in Name,in BoneName,in Default,out Value,out Found)"}.Value
   )
   > &RigUnit_ModifyTransforms(
    ItemToModify=((Item=(Type=Bone,Name="dyn_necklace_bottom_base"),Transform=(Rotation=(X=0,Y=0,Z=0,W=1),Translation=(X=-0.4739,Y=1,Z=0),Scale3D=(X=1,Y=1,Z=1))))
    WeightMinimum=0
    Weight=RigVMDispatchNode(BoneName=Root, Default=0, Name=deformRig_l_pec_alpha) {template="DISPATCH_RigDispatch_GetAnimAttribute(in Name,in BoneName,in Default,out Value,out Found)"}.Value
   )
   > &RigUnit_ModifyTransforms(
    ItemToModify=((Item=(Type=Bone,Name="dyn_necklace_bottom_base"),Transform=(Rotation=(X=0,Y=0,Z=0,W=1),Translation=(X=-0.4739,Y=1,Z=0),Scale3D=(X=1,Y=1,Z=1))))
    WeightMinimum=0
    Weight=RigVMDispatchNode(BoneName=Root, Default=0, Name=deformRig_r_pec_alpha) {template="DISPATCH_RigDispatch_GetAnimAttribute(in Name,in BoneName,in Default,out Value,out Found)"}.Value
   )
   > n23 = &RigUnit_SphericalPoseReader(
    ActiveRegionScaleFactors=(PositiveWidth=1,NegativeWidth=1,PositiveHeight=1,NegativeHeight=1)
    ActiveRegionSize=0.15
    Debug=(bDrawDebug=True,bDraw2D=False,bDrawLocalAxes=False,DebugScale=25,DebugSegments=20,DebugThickness=0.25)
    DriverAxis=(X=-1,Y=0,Z=0)
    DriverItem=(Type=Bone,Name="upperarm_r")
    FalloffRegionScaleFactors=(PositiveWidth=1,NegativeWidth=1,PositiveHeight=1,NegativeHeight=1)
    FalloffSize=0.3
    FlipHeightScaling=False
    FlipWidthScaling=False
    OptionalParentItem=(Type=Bone,Name="None")
    RotationOffset=(X=-130,Y=20,Z=40)
   )
   n24 = &RigVMFunction_AlphaInterp(
    InterpSpeedDecreasing=8
    InterpSpeedIncreasing=8
    OutRange=(Max=1)
    Scale=1
    bClampResult=true
    bMapRange=true
    Bias=~
    ClampMin=~
    Value=&RigVMFunction_MathDoubleMul(B=1, A=@n23.OutputParam).Result
   )
   > &RigUnit_ModifyTransforms(
    ItemToModify=((Item=(Type=Bone,Name="support_shoulder_r_rotator"),Transform=(Rotation=(X=0,Y=0,Z=0,W=1),Translation=(X=-2,Y=0,Z=0),Scale3D=(X=1,Y=1,Z=1))))
    WeightMinimum=0
    Weight=@n24.Result
   )
   > &RigUnit_ModifyTransforms(
    ItemToModify=((Item=(Type=Bone,Name="skirt_bk_ALT_support"),Transform=(Rotation=(X=0,Y=0,Z=-0.2164,W=0.9763),Translation=(X=-4,Y=0,Z=0),Scale3D=(X=1,Y=1,Z=1))))
    WeightMinimum=0
    Weight=RigVMDispatchNode(BoneName=Root, Default=0, Name=deformRig_l_knee_alpha) {template="DISPATCH_RigDispatch_GetAnimAttribute(in Name,in BoneName,in Default,out Value,out Found)"}.Value
   )
   > &RigUnit_ModifyTransforms(
    ItemToModify=((Item=(Type=Bone,Name="skirt_bk_ALT_support"),Transform=(Rotation=(X=0,Y=0,Z=-0.2164,W=0.9763),Translation=(X=-4,Y=0,Z=0),Scale3D=(X=1,Y=1,Z=1))))
    WeightMinimum=0
    Weight=RigVMDispatchNode(BoneName=Root, Default=0, Name=deformRig_r_knee_alpha) {template="DISPATCH_RigDispatch_GetAnimAttribute(in Name,in BoneName,in Default,out Value,out Found)"}.Value
   )
   n25 = RigVMVariableNode() {variable=Is Right Left Movement}
   RigVMVariableNode(
    Value=&RigVMFunction_MathDoubleRemap(SourceMaximum=0, SourceMinimum=-1, TargetMaximum=0, TargetMinimum=1, Value=@n25.Value).Result
    ExecuteContext=@RigVMFunction_Sequence.C
   ) {variable=Is Left Movement}
   > RigVMVariableNode(Value=&RigVMFunction_MathDoubleRemap(SourceMaximum=1, SourceMinimum=0, TargetMaximum=1, TargetMinimum=0, Value=@n25.Value).Result) {variable=Is Right Movement}
  ]}
]
hierarchy [
 root {loc=<0 0 0>, rot=<0 0 0 1>}
 pelvis {parent=root, loc=<0 -0.5301 98.0571>, rot=<0.0127 0.707 -0.0127 -0.707>}
 spine_01 {parent=pelvis, loc=<6.1208 -0.9328 0>, rot=<0 0 0.034 -0.9994>}
 spine_02 {parent=spine_01, loc=<6.5997 -0.2095 0>, rot=<0 0 0.0238 -0.9997>}
 spine_03 {parent=spine_02, loc=<6.4914 -0.1237 0>, rot=<0 0 0.0149 -0.9999>}
 spine_04 {parent=spine_03, loc=<6.9598 0.0794 0>, rot=<0 0 -0.0132 -0.9999>}
 spine_05 {parent=spine_04, loc=<6.9987 0 0>, rot=<0 0 -0.0015 -1>}
 neck_01 {parent=spine_05, loc=<11.3138 0.1529 0>, rot=<0 0 -0.2003 -0.9797>}
 neck_02 {parent=neck_01, loc=<4.5627 0 0>, rot=<0 0 0.0504 -0.9987>}
 head {parent=neck_02, loc=<5.3914 0 0>, rot=<0 0 0.115 -0.9934>}
 clavicle_l {parent=spine_05, loc=<5.4831 -0.1053 -3.2946>, rot=<0.0465 -0.7289 -0.009 -0.6829>}
 upperarm_l {parent=clavicle_l, loc=<9.8658 -1.2136 -0.4105>, rot=<0.0164 -0.3245 -0.0269 -0.9454>}
 lowerarm_l {parent=upperarm_l, loc=<26.465 -0.8345 0.259>, rot=<0.0036 0.0161 -0.2298 -0.9731>}
 hand_l {parent=lowerarm_l, loc=<22.7355 -0.3203 -0.6105>, rot=<-0.6784 -0.0223 0.0735 -0.7307>}
 thumb_01_l {parent=hand_l, loc=<1.3118 -0.3191 -2.459>, rot=<0.6546 -0.4042 0.0664 -0.6354>}
 thumb_02_l {parent=thumb_01_l, loc=<5.9336 -0.0448 0.0401>, rot=<-0.0037 0.0213 0.298 -0.9543>}
 thumb_03_l {parent=thumb_02_l, loc=<2.2223 0.3682 -0.0907>, rot=<0.0118 -0.013 0.0066 -0.9998>}
 index_metacarpal_l {parent=hand_l, loc=<1.6535 0.0183 -2.96>, rot=<0.0158 -0.0124 -0.0173 -0.9996>}
 index_01_l {parent=index_metacarpal_l, loc=<8.3008 0.0357 -0.0524>, rot=<-0.0044 0.0485 0.0924 -0.9945>}
 index_02_l {parent=index_01_l, loc=<3.3587 -0.3004 0.0452>, rot=<-0.002 0.0027 0.3225 -0.9466>}
 index_03_l {parent=index_02_l, loc=<2.5578 -0.123 -0.0252>, rot=<-0.002 0.0067 0.2558 -0.9667>}
 middle_metacarpal_l {parent=hand_l, loc=<2.1856 -0.1725 -0.5984>, rot=<0.006 0.0168 -0.0329 -0.9993>}
 middle_01_l {parent=middle_metacarpal_l, loc=<7.3275 -0.2447 0.0129>, rot=<-0.0208 0.0042 0.0858 -0.9961>}
 middle_02_l {parent=middle_01_l, loc=<4.0123 0.1401 0.1175>, rot=<-0.0218 0.0258 0.3512 -0.9357>}
 middle_03_l {parent=middle_02_l, loc=<2.6077 0.0534 0.1373>, rot=<0.0082 0.0188 0.1708 -0.9851>}
 ring_metacarpal_l {parent=hand_l, loc=<2.2074 -0.3625 1.3638>, rot=<-0.0284 0.0722 -0.0458 -0.9959>}
 ring_01_l {parent=ring_metacarpal_l, loc=<7.0295 -0.6197 -0.1446>, rot=<-0.0489 -0.0252 0.1892 -0.9804>}
 ring_02_l {parent=ring_01_l, loc=<3.6174 0.4647 -0.1027>, rot=<0 0 0.2806 -0.9598>}
 ring_03_l {parent=ring_02_l, loc=<2.4152 0.114 0.0199>, rot=<0 0 0.1189 -0.9929>}
 pinky_metacarpal_l {parent=hand_l, loc=<1.7616 -0.9753 3.3814>, rot=<-0.0397 0.0814 -0.069 -0.9935>}
 pinky_01_l {parent=pinky_metacarpal_l, loc=<7.1709 -0.4831 -0.1094>, rot=<-0.0654 -0.0035 0.2242 -0.9723>}
 pinky_02_l {parent=pinky_01_l, loc=<2.8274 0.0893 -0.1043>, rot=<0.0037 0.0011 0.2854 -0.9584>}
 pinky_03_l {parent=pinky_02_l, loc=<2.2046 0.0307 -0.0078>, rot=<-0.0002 0.0098 0.1445 -0.9895>}
 deform_wrist_dn_base_l {parent=hand_l, loc=<0 -1.5326 0>, rot=<0 0 0 -1>}
 deform_wrist_dn_l {parent=deform_wrist_dn_base_l, loc=<0 0 0>, rot=<0 0 0 -1>}
 deform_wrist_up_base_l {parent=hand_l, loc=<0 2.3009 0>, rot=<0 0 0 -1>}
 deform_wrist_up_l {parent=deform_wrist_up_base_l, loc=<0 0 0>, rot=<0 0 0 -1>}
 lowerarm_twist_02_l {parent=lowerarm_l, loc=<10.6707 0 0>, rot=<0 0 0 1>}
 lowerarm_twist_01_l {parent=lowerarm_l, loc=<18.316 0 0>, rot=<0 0 0 1>}
 deform_elbow_in_base_l {parent=lowerarm_l, loc=<0.1834 3.6062 -0.2572>, rot=<0 0 0 -1>}
 deform_elbow_in_l {parent=deform_elbow_in_base_l, loc=<0 0 0>, rot=<0 0 0 -1>}
 deform_elbow_out_base_l {parent=lowerarm_l, loc=<0 -4 0>, rot=<0 0 0 1>}
 deform_elbow_out_l {parent=deform_elbow_out_base_l, loc=<0 0 0>, rot=<0 0 0 1>}
 upperarm_twist_01_l {parent=upperarm_l, loc=<7.2135 -0.1134 0>, rot=<0 0 0 1>}
 deform_upperarm_bk_base_l {parent=upperarm_twist_01_l, loc=<-6.4053 -4.3746 0.314>, rot=<0 0 0 -1>}
 deform_upperarm_bk_l {parent=deform_upperarm_bk_base_l, loc=<0 0 0>, rot=<0 0 0 -1>}
 deform_upperarm_fr_base_l {parent=upperarm_twist_01_l, loc=<-5.6312 5.092 0.3354>, rot=<0 0 0 -1>}
 deform_upperarm_fr_l {parent=deform_upperarm_fr_base_l, loc=<0 0 0>, rot=<0 0 0 -1>}
 deform_upperarm_up_base_l {parent=upperarm_twist_01_l, loc=<-8.891 0.4825 3.9517>, rot=<0 0 0 -1>}
 deform_upperarm_up_l {parent=deform_upperarm_up_base_l, loc=<0 0 0>, rot=<0 0 0 -1>}
 dyn_shoulder_l_01 {parent=upperarm_twist_01_l, loc=<-2.7365 -7.0535 0.2723>, rot=<0 -0.2876 0 -0.9578>}
 dyn_shoulder_l_03 {parent=upperarm_twist_01_l, loc=<-3.2461 -1.4034 7.7726>, rot=<-0.0425 -0.1776 0.025 -0.9829>}
 upperarm_twist_02_l {parent=upperarm_l, loc=<19.7998 -0.1134 0>, rot=<0 0 0 1>}
 support_shoulder_l_rotator {parent=upperarm_l, loc=<0 0 0>, rot=<0 0 0 1>}
 support_shoulder_l {parent=support_shoulder_l_rotator, loc=<-4.2626 0 5.1178>, rot=<0 0.5337 0 -0.8457>}
 dyn_shoulder_l_02 {parent=support_shoulder_l_rotator, loc=<6.7127 6.1475 2.3883>, rot=<-0.244 -0.3532 0.0351 -0.9025>}
 clavicle_r {parent=spine_05, loc=<5.4831 -0.1053 3.2946>, rot=<0.7289 0.0465 -0.6829 0.009>}
 upperarm_r {parent=clavicle_r, loc=<-9.8658 1.2136 0.4105>, rot=<0.0164 -0.3245 -0.0269 -0.9454>}
 lowerarm_r {parent=upperarm_r, loc=<-26.465 0.8345 -0.259>, rot=<0.0036 0.0161 -0.2298 -0.9731>}
 hand_r {parent=lowerarm_r, loc=<-22.7355 0.3203 0.6106>, rot=<-0.6784 -0.0223 0.0735 -0.7307>}
 thumb_01_r {parent=hand_r, loc=<-1.3118 0.3191 2.459>, rot=<0.6546 -0.4042 0.0664 -0.6354>}
 thumb_02_r {parent=thumb_01_r, loc=<-5.9336 0.0448 -0.0401>, rot=<-0.0037 0.0213 0.298 -0.9543>}
 thumb_03_r {parent=thumb_02_r, loc=<-2.2223 -0.3682 0.0907>, rot=<0.0118 -0.013 0.0066 -0.9998>}
 index_metacarpal_r {parent=hand_r, loc=<-1.6535 -0.0183 2.96>, rot=<0.0158 -0.0124 -0.0173 -0.9996>}
 index_01_r {parent=index_metacarpal_r, loc=<-8.3008 -0.0357 0.0524>, rot=<-0.0044 0.0485 0.0924 -0.9945>}
 index_02_r {parent=index_01_r, loc=<-3.3587 0.3004 -0.0452>, rot=<-0.002 0.0027 0.3225 -0.9466>}
 index_03_r {parent=index_02_r, loc=<-2.5578 0.123 0.0252>, rot=<-0.002 0.0067 0.2558 -0.9667>}
 middle_metacarpal_r {parent=hand_r, loc=<-2.1856 0.1725 0.5984>, rot=<0.006 0.0168 -0.0329 -0.9993>}
 middle_01_r {parent=middle_metacarpal_r, loc=<-7.3275 0.2447 -0.0129>, rot=<-0.0208 0.0042 0.0858 -0.9961>}
 middle_02_r {parent=middle_01_r, loc=<-4.0123 -0.1401 -0.1175>, rot=<-0.0218 0.0258 0.3512 -0.9357>}
 middle_03_r {parent=middle_02_r, loc=<-2.6077 -0.0534 -0.1373>, rot=<0.0082 0.0188 0.1708 -0.9851>}
 ring_metacarpal_r {parent=hand_r, loc=<-2.2074 0.3625 -1.3638>, rot=<-0.0284 0.0722 -0.0458 -0.9959>}
 ring_01_r {parent=ring_metacarpal_r, loc=<-7.0295 0.6197 0.1446>, rot=<-0.0489 -0.0252 0.1892 -0.9804>}
 ring_02_r {parent=ring_01_r, loc=<-3.6174 -0.4647 0.1027>, rot=<0 0 0.2806 -0.9598>}
 ring_03_r {parent=ring_02_r, loc=<-2.4152 -0.114 -0.0199>, rot=<0 0 0.1189 -0.9929>}
 pinky_metacarpal_r {parent=hand_r, loc=<-1.7616 0.9753 -3.3814>, rot=<-0.0397 0.0814 -0.069 -0.9935>}
 pinky_01_r {parent=pinky_metacarpal_r, loc=<-7.1709 0.4831 0.1094>, rot=<-0.0654 -0.0035 0.2242 -0.9723>}
 pinky_02_r {parent=pinky_01_r, loc=<-2.8274 -0.0893 0.1043>, rot=<0.0037 0.0011 0.2854 -0.9584>}
 pinky_03_r {parent=pinky_02_r, loc=<-2.2046 -0.0307 0.0078>, rot=<-0.0002 0.0098 0.1445 -0.9895>}
 deform_wrist_dn_base_r {parent=hand_r, loc=<-0.0001 1.5328 0>, rot=<0 0 0 -1>}
 deform_wrist_dn_r {parent=deform_wrist_dn_base_r, loc=<0 0 0>, rot=<0 0 0 -1>}
 deform_wrist_up_base_r {parent=hand_r, loc=<-0.0002 -2.3007 -0.0001>, rot=<0 0 0 -1>}
 deform_wrist_up_r {parent=deform_wrist_up_base_r, loc=<0 0 0>, rot=<0 0 0 -1>}
 lowerarm_twist_02_r {parent=lowerarm_r, loc=<-10.6707 0 0>, rot=<0 0 0 -1>}
 lowerarm_twist_01_r {parent=lowerarm_r, loc=<-18.316 0 0>, rot=<0 0 0 -1>}
 deform_elbow_in_base_r {parent=lowerarm_r, loc=<-0.1834 -3.6062 0.2572>, rot=<0 0 0 -1>}
 deform_elbow_in_r {parent=deform_elbow_in_base_r, loc=<0 0 0>, rot=<0 0 0 -1>}
 deform_elbow_out_base_r {parent=lowerarm_r, loc=<0 4 -0.0001>, rot=<0 0 0 -1>}
 deform_elbow_out_r {parent=deform_elbow_out_base_r, loc=<0 0 0>, rot=<0 0 0 1>}
 upperarm_twist_01_r {parent=upperarm_r, loc=<-7.2135 0.1134 0>, rot=<0 0 0 -1>}
 deform_upperarm_bk_base_r {parent=upperarm_twist_01_r, loc=<6.4053 4.3746 -0.314>, rot=<0 0 0 -1>}
 deform_upperarm_bk_r {parent=deform_upperarm_bk_base_r, loc=<0 0 0>, rot=<0 0 0 -1>}
 deform_upperarm_fr_base_r {parent=upperarm_twist_01_r, loc=<5.6312 -5.092 -0.3354>, rot=<0 0 0 -1>}
 deform_upperarm_fr_r {parent=deform_upperarm_fr_base_r, loc=<0 0 0>, rot=<0 0 0 -1>}
 deform_upperarm_up_base_r {parent=upperarm_twist_01_r, loc=<8.891 -0.4825 -3.9517>, rot=<0 0 0 -1>}
 deform_upperarm_up_r {parent=deform_upperarm_up_base_r, loc=<0 0 0>, rot=<0 0 0 -1>}
 upperarm_twist_02_r {parent=upperarm_r, loc=<-19.7998 0.1134 0>, rot=<0 0 0 -1>}
 support_shoulder_r_rotator {parent=upperarm_r, loc=<0 0 0>, rot=<0 0 0 1>}
 support_shoulder_r {parent=support_shoulder_r_rotator, loc=<4.8155 0 -4.1954>, rot=<0 -0.4857 0 -0.8741>}
 attach_backpack {parent=spine_05, loc=<-0.123 -13.473 0>, rot=<-0.0093 -0.707 0.0093 -0.707>}
 Backpack_BR {parent=attach_backpack, type=Null, loc=<0 -1.7874 0>, rot=<0 0 0 1>}
 attach_cape {parent=spine_05, loc=<9.017 -11.714 0>, rot=<0.0616 -0.7044 -0.0616 -0.7044>}
 cape {parent=attach_cape, type=Null, loc=<0 -2.0445 0>, rot=<0 0 0 1>}
 dyn_necklace_bottom_base {parent=spine_05, loc=<1.2955 9.4012 0>, rot=<-0.2733 -0.9619 0 0>}
 dyn_necklace_bottom_01 {parent=dyn_necklace_bottom_base, loc=<-0.0009 0 0>, rot=<0 0 0 1>}
 dyn_necklace_bottom_02 {parent=dyn_necklace_bottom_01, loc=<9.1602 -0.5918 -1.6452>, rot=<-0.0209 0.0074 0.3331 -0.9426>}
 deform_pec_aim_base_l {parent=spine_04, loc=<3.2456 9.3309 -1.2237>, rot=<0.0603 -0.4033 0.3318 -0.8507>}
 deform_pec_aim_l {parent=deform_pec_aim_base_l, loc=<0 0 0>, rot=<0 0 0 1>}
 deform_pec_base_l {parent=deform_pec_aim_l, loc=<9.3665 0 0>, rot=<0 0 0 1>}
 deform_pec_l {parent=deform_pec_base_l, loc=<0 0 0>, rot=<0 0 0 -1>}
 deform_pec_aim_base_r {parent=spine_04, loc=<3.2452 9.3309 1.2237>, rot=<-0.4033 -0.0603 0.8507 0.3318>}
 deform_pec_aim_r {parent=deform_pec_aim_base_r, loc=<0 0 0>, rot=<0 0 0 1>}
 deform_pec_base_r {parent=deform_pec_aim_r, loc=<-9.3667 -0.0003 -0.0004>, rot=<0 0 0 1>}
 deform_pec_r {parent=deform_pec_base_r, loc=<0 0 0>, rot=<0 0 0 -1>}
 dyn_torso_l_base {parent=spine_04, loc=<0.5 5 -5.2>, rot=<-0.5 -0.5 -0.5 -0.5>}
 dyn_torso_l {parent=dyn_torso_l_base, loc=<0 0 0>, rot=<0 0 0 1>}
 dyn_torso_r_base {parent=spine_04, loc=<0.5 5 5.2>, rot=<-0.5 -0.5 -0.5 -0.5>}
 dyn_torso_r {parent=dyn_torso_r_base, loc=<0 0 0>, rot=<0 0 0 1>}
 thigh_l {parent=pelvis, loc=<-1.0855 -0.2109 -8.7053>, rot=<0.0009 0.0355 0.0155 -0.9993>}
 calf_l {parent=thigh_l, loc=<-42.5284 -0.0019 -0.7297>, rot=<0.0006 -0.0161 -0.0391 -0.9991>}
 foot_l {parent=calf_l, loc=<-46.7401 -2.5014 -0.592>, rot=<0.0427 -0.0238 0.0353 -0.9982>}
 ball_l {parent=foot_l, loc=<-3.5658 10.8066 -0.1544>, rot=<0.0001 -0.0001 -0.7063 -0.7079>}
 support_foot_l {parent=foot_l, loc=<3.5 -6 0>, rot=<0 0 0 1>}
 calf_twist_02_l {parent=calf_l, loc=<-19.1546 -1.7943 -0.2638>, rot=<0.0237 0.0095 -0.0035 -0.9997>}
 calf_twist_01_l {parent=calf_l, loc=<-35.8483 -1.9197 -0.5794>, rot=<0.0237 0.0095 -0.0035 -0.9997>}
 dyn_boot_fur_l_01 {parent=calf_twist_01_l, loc=<3.3435 2.2585 -4.8484>, rot=<-0.0502 -0.9748 -0.1634 -0.1437>}
 dyn_boot_fur_l_02 {parent=calf_twist_01_l, loc=<3.564 -3.4498 4.2088>, rot=<0.1322 -0.3135 0.9402 -0.0151>}
 dyn_boot_fur_l_03 {parent=calf_twist_01_l, loc=<2.8837 -5.7173 -2.6655>, rot=<0.0654 -0.839 0.5367 -0.0611>}
 dyn_boot_fur_l_04 {parent=calf_twist_01_l, loc=<3.7081 5.0066 1.6076>, rot=<0.2119 0.5144 0.8173 0.1498>}
 deform_knee_in_base_l {parent=calf_l, loc=<0.8031 -5.0231 0.2456>, rot=<0 0 0 -1>}
 deform_knee_in_l {parent=deform_knee_in_base_l, loc=<0 0 0>, rot=<0 0 0 -1>}
 deform_knee_out_base_l {parent=calf_l, loc=<0 4 0>, rot=<0 0 0 -1>}
 deform_knee_out_l {parent=deform_knee_out_base_l, loc=<0 0 0>, rot=<0 0 0 1>}
 thigh_twist_01_l {parent=thigh_l, loc=<-27.952 0 0>, rot=<0.0013 0 0 -1>}
 skirt_side_l_ALT_support {parent=thigh_twist_01_l, loc=<-8.5205 -2.8995 -9.9415>, rot=<-0.0026 -0.993 -0.0012 -0.1181>}
 deform_groin_base_l {parent=thigh_l, loc=<0.4801 5.6776 -0.2486>, rot=<0 0 0 -1>}
 deform_groin_l {parent=deform_groin_base_l, loc=<0 0 0>, rot=<0 0 0 -1>}
 dyn_skirt_side_l {parent=thigh_l, loc=<-19.0152 -2.7689 -9.6281>, rot=<-0.0168 -0.9872 0.0777 -0.1379>}
 dyn_skirt_arrow_01_base {parent=thigh_l, loc=<-5.6991 8.8864 1.0021>, rot=<-0.0474 -0.9841 0.1663 -0.0405>}
 dyn_skirt_arrow_01_01 {parent=dyn_skirt_arrow_01_base, loc=<0 0 0>, rot=<0 0 0 1>}
 dyn_skirt_arrow_01_02 {parent=dyn_skirt_arrow_01_01, loc=<11.0696 0 0>, rot=<0 0 0 1>}
 thigh_r {parent=pelvis, loc=<-1.0855 -0.2109 8.7053>, rot=<0.0355 -0.0009 0.9993 0.0155>}
 calf_r {parent=thigh_r, loc=<42.5284 0.0019 0.7297>, rot=<0.0006 -0.0161 -0.0391 -0.9991>}
 foot_r {parent=calf_r, loc=<46.7401 2.5014 0.592>, rot=<0.0427 -0.0238 0.0353 -0.9982>}
 ball_r {parent=foot_r, loc=<3.5658 -10.8066 0.1544>, rot=<0.0001 -0.0001 -0.7063 -0.7079>}
 support_foot_r {parent=foot_r, loc=<-3.5 6 0>, rot=<0 0 0 1>}
 calf_twist_02_r {parent=calf_r, loc=<19.1546 1.7943 0.2638>, rot=<0.0237 0.0095 -0.0035 -0.9997>}
 calf_twist_01_r {parent=calf_r, loc=<35.8483 1.9197 0.5794>, rot=<0.0237 0.0095 -0.0035 -0.9997>}
 dyn_boot_fur_r_04 {parent=calf_twist_01_r, loc=<-3.7082 -5.0066 -1.6076>, rot=<-0.8173 0.1498 0.2119 -0.5144>}
 dyn_boot_fur_r_01 {parent=calf_twist_01_r, loc=<-3.3434 -2.2585 4.8484>, rot=<-0.1634 0.1437 0.0502 -0.9748>}
 dyn_boot_fur_r_02 {parent=calf_twist_01_r, loc=<-3.564 3.4498 -4.2088>, rot=<0.9402 0.0151 -0.1322 -0.3135>}
 dyn_boot_fur_r_03 {parent=calf_twist_01_r, loc=<-2.8837 5.7173 2.6655>, rot=<0.5367 0.0611 -0.0654 -0.839>}
 deform_knee_in_base_r {parent=calf_r, loc=<-0.8031 5.0231 -0.2457>, rot=<0 0 0 -1>}
 deform_knee_in_r {parent=deform_knee_in_base_r, loc=<0 0 0>, rot=<0 0 0 -1>}
 deform_knee_out_base_r {parent=calf_r, loc=<0 -4 0>, rot=<0 0 0 -1>}
 deform_knee_out_r {parent=deform_knee_out_base_r, loc=<0 0 0>, rot=<0 0 0 1>}
 thigh_twist_01_r {parent=thigh_r, loc=<27.952 0 0>, rot=<0.0013 0 0 -1>}
 dyn_skirt_side_r {parent=thigh_twist_01_r, loc=<-2.0175 3.957 8.7887>, rot=<0.9879 0.0487 0.0964 0.1112>}
 dyn_leg_strap_r_base {parent=thigh_twist_01_r, loc=<1.9838 5.8333 -4.7221>, rot=<0.9343 0.0148 -0.0182 0.3558>}
 dyn_leg_strap_r_02 {parent=dyn_leg_strap_r_base, loc=<0.2318 0.3922 1.6119>, rot=<0.1631 0.2053 -0.1224 -0.9572>}
 dyn_leg_strap_r_01 {parent=dyn_leg_strap_r_base, loc=<0.3749 0.4164 -1.4519>, rot=<-0.0131 -0.3065 -0.0406 -0.9509>}
 deform_groin_base_r {parent=thigh_r, loc=<-0.4801 -5.6776 0.2486>, rot=<0 0 0 -1>}
 deform_groin_r {parent=deform_groin_base_r, loc=<0 0 0>, rot=<0 0 0 -1>}
 dyn_skirt_fr_r_base {parent=thigh_r, loc=<7.6722 -9.2159 -0.2959>, rot=<-0.0398 -0.997 0.0623 -0.0243>}
 dyn_skirt_fr_r_01 {parent=dyn_skirt_fr_r_base, loc=<0 0 0>, rot=<0 0 0 -1>}
 dyn_skirt_fr_strap_support {parent=dyn_skirt_fr_r_base, loc=<-2.4556 -1.3745 -0.0182>, rot=<0 0 0 1>}
 dyn_skirt_arrow_07_base {parent=dyn_skirt_fr_r_base, loc=<-4.8989 -0.2546 -0.0641>, rot=<-0.0561 -0.2054 0.9765 -0.0332>}
 dyn_skirt_arrow_07_01 {parent=dyn_skirt_arrow_07_base, loc=<0 0 0>, rot=<0 0 0 1>}
 dyn_skirt_arrow_06_base {parent=thigh_r, loc=<12.0231 -10.2469 -2.3459>, rot=<0.9967 -0.0392 0.0338 0.0627>}
 dyn_skirt_arrow_06_01 {parent=dyn_skirt_arrow_06_base, loc=<0 0 0>, rot=<0 0 0 1>}
 dyn_skirt_arrow_06_02 {parent=dyn_skirt_arrow_06_01, loc=<6.6538 0 0>, rot=<0 0 0 1>}
 deform_glute_base_l {parent=pelvis, loc=<-7.3466 -9.0213 -9.1245>, rot=<0.0172 0.031 -0.4477 -0.8935>}
 deform_glute_l {parent=deform_glute_base_l, loc=<0 0 0>, rot=<0 0 0 1>}
 deform_glute_base_r {parent=pelvis, loc=<-7.3464 -9.0213 9.1245>, rot=<0.031 -0.0172 0.8935 -0.4477>}
 deform_glute_r {parent=deform_glute_base_r, loc=<0 0 0>, rot=<0 0 0 1>}
 dyn_skirt_fr_l_base {parent=pelvis, loc=<-4.817 8.3281 -10.3591>, rot=<-0.0684 -0.981 0.166 -0.0729>}
 dyn_skirt_fr_l_01 {parent=dyn_skirt_fr_l_base, loc=<0 0 0>, rot=<0 0 0 1>}
 dyn_skirt_fr_l_03 {parent=dyn_skirt_fr_l_01, loc=<1.4894 0 0>, rot=<0 0 0 1>}
 dyn_skirt_fr_l_04 {parent=dyn_skirt_fr_l_03, loc=<11.8248 0 0>, rot=<0 0 0 1>}
 dyn_skirt_bow_r_base {parent=dyn_skirt_fr_l_base, loc=<0.1727 0.9119 2.611>, rot=<-0.2334 0.1139 -0.063 -0.9636>}
 dyn_skirt_bow_r_01 {parent=dyn_skirt_bow_r_base, loc=<0 0 0>, rot=<0 0 0 1>}
 dyn_skirt_bow_r_02 {parent=dyn_skirt_bow_r_01, loc=<7.5339 0 0>, rot=<0 0 0 1>}
 dyn_skirt_bow_l_base {parent=dyn_skirt_fr_l_base, loc=<-1.2139 1.6188 -2.5812>, rot=<0.0049 -0.0672 0.0126 -0.9976>}
 dyn_skirt_bow_l_01 {parent=dyn_skirt_bow_l_base, loc=<0 0 0>, rot=<0 0 0 1>}
 dyn_skirt_bow_l_02 {parent=dyn_skirt_bow_l_01, loc=<7.5339 0 0>, rot=<0 0 0 1>}
 dyn_skirt_bk_l_base {parent=pelvis, loc=<-7.2014 -12.5389 -9.4069>, rot=<0.0219 -0.9793 -0.1899 -0.0668>}
 dyn_skirt_bk_l_01 {parent=dyn_skirt_bk_l_base, loc=<0 0 0>, rot=<0 0 0 1>}
 dyn_skirt_bk_l_02 {parent=dyn_skirt_bk_l_01, loc=<10.1001 0 0>, rot=<0 0 0 1>}
 dyn_skirt_bk_l_03 {parent=dyn_skirt_bk_l_02, loc=<10.6954 0 0>, rot=<0 0 0 1>}
 dyn_skirt_bk_mid_base {parent=pelvis, loc=<-9.5621 -13.5309 1.5593>, rot=<0.0468 -0.9987 0.0131 0.0154>}
 dyn_skirt_bk_mid_01 {parent=dyn_skirt_bk_mid_base, loc=<0 0 0>, rot=<0 0 0 1>}
 dyn_skirt_bk_mid_02 {parent=dyn_skirt_bk_mid_01, loc=<9.8397 0 0>, rot=<0 0 0 1>}
 dyn_skirt_bk_mid_03 {parent=dyn_skirt_bk_mid_02, loc=<9.7004 0 0>, rot=<0 0 0 1>}
 dyn_skirt_bk_mid_04 {parent=dyn_skirt_bk_mid_03, loc=<9.268 0 0>, rot=<0 0 0 1>}
 dyn_skirt_bk_mid_05 {parent=dyn_skirt_bk_mid_04, loc=<9.068 0 0>, rot=<0 0 0 1>}
 dyn_skirt_bk_07 {parent=dyn_skirt_bk_mid_02, loc=<4.0415 2.271 9.357>, rot=<0.1603 0.0456 -0.0303 -0.9855>}
 dyn_skirt_bk_06 {parent=dyn_skirt_bk_mid_02, loc=<8.3244 2.5655 5.6748>, rot=<0.0129 0.0156 -0.0319 -0.9993>}
 dyn_skirt_bk_09 {parent=dyn_skirt_bk_mid_02, loc=<4.9257 1.7612 -0.3042>, rot=<-0.0264 0.0098 -0.0326 -0.9991>}
 dyn_skirt_bk_10 {parent=dyn_skirt_bk_mid_02, loc=<4.495 1.9075 -7.3714>, rot=<-0.0258 -0.0081 -0.0331 -0.9991>}
 dyn_skirt_bk_11 {parent=dyn_skirt_bk_mid_02, loc=<-1.1822 0.9614 -8.8711>, rot=<-0.2381 -0.1389 0.0202 -0.961>}
 dyn_skirt_side_l_frnt_base {parent=pelvis, loc=<-22.026 4.2457 -19.0975>, rot=<-0.0112 -0.9829 -0.1477 -0.1098>}
 dyn_skirt_side_l_frnt_01 {parent=dyn_skirt_side_l_frnt_base, loc=<0 0 0>, rot=<0 0 0 1>}
 dyn_skirt_side_l_frnt_02 {parent=dyn_skirt_side_l_frnt_01, loc=<9.5765 0 0>, rot=<0 0 0 1>}
 dyn_skirt_side_l_frnt_03 {parent=dyn_skirt_side_l_frnt_02, loc=<8.8718 0 0>, rot=<0 0 0 1>}
 dyn_skirt_side_l_bk_base {parent=pelvis, loc=<-22.4108 -6.307 -16.4371>, rot=<-0.0152 -0.9572 0.2701 -0.1025>}
 dyn_skirt_side_l_bk_01 {parent=dyn_skirt_side_l_bk_base, loc=<0 0 0>, rot=<0 0 0 1>}
 dyn_skirt_side_l_bk_02 {parent=dyn_skirt_side_l_bk_01, loc=<9.836 0 0>, rot=<0 0 0 1>}
 dyn_skirt_side_l_bk_03 {parent=dyn_skirt_side_l_bk_02, loc=<10.4007 0 0>, rot=<0 0 0 1>}
 dyn_skirt_fr_mid_base {parent=pelvis, loc=<-15.9966 10.5425 -1.5795>, rot=<-0.0016 -0.0009 0.9987 0.051>}
 dyn_skirt_fr_mid_01 {parent=dyn_skirt_fr_mid_base, loc=<0 0 0>, rot=<0 0 0 1>}
 dyn_skirt_arrow_05_base {parent=pelvis, loc=<-9.3497 9.7628 3.8986>, rot=<-0.0658 -0.9967 0.0013 0.048>}
 dyn_skirt_arrow_05_01 {parent=dyn_skirt_arrow_05_base, loc=<0 0 0>, rot=<0 0 0 1>}
 dyn_skirt_arrow_05_02 {parent=dyn_skirt_arrow_05_01, loc=<7.8977 0 0>, rot=<0 0 0 1>}
 dyn_skirt_arrow_04_base {parent=pelvis, loc=<-8.0156 9.5562 0.5298>, rot=<-0.0657 -0.9972 0.0022 0.0343>}
 dyn_skirt_arrow_04_01 {parent=dyn_skirt_arrow_04_base, loc=<0 0 0>, rot=<0 0 0 1>}
 dyn_skirt_arrow_04_02 {parent=dyn_skirt_arrow_04_01, loc=<9.9058 0 0>, rot=<0 0 0 1>}
 dyn_skirt_arrow_03_base {parent=pelvis, loc=<-7.659 9.3462 -2.3551>, rot=<-0.0586 -0.9978 0.0288 -0.0084>}
 dyn_skirt_arrow_03_01 {parent=dyn_skirt_arrow_03_base, loc=<0 0 0>, rot=<0 0 0 1>}
 dyn_skirt_arrow_03_02 {parent=dyn_skirt_arrow_03_01, loc=<10.5105 0 0>, rot=<0 0 0 1>}
 dyn_skirt_arrow_02_base {parent=pelvis, loc=<-7.0378 9.2768 -5.0479>, rot=<-0.0623 -0.996 0.0525 -0.0354>}
 dyn_skirt_arrow_02_01 {parent=dyn_skirt_arrow_02_base, loc=<0 0 0>, rot=<0 0 0 1>}
 dyn_skirt_arrow_02_02 {parent=dyn_skirt_arrow_02_01, loc=<11.0199 0 0>, rot=<0 0 0 1>}
 skirt_bk_ALT_support {parent=pelvis, loc=<-37.7437 -16.4915 2.5488>, rot=<0.0394 -0.9992 0 0>}
]
variables [
 Crouching ON {type=double, default=0}
 Is Backward Forward Movement {type=double, default=0}
 Is Backward Movement {type=double, default=0}
 Is Forward Movement {type=double, default=0}
 Is Left Movement {type=double, default=0}
 Is Right Left Movement {type=double, default=0}
 Is Right Movement {type=double, default=0}
 Jumping ON {type=double, default=0}
 KneeSliding ON {type=double, default=0}
]
