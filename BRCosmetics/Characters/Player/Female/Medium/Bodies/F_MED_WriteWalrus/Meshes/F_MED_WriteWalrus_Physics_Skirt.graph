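# Physics asset for the F_MED_WriteWalrus character's skirt, simulating its back and left-side panels, bows and arrow pieces.
# Capsule chains hang from the pelvis with tight swing and twist limits, except the free-rotating arrow chains, one of which attaches to the left thigh.
# Free-angular cross-links tie neighboring back chains and the two bow tips together.
# Pelvis and leg capsules provide body shapes, plus an oversized root sphere-like capsule below, likely approximating ground.
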PhysicsAsset /BRCosmetics/Characters/Player/Female/Medium/Bodies/F_MED_WriteWalrus/Meshes/F_MED_WriteWalrus_Physics_Skirt
bodies {
 calf_l [
  {shapeType=Capsule, length=43.0375, radius=7.2998, shapeName=calf_l_capsule, center=<-22.5307 -2.2827 0>, rotation=<rot 90 0 0>}
  {shapeType=Capsule, length=0.0011, radius=9.1411, shapeName=calf_l_capsule, center=<-36.3077 -4.136 -0.7561>}
 ]
 calf_r [
  {shapeType=Capsule, length=43.0375, radius=7.2998, shapeName=calf_r_capsule, center=<22.5307 2.2827 0>, rotation=<rot 90 180 0>}
  {shapeType=Capsule, length=0.0011, radius=9.1411, shapeName=calf_r_capsule, center=<36.3077 4.136 0.7561>, rotation=<rot 0 0 -180>}
 ]
 dyn_skirt_arrow_01_01 [{shapeType=Capsule, length=2.3905, radius=3.8534, shapeName=dyn_skirt_arrow_01_01_capsule, center=<3.7565 -2.3513 0>}]
 dyn_skirt_arrow_01_02 [{shapeType=Capsule, length=7.359, radius=4.5546, shapeName=dyn_skirt_arrow_01_02_capsule, center=<0 -1.6756 0>}]
 dyn_skirt_arrow_01_base [{shapeType=Capsule, length=0.505, radius=0.505, shapeName=dyn_skirt_arrow_01_base_capsule}]
 dyn_skirt_arrow_02_01 [{shapeType=Capsule, length=4.8326, radius=3.8523, shapeName=dyn_skirt_arrow_02_01_capsule, center=<3.7565 -2.3513 0>}]
 dyn_skirt_arrow_02_02 [{shapeType=Capsule, length=6.7207, radius=4.3326, shapeName=dyn_skirt_arrow_02_02_capsule, center=<0 -1.6756 0>}]
 dyn_skirt_arrow_02_base [{shapeType=Capsule, length=0.505, radius=0.505, shapeName=dyn_skirt_arrow_02_base_capsule}]
 dyn_skirt_bk_l_01 [{shapeType=Capsule, length=6.0987, radius=3.4666, shapeName=dyn_skirt_bk_l_01_capsule, center=<8.1361 0 0>, rotation=<rot 90 0 0>}]
 dyn_skirt_bk_l_02 [{shapeType=Capsule, length=7.9927, radius=4.3661, shapeName=dyn_skirt_bk_l_02_capsule, center=<6.9556 0 0>, rotation=<rot 90 0 0>}]
 dyn_skirt_bk_l_03 [{shapeType=Capsule, length=7.888, radius=4.9411, shapeName=dyn_skirt_bk_l_03_capsule, center=<6.9556 0 0>, rotation=<rot 90 0 0>}]
 dyn_skirt_bk_l_base [{shapeType=Capsule, length=0, radius=0, shapeName=dyn_skirt_bk_l_base_capsule}]
 dyn_skirt_bk_mid_01 [{shapeType=Capsule, length=5.7301, radius=6, shapeName=dyn_skirt_bk_mid_01_capsule, center=<6.7851 0 0>}]
 dyn_skirt_bk_mid_02 [{shapeType=Capsule, length=7.9829, radius=6, shapeName=dyn_skirt_bk_mid_02_capsule, center=<6.7851 0 0>}]
 dyn_skirt_bk_mid_03 [{shapeType=Capsule, length=10, radius=6, shapeName=dyn_skirt_bk_mid_03_capsule, center=<6.7851 0 0>}]
 dyn_skirt_bk_mid_04 [{shapeType=Capsule, length=10, radius=6, shapeName=dyn_skirt_bk_mid_04_capsule, center=<6.7851 0 0>}]
 dyn_skirt_bk_mid_05 [{shapeType=Capsule, length=8.9082, radius=6.7048, shapeName=dyn_skirt_bk_mid_05_capsule, center=<6.7851 0 0>}]
 dyn_skirt_bk_mid_base [{shapeType=Capsule, length=0, radius=0, shapeName=dyn_skirt_bk_mid_base_capsule}]
 dyn_skirt_bow_l_01 [{shapeType=Capsule, length=5.4567, radius=3.2483, shapeName=dyn_skirt_bow_l_01_capsule, center=<6.0644 1.7397 0>, rotation=<rot 90 0 0>}]
 dyn_skirt_bow_l_02 [{
   shapeType=Capsule
   length=8.1066
   radius=3.7781
   shapeName=dyn_skirt_bow_l_02_capsule
   center=<7.4604 1.9605 -0.5602>
   rotation=<rot 85.7054 0.0012 -0.001>
  }]
 dyn_skirt_bow_l_base [{shapeType=Capsule, length=0.505, radius=0.505, shapeName=dyn_skirt_bow_l_base_capsule}]
 dyn_skirt_bow_r_01 [{shapeType=Capsule, length=2.645, radius=3.9009, shapeName=dyn_skirt_bow_r_01_capsule, center=<5.1163 2.0973 0>, rotation=<rot 90 0 0>}]
 dyn_skirt_bow_r_02 [{shapeType=Capsule, length=7.0569, radius=5.3097, shapeName=dyn_skirt_bow_r_02_capsule, center=<5.9921 0.2457 0>, rotation=<rot 90 0 0>}]
 dyn_skirt_bow_r_base [{shapeType=Capsule, length=0.505, radius=0.505, shapeName=dyn_skirt_bow_r_base_capsule}]
 dyn_skirt_side_l_bk_01 [{shapeType=Capsule, length=6.5041, radius=5.0587, shapeName=dyn_skirt_side_l_bk_01_capsule, center=<5.2285 0 3.1358>, rotation=<rot 0 0 90>}]
 dyn_skirt_side_l_bk_02 [{shapeType=Capsule, length=7.769, radius=5.6411, shapeName=dyn_skirt_side_l_bk_02_capsule, center=<5.2285 0 3.7635>, rotation=<rot 0 0 90>}]
 dyn_skirt_side_l_bk_03 [{
   shapeType=Capsule
   length=8.2056
   radius=6.7259
   shapeName=dyn_skirt_side_l_bk_03_capsule
   center=<7.4422 0 2.7771>
   rotation=<rot 0 20.8826 90>
  }]
 dyn_skirt_side_l_bk_base [{shapeType=Capsule, length=0.505, radius=0.505, shapeName=dyn_skirt_side_l_bk_base_capsule}]
 dyn_skirt_side_l_frnt_01 [{shapeType=Capsule, length=7.2969, radius=4, shapeName=dyn_skirt_side_l_frnt_01_capsule, center=<5.689 0 1.2152>, rotation=<rot 0 0 90>}]
 dyn_skirt_side_l_frnt_02 [{shapeType=Capsule, length=7.085, radius=5.2016, shapeName=dyn_skirt_side_l_frnt_02_capsule, center=<4.6608 0 1.2152>, rotation=<rot 0 0 90>}]
 dyn_skirt_side_l_frnt_03 [{shapeType=Capsule, length=7.5183, radius=6.3325, shapeName=dyn_skirt_side_l_frnt_03_capsule, center=<4.6608 0 1.2152>, rotation=<rot 0 0 90>}]
 dyn_skirt_side_l_frnt_base [{shapeType=Capsule, length=0.505, radius=0.505, shapeName=dyn_skirt_side_l_frnt_base_capsule}]
 foot_l [{shapeType=Capsule, length=12.8663, radius=8.0527, shapeName=foot_l_capsule, center=<-4.6126 6.0045 0>, rotation=<rot 0 0 90>}]
 foot_r [{shapeType=Capsule, length=12.8663, radius=8.0527, shapeName=foot_r_capsule, center=<4.6126 -6.0045 0>, rotation=<rot 0 0 -90>}]
 pelvis [{shapeType=Capsule, length=12.0618, radius=10.1482, shapeName=pelvis_capsule, center=<-8.3156 0 0>}]
 root [{shapeType=Capsule, length=0, radius=138.7537, shapeName=root_capsule, center=<0 -17.8998 -139.9283>, rotation=<rot 0 -90 0>}]
 thigh_l [{shapeType=Capsule, length=33.1742, radius=7.9562, shapeName=thigh_l_capsule, center=<-23.8193 -1.1009 0>, rotation=<rot 90 0 0>}]
 thigh_r [{shapeType=Capsule, length=33.1742, radius=7.9562, shapeName=thigh_r_capsule, center=<23.8193 1.1009 0>, rotation=<rot 90 180 0>}]
}
constraints [
 {
  bone1Name=dyn_skirt_bk_mid_base
  bone2Name=pelvis
  swing1LimitDegrees=45
  swing1Motion=Free
  swing2LimitDegrees=45
  swing2Motion=Free
  twistLimitDegrees=45
  twistMotion=Free
 }
 {bone1Name=dyn_skirt_bk_mid_01, bone2Name=dyn_skirt_bk_mid_base, swing1LimitDegrees=30, swing2LimitDegrees=10, twistLimitDegrees=0}
 {bone1Name=dyn_skirt_bk_mid_02, bone2Name=dyn_skirt_bk_mid_01, swing1LimitDegrees=20, swing2LimitDegrees=10, twistLimitDegrees=0}
 {bone1Name=dyn_skirt_bk_mid_03, bone2Name=dyn_skirt_bk_mid_02, swing1LimitDegrees=20, swing2LimitDegrees=20, twistLimitDegrees=5}
 {bone1Name=dyn_skirt_bk_mid_04, bone2Name=dyn_skirt_bk_mid_03, swing1LimitDegrees=20, swing2LimitDegrees=20, twistLimitDegrees=2}
 {bone1Name=dyn_skirt_bk_mid_05, bone2Name=dyn_skirt_bk_mid_04, swing1LimitDegrees=30, swing2LimitDegrees=15, twistLimitDegrees=5}
 {
  bone1Name=dyn_skirt_bk_l_base
  bone2Name=pelvis
  swing1LimitDegrees=45
  swing1Motion=Free
  swing2LimitDegrees=45
  swing2Motion=Free
  twistLimitDegrees=45
  twistMotion=Free
 }
 {bone1Name=dyn_skirt_bk_l_01, bone2Name=dyn_skirt_bk_l_base, swing1LimitDegrees=20, swing2LimitDegrees=10, twistLimitDegrees=0}
 {bone1Name=dyn_skirt_bk_l_02, bone2Name=dyn_skirt_bk_l_01, swing1LimitDegrees=20, swing2LimitDegrees=30, twistLimitDegrees=10}
 {bone1Name=dyn_skirt_bk_l_03, bone2Name=dyn_skirt_bk_l_02, swing1LimitDegrees=30, swing2LimitDegrees=30, twistLimitDegrees=10}
 {
  bone1Name=thigh_l
  bone2Name=pelvis
  swing1LimitDegrees=45
  swing1Motion=Free
  swing2LimitDegrees=45
  swing2Motion=Free
  twistLimitDegrees=45
  twistMotion=Free
 }
 {
  bone1Name=calf_l
  bone2Name=thigh_l
  swing1LimitDegrees=45
  swing1Motion=Free
  swing2LimitDegrees=45
  swing2Motion=Free
  twistLimitDegrees=45
  twistMotion=Free
 }
 {
  bone1Name=foot_l
  bone2Name=calf_l
  swing1LimitDegrees=45
  swing1Motion=Free
  swing2LimitDegrees=45
  swing2Motion=Free
  twistLimitDegrees=45
  twistMotion=Free
 }
 {
  bone1Name=thigh_r
  bone2Name=pelvis
  swing1LimitDegrees=45
  swing1Motion=Free
  swing2LimitDegrees=45
  swing2Motion=Free
  twistLimitDegrees=45
  twistMotion=Free
 }
 {
  bone1Name=calf_r
  bone2Name=thigh_r
  swing1LimitDegrees=45
  swing1Motion=Free
  swing2LimitDegrees=45
  swing2Motion=Free
  twistLimitDegrees=45
  twistMotion=Free
 }
 {
  bone1Name=foot_r
  bone2Name=calf_r
  swing1LimitDegrees=45
  swing1Motion=Free
  swing2LimitDegrees=45
  swing2Motion=Free
  twistLimitDegrees=45
  twistMotion=Free
 }
 {
  bone1Name=dyn_skirt_bk_l_02
  bone2Name=dyn_skirt_bk_mid_02
  swing1LimitDegrees=45
  swing1Motion=Free
  swing2LimitDegrees=45
  swing2Motion=Free
  twistLimitDegrees=45
  twistMotion=Free
 }
 {
  bone1Name=pelvis
  bone2Name=root
  swing1LimitDegrees=45
  swing1Motion=Free
  swing2LimitDegrees=45
  swing2Motion=Free
  twistLimitDegrees=45
  twistMotion=Free
 }
 {
  bone1Name=dyn_skirt_side_l_frnt_base
  bone2Name=pelvis
  swing1LimitDegrees=45
  swing1Motion=Free
  swing2LimitDegrees=45
  swing2Motion=Free
  twistLimitDegrees=45
  twistMotion=Free
 }
 {bone1Name=dyn_skirt_side_l_frnt_01, bone2Name=dyn_skirt_side_l_frnt_base, swing1LimitDegrees=10, swing2LimitDegrees=10, twistLimitDegrees=0}
 {bone1Name=dyn_skirt_side_l_frnt_02, bone2Name=dyn_skirt_side_l_frnt_01, swing1LimitDegrees=10, swing2LimitDegrees=10, twistLimitDegrees=0}
 {bone1Name=dyn_skirt_side_l_frnt_03, bone2Name=dyn_skirt_side_l_frnt_02, swing1LimitDegrees=20, swing2LimitDegrees=20, twistLimitDegrees=5}
 {
  bone1Name=dyn_skirt_side_l_bk_base
  bone2Name=pelvis
  swing1LimitDegrees=45
  swing1Motion=Free
  swing2LimitDegrees=45
  swing2Motion=Free
  twistLimitDegrees=45
  twistMotion=Free
 }
 {bone1Name=dyn_skirt_side_l_bk_01, bone2Name=dyn_skirt_side_l_bk_base, swing1LimitDegrees=10, swing2LimitDegrees=10, twistLimitDegrees=0}
 {bone1Name=dyn_skirt_side_l_bk_02, bone2Name=dyn_skirt_side_l_bk_01, swing1LimitDegrees=10, swing2LimitDegrees=10, twistLimitDegrees=0}
 {bone1Name=dyn_skirt_side_l_bk_03, bone2Name=dyn_skirt_side_l_bk_02, swing1LimitDegrees=20, swing2LimitDegrees=20, twistLimitDegrees=5}
 {
  bone1Name=dyn_skirt_bow_l_base
  bone2Name=pelvis
  swing1LimitDegrees=45
  swing1Motion=Free
  swing2LimitDegrees=45
  swing2Motion=Free
  twistLimitDegrees=45
  twistMotion=Free
 }
 {bone1Name=dyn_skirt_bow_l_01, bone2Name=dyn_skirt_bow_l_base, swing1LimitDegrees=10, swing2LimitDegrees=10, twistLimitDegrees=0}
 {bone1Name=dyn_skirt_bow_l_02, bone2Name=dyn_skirt_bow_l_01, swing1LimitDegrees=10, swing2LimitDegrees=10, twistLimitDegrees=5}
 {
  bone1Name=dyn_skirt_bow_r_base
  bone2Name=pelvis
  swing1LimitDegrees=45
  swing1Motion=Free
  swing2LimitDegrees=45
  swing2Motion=Free
  twistLimitDegrees=45
  twistMotion=Free
 }
 {bone1Name=dyn_skirt_bow_r_01, bone2Name=dyn_skirt_bow_r_base, swing1LimitDegrees=20, swing2LimitDegrees=20, twistLimitDegrees=0}
 {bone1Name=dyn_skirt_bow_r_02, bone2Name=dyn_skirt_bow_r_01, swing1LimitDegrees=20, swing2LimitDegrees=10, twistLimitDegrees=5}
 {
  bone1Name=dyn_skirt_bow_l_02
  bone2Name=dyn_skirt_bow_r_02
  swing1LimitDegrees=45
  swing1Motion=Free
  swing2LimitDegrees=45
  swing2Motion=Free
  twistLimitDegrees=45
  twistMotion=Free
 }
 {
  bone1Name=dyn_skirt_arrow_02_base
  bone2Name=pelvis
  swing1LimitDegrees=45
  swing1Motion=Free
  swing2LimitDegrees=45
  swing2Motion=Free
  twistLimitDegrees=45
  twistMotion=Free
 }
 {
  bone1Name=dyn_skirt_arrow_02_01
  bone2Name=dyn_skirt_arrow_02_base
  swing1LimitDegrees=45
  swing1Motion=Free
  swing2LimitDegrees=45
  swing2Motion=Free
  twistLimitDegrees=45
  twistMotion=Free
 }
 {
  bone1Name=dyn_skirt_arrow_02_02
  bone2Name=dyn_skirt_arrow_02_01
  swing1LimitDegrees=45
  swing1Motion=Free
  swing2LimitDegrees=45
  swing2Motion=Free
  twistLimitDegrees=45
  twistMotion=Free
 }
 {
  bone1Name=dyn_skirt_arrow_01_base
  bone2Name=thigh_l
  swing1LimitDegrees=45
  swing1Motion=Free
  swing2LimitDegrees=45
  swing2Motion=Free
  twistLimitDegrees=45
  twistMotion=Free
 }
 {
  bone1Name=dyn_skirt_arrow_01_01
  bone2Name=dyn_skirt_arrow_01_base
  swing1LimitDegrees=45
  swing1Motion=Free
  swing2LimitDegrees=45
  swing2Motion=Free
  twistLimitDegrees=45
  twistMotion=Free
 }
 {
  bone1Name=dyn_skirt_arrow_01_02
  bone2Name=dyn_skirt_arrow_01_01
  swing1LimitDegrees=45
  swing1Motion=Free
  swing2LimitDegrees=45
  swing2Motion=Free
  twistLimitDegrees=45
  twistMotion=Free
 }
]
