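# Animation Blueprint for the F_MED_WriteWalrus female body cosmetic, adding secondary motion to a pose copied from the attached parent mesh.
# Weighted multi-bone constraints place strap and skirt-arrow base bones.
# A movement-driven Control Rig feeds a skirt RigidBody simulation and AnimDynamics on skirt bones, cached for the final chain.
# That chain layers angular and linear spring passes on torso bones, then simulates boot fur, shoulders, leg strap and necklace.
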AnimBlueprint /BRCosmetics/Characters/Player/Female/Medium/Bodies/F_MED_WriteWalrus/Meshes/F_MED_WriteWalrus_AnimBP
target_skeleton=&/F_MED_WriteWalrus_Skeleton
graphs [
 AnimGraph {nodes [
   n1 = "Transform (Modify) Bone - Bone: dyn_torso_l"(ComponentPose=Local To Component(LocalPose=Use cached pose 'toTorsoDyn'()), Alpha=1, Translation="0,0,1", Rotation="0,0,0", Scale="1,1,1") {translationMode=BMM_Additive, translationSpace=BCS_BoneSpace, boneToModify {boneName=dyn_torso_l}, translation {z=1}}
   n2 = "Transform (Modify) Bone - Bone: dyn_torso_r"(ComponentPose=@n1, Alpha=1, Translation="0,0,1", Rotation="0,0,0", Scale="1,1,1") {translationMode=BMM_Additive, translationSpace=BCS_BoneSpace, boneToModify {boneName=dyn_torso_r}, translation {z=1}}
   n3 = Get Sim Space Settings() -> (Sim Space Settings:Anim Phys Sim Space Settings Structure)
   n4 = Anim Dynamics - Bone: dyn_torso_r!(ComponentPose=@n2, Alpha=1, AngularSpringConstant:float=5000, SimSpaceSettings:Anim Phys Sim Space Settings Structure=@n3) -> (Pose) {
    angularBiasOverride=1
    angularDampingOverride=0.9999
    angularSpringConstant=5000
    bAngularSpring=true
    bGravityOverrideInSimSpace=true
    bOverrideAngularBias=true
    bOverrideAngularDamping=true
    bUseGravityOverride=true
    linearDampingOverride=0.9999
    linearSpringConstant=100
    simulationSpace=RootRelative
    boundBone {boneName=dyn_torso_r}
    componentAppliedLinearAccClamp=<2 2 2>
    gravityOverride {z=-980}
    physicsBodyDefinitions [{
      boundBone {boneName=dyn_torso_r}
      boxExtents=<21 21 21>
      constraintSetup {angularLimitsMax=<0 1 5>, angularLimitsMin=<0 -5 -5>, angularTarget=<1 0 0>}
      localJointOffset=<2 0 0>
     }]
    relativeSpaceBone {boneName=spine_04}
    simSpaceSettings {maxAngularAcceleration=45, maxAngularVelocity=8, simSpaceAngularAlpha=0.25}
   }
   n5 = Anim Dynamics - Bone: dyn_torso_l!(ComponentPose=@n4, Alpha=1, AngularSpringConstant:float=5000, SimSpaceSettings:Anim Phys Sim Space Settings Structure=@n3) -> (Pose) {
    angularBiasOverride=1
    angularDampingOverride=0.9999
    angularSpringConstant=5000
    bAngularSpring=true
    bGravityOverrideInSimSpace=true
    bOverrideAngularBias=true
    bOverrideAngularDamping=true
    bUseGravityOverride=true
    linearDampingOverride=0.9999
    linearSpringConstant=100
    simulationSpace=RootRelative
    boundBone {boneName=dyn_torso_l}
    componentAppliedLinearAccClamp=<2 2 2>
    gravityOverride {z=-980}
    physicsBodyDefinitions [{
      boundBone {boneName=dyn_torso_l}
      boxExtents=<20 20 20>
      constraintSetup {angularLimitsMax=<0 1 5>, angularLimitsMin=<0 -5 -5>, angularTarget=<1 0 0>}
      localJointOffset=<2 0 0>
     }]
    relativeSpaceBone {boneName=spine_04}
    simSpaceSettings {maxAngularAcceleration=45, maxAngularVelocity=8, simSpaceAngularAlpha=0.25}
   }
   n6 = Anim Dynamics - Bone: dyn_torso_r!(ComponentPose=@n5, Alpha=1, GravityOverride:Vector="0,0,-200", LinearSpringConstant:float=180) -> (Pose) {
    angularBiasOverride=1
    angularDampingOverride=0.99
    angularSpringConstant=10000
    bGravityOverrideInSimSpace=true
    bLinearSpring=true
    bOverrideAngularBias=true
    bOverrideLinearDamping=true
    bUseGravityOverride=true
    linearDampingOverride=1
    linearSpringConstant=180
    simulationSpace=RootRelative
    boundBone {boneName=dyn_torso_r}
    componentAppliedLinearAccClamp=<5 5 5>
    gravityOverride {z=-200}
    physicsBodyDefinitions [{
      boundBone {boneName=dyn_torso_r}
      boxExtents=<80 80 80>
      constraintSetup {angularLimitsMax=<0 0 0>, angularLimitsMin=<0 0 0>, angularTarget=<1 0 0>, linearAxesMax=<0 0.5 1>, linearAxesMin=<0 -0.5 -0.5>}
      localJointOffset=<0.2 0 0>
     }]
    relativeSpaceBone {boneName=spine_04}
    simSpaceSettings {maxAngularAcceleration=50, maxAngularVelocity=10, simSpaceAngularAlpha=0.1}
   }
   n7 = Anim Dynamics - Bone: dyn_torso_l!(ComponentPose=@n6, Alpha=1, GravityOverride:Vector="0,0,-200", LinearSpringConstant:float=180) -> (Pose) {
    angularBiasOverride=1
    angularDampingOverride=0.99
    angularSpringConstant=10000
    bGravityOverrideInSimSpace=true
    bLinearSpring=true
    bOverrideAngularBias=true
    bOverrideLinearDamping=true
    bUseGravityOverride=true
    linearDampingOverride=1
    linearSpringConstant=180
    simulationSpace=RootRelative
    boundBone {boneName=dyn_torso_l}
    componentAppliedLinearAccClamp=<5 5 5>
    gravityOverride {z=-200}
    physicsBodyDefinitions [{
      boundBone {boneName=dyn_torso_l}
      boxExtents=<90 90 90>
      constraintSetup {angularLimitsMax=<0 0 0>, angularLimitsMin=<0 0 0>, angularTarget=<1 0 0>, linearAxesMax=<0 0.5 1>, linearAxesMin=<0 -0.5 -0.5>}
      localJointOffset=<0.2 0 0>
     }]
    relativeSpaceBone {boneName=spine_04}
    simSpaceSettings {maxAngularAcceleration=50, maxAngularVelocity=10, simSpaceAngularAlpha=0.1}
   }
   n8 = Anim Dynamics - Bone: dyn_boot_fur_l_01!(
    ComponentPose=@n7
    Alpha=1
    GravityOverride:Vector="0,0,0"
    SimSpaceSettings:Anim Phys Sim Space Settings Structure=(SimSpaceAngularAlpha=0,MaxAngularVelocity=10000,MaxAngularAcceleration=10000,ExternalAngularVelocity=(X=0,Y=0,Z=0))
    ExternalForce="0,0,0"
   ) -> (Pose) {
    angularDampingOverride=0.8
    bAngularSpring=true
    bOverrideAngularDamping=true
    bUseGravityOverride=true
    simulationSpace=RootRelative
    boundBone {boneName=dyn_boot_fur_l_01}
    componentAppliedLinearAccClamp=<3000 3000 3000>
    physicsBodyDefinitions [{
      boundBone {boneName=dyn_boot_fur_l_01}
      boxExtents=<10 10 10>
      constraintSetup {angularLimitsMax=<0 5 5>, angularLimitsMin=<0 -30 -5>, angularTarget=<1 1 1>}
      localJointOffset=<0.1 0 0>
     }]
   }
   n9 = Anim Dynamics - Bone: dyn_boot_fur_l_02!(
    ComponentPose=@n8
    Alpha=1
    GravityOverride:Vector="0,0,0"
    SimSpaceSettings:Anim Phys Sim Space Settings Structure=(SimSpaceAngularAlpha=0,MaxAngularVelocity=10000,MaxAngularAcceleration=10000,ExternalAngularVelocity=(X=0,Y=0,Z=0))
    ExternalForce="0,0,0"
   ) -> (Pose) {
    angularDampingOverride=0.8
    bAngularSpring=true
    bOverrideAngularDamping=true
    bUseGravityOverride=true
    simulationSpace=RootRelative
    boundBone {boneName=dyn_boot_fur_l_02}
    componentAppliedLinearAccClamp=<3000 3000 3000>
    physicsBodyDefinitions [{
      boundBone {boneName=dyn_boot_fur_l_02}
      boxExtents=<10 10 10>
      constraintSetup {angularLimitsMax=<0 5 5>, angularLimitsMin=<0 -30 -5>, angularTarget=<1 1 1>}
      localJointOffset=<0.1 0 0>
     }]
   }
   n10 = Anim Dynamics - Bone: dyn_boot_fur_l_03!(
    ComponentPose=@n9
    Alpha=1
    GravityOverride:Vector="0,0,0"
    SimSpaceSettings:Anim Phys Sim Space Settings Structure=(SimSpaceAngularAlpha=0,MaxAngularVelocity=10000,MaxAngularAcceleration=10000,ExternalAngularVelocity=(X=0,Y=0,Z=0))
    ExternalForce="0,0,0"
   ) -> (Pose) {
    angularDampingOverride=0.8
    bAngularSpring=true
    bOverrideAngularDamping=true
    bUseGravityOverride=true
    simulationSpace=RootRelative
    boundBone {boneName=dyn_boot_fur_l_03}
    componentAppliedLinearAccClamp=<3000 3000 3000>
    physicsBodyDefinitions [{
      boundBone {boneName=dyn_boot_fur_l_03}
      boxExtents=<10 10 10>
      constraintSetup {angularLimitsMax=<0 5 5>, angularLimitsMin=<0 -30 -5>, angularTarget=<1 1 1>}
      localJointOffset=<0.1 0 0>
     }]
   }
   n11 = Anim Dynamics - Bone: dyn_boot_fur_l_04!(
    ComponentPose=@n10
    Alpha=1
    GravityOverride:Vector="0,0,0"
    SimSpaceSettings:Anim Phys Sim Space Settings Structure=(SimSpaceAngularAlpha=0,MaxAngularVelocity=10000,MaxAngularAcceleration=10000,ExternalAngularVelocity=(X=0,Y=0,Z=0))
    ExternalForce="0,0,0"
   ) -> (Pose) {
    angularDampingOverride=0.8
    bAngularSpring=true
    bOverrideAngularDamping=true
    bUseGravityOverride=true
    simulationSpace=RootRelative
    boundBone {boneName=dyn_boot_fur_l_04}
    componentAppliedLinearAccClamp=<3000 3000 3000>
    physicsBodyDefinitions [{
      boundBone {boneName=dyn_boot_fur_l_04}
      boxExtents=<10 10 10>
      constraintSetup {angularLimitsMax=<0 15 5>, angularLimitsMin=<0 -30 -5>, angularTarget=<1 1 1>}
      localJointOffset=<0.1 0 0>
     }]
   }
   n12 = Anim Dynamics - Bone: dyn_boot_fur_r_01!(
    ComponentPose=@n11
    Alpha=1
    GravityOverride:Vector="0,0,0"
    SimSpaceSettings:Anim Phys Sim Space Settings Structure=(SimSpaceAngularAlpha=0,MaxAngularVelocity=10000,MaxAngularAcceleration=10000,ExternalAngularVelocity=(X=0,Y=0,Z=0))
    ExternalForce="0,0,0"
   ) -> (Pose) {
    angularDampingOverride=0.8
    bAngularSpring=true
    bOverrideAngularDamping=true
    bUseGravityOverride=true
    simulationSpace=RootRelative
    boundBone {boneName=dyn_boot_fur_r_01}
    componentAppliedLinearAccClamp=<3000 3000 3000>
    physicsBodyDefinitions [{
      boundBone {boneName=dyn_boot_fur_r_01}
      boxExtents=<10 10 10>
      constraintSetup {angularLimitsMax=<0 5 5>, angularLimitsMin=<0 -30 -5>, angularTarget=<1 1 1>}
      localJointOffset=<0.1 0 0>
     }]
   }
   n13 = Anim Dynamics - Bone: dyn_boot_fur_r_02!(
    ComponentPose=@n12
    Alpha=1
    GravityOverride:Vector="0,0,0"
    SimSpaceSettings:Anim Phys Sim Space Settings Structure=(SimSpaceAngularAlpha=0,MaxAngularVelocity=10000,MaxAngularAcceleration=10000,ExternalAngularVelocity=(X=0,Y=0,Z=0))
    ExternalForce="0,0,0"
   ) -> (Pose) {
    angularDampingOverride=0.8
    bAngularSpring=true
    bOverrideAngularDamping=true
    bUseGravityOverride=true
    simulationSpace=RootRelative
    boundBone {boneName=dyn_boot_fur_r_02}
    componentAppliedLinearAccClamp=<3000 3000 3000>
    physicsBodyDefinitions [{
      boundBone {boneName=dyn_boot_fur_r_02}
      boxExtents=<10 10 10>
      constraintSetup {angularLimitsMax=<0 10 5>, angularLimitsMin=<0 -30 -5>, angularTarget=<1 1 1>}
      localJointOffset=<0.1 0 0>
     }]
   }
   n14 = Anim Dynamics - Bone: dyn_boot_fur_r_03!(
    ComponentPose=@n13
    Alpha=1
    GravityOverride:Vector="0,0,0"
    SimSpaceSettings:Anim Phys Sim Space Settings Structure=(SimSpaceAngularAlpha=0,MaxAngularVelocity=10000,MaxAngularAcceleration=10000,ExternalAngularVelocity=(X=0,Y=0,Z=0))
    ExternalForce="0,0,0"
   ) -> (Pose) {
    angularDampingOverride=0.8
    bAngularSpring=true
    bOverrideAngularDamping=true
    bUseGravityOverride=true
    simulationSpace=RootRelative
    boundBone {boneName=dyn_boot_fur_r_03}
    componentAppliedLinearAccClamp=<3000 3000 3000>
    physicsBodyDefinitions [{
      boundBone {boneName=dyn_boot_fur_r_03}
      boxExtents=<10 10 10>
      constraintSetup {angularLimitsMax=<0 5 5>, angularLimitsMin=<0 -30 -5>, angularTarget=<1 1 1>}
      localJointOffset=<0.1 0 0>
     }]
   }
   n15 = Anim Dynamics - Bone: dyn_boot_fur_r_04!(
    ComponentPose=@n14
    Alpha=1
    GravityOverride:Vector="0,0,0"
    SimSpaceSettings:Anim Phys Sim Space Settings Structure=(SimSpaceAngularAlpha=0,MaxAngularVelocity=10000,MaxAngularAcceleration=10000,ExternalAngularVelocity=(X=0,Y=0,Z=0))
    ExternalForce="0,0,0"
   ) -> (Pose) {
    angularDampingOverride=0.8
    bAngularSpring=true
    bOverrideAngularDamping=true
    bUseGravityOverride=true
    simulationSpace=RootRelative
    boundBone {boneName=dyn_boot_fur_r_04}
    componentAppliedLinearAccClamp=<3000 3000 3000>
    physicsBodyDefinitions [{
      boundBone {boneName=dyn_boot_fur_r_04}
      boxExtents=<10 10 10>
      constraintSetup {angularLimitsMax=<0 15 5>, angularLimitsMin=<0 -10 -5>, angularTarget=<1 1 1>}
      localJointOffset=<0.1 0 0>
     }]
   }
   n16 = Anim Dynamics - Bone: dyn_shoulder_l_01!(
    ComponentPose=@n15
    Alpha=1
    GravityOverride:Vector="0,0,-980"
    SimSpaceSettings:Anim Phys Sim Space Settings Structure=(SimSpaceAngularAlpha=0,MaxAngularVelocity=10000,MaxAngularAcceleration=10000,ExternalAngularVelocity=(X=0,Y=0,Z=0))
    ExternalForce="0,0,0"
   ) -> (Pose) {
    angularDampingOverride=0.8
    bAngularSpring=true
    bOverrideAngularDamping=true
    bUseGravityOverride=true
    simulationSpace=RootRelative
    boundBone {boneName=dyn_shoulder_l_01}
    componentAppliedLinearAccClamp=<3000 3000 3000>
    gravityOverride {z=-980}
    physicsBodyDefinitions [{
      boundBone {boneName=dyn_shoulder_l_01}
      boxExtents=<10 10 10>
      constraintSetup {angularLimitsMax=<0 20 10>, angularLimitsMin=<0 -20 -10>, angularTarget=<1 1 1>}
      localJointOffset=<0.1 0 0>
     }]
   }
   n17 = Anim Dynamics - Bone: dyn_shoulder_l_02!(
    ComponentPose=@n16
    Alpha=1
    GravityOverride:Vector="0,0,-980"
    SimSpaceSettings:Anim Phys Sim Space Settings Structure=(SimSpaceAngularAlpha=0,MaxAngularVelocity=10000,MaxAngularAcceleration=10000,ExternalAngularVelocity=(X=0,Y=0,Z=0))
    ExternalForce="0,0,0"
   ) -> (Pose) {
    angularDampingOverride=0.8
    bAngularSpring=true
    bOverrideAngularDamping=true
    bUseGravityOverride=true
    simulationSpace=RootRelative
    boundBone {boneName=dyn_shoulder_l_02}
    componentAppliedLinearAccClamp=<3000 3000 3000>
    gravityOverride {z=-980}
    physicsBodyDefinitions [{
      boundBone {boneName=dyn_shoulder_l_02}
      boxExtents=<10 10 10>
      constraintSetup {angularLimitsMax=<0 5 30>, angularLimitsMin=<0 -5 0>, angularTarget=<1 1 1>}
      localJointOffset=<0.1 0 0>
     }]
   }
   n18 = Anim Dynamics - Bone: dyn_shoulder_l_03!(
    ComponentPose=@n17
    Alpha=1
    GravityOverride:Vector="0,0,-980"
    SimSpaceSettings:Anim Phys Sim Space Settings Structure=(SimSpaceAngularAlpha=0,MaxAngularVelocity=10000,MaxAngularAcceleration=10000,ExternalAngularVelocity=(X=0,Y=0,Z=0))
    ExternalForce="0,0,0"
   ) -> (Pose) {
    angularDampingOverride=0.8
    bAngularSpring=true
    bOverrideAngularDamping=true
    bUseGravityOverride=true
    simulationSpace=RootRelative
    boundBone {boneName=dyn_shoulder_l_03}
    componentAppliedLinearAccClamp=<3000 3000 3000>
    gravityOverride {z=-980}
    physicsBodyDefinitions [{
      boundBone {boneName=dyn_shoulder_l_03}
      boxExtents=<10 10 10>
      constraintSetup {angularLimitsMax=<0 5 5>, angularLimitsMin=<0 -5 -5>, angularTarget=<1 1 1>}
      localJointOffset=<0.1 0 0>
     }]
   }
   n19 = Anim Dynamics - Bone: dyn_leg_strap_r_01!(
    ComponentPose=@n18
    Alpha=1
    GravityOverride:Vector="0,0,-980"
    SimSpaceSettings:Anim Phys Sim Space Settings Structure=(SimSpaceAngularAlpha=0,MaxAngularVelocity=10000,MaxAngularAcceleration=10000,ExternalAngularVelocity=(X=0,Y=0,Z=0))
    ExternalForce="0,0,0"
   ) -> (Pose) {
    angularDampingOverride=0.7
    bAngularSpring=true
    bOverrideAngularDamping=true
    bUseGravityOverride=true
    simulationSpace=RootRelative
    boundBone {boneName=dyn_leg_strap_r_01}
    componentAppliedLinearAccClamp=<7000 7000 7000>
    gravityOverride {z=-980}
    physicsBodyDefinitions [{
      boundBone {boneName=dyn_leg_strap_r_01}
      boxExtents=<2 2 2>
      constraintSetup {angularLimitsMax=<10 20 20>, angularLimitsMin=<-10 -20 -20>, angularTarget=<1 0 0>}
      localJointOffset=<0.1 0 0>
     }]
   }
   n20 = Anim Dynamics - Bone: dyn_leg_strap_r_02!(
    ComponentPose=@n19
    Alpha=1
    GravityOverride:Vector="0,0,-980"
    SimSpaceSettings:Anim Phys Sim Space Settings Structure=(SimSpaceAngularAlpha=0,MaxAngularVelocity=10000,MaxAngularAcceleration=10000,ExternalAngularVelocity=(X=0,Y=0,Z=0))
    ExternalForce="0,0,0"
   ) -> (Pose) {
    angularDampingOverride=0.7
    bAngularSpring=true
    bOverrideAngularDamping=true
    bUseGravityOverride=true
    simulationSpace=RootRelative
    boundBone {boneName=dyn_leg_strap_r_02}
    componentAppliedLinearAccClamp=<7000 7000 7000>
    gravityOverride {z=-980}
    physicsBodyDefinitions [{
      boundBone {boneName=dyn_leg_strap_r_02}
      boxExtents=<2 2 2>
      constraintSetup {angularLimitsMax=<10 20 20>, angularLimitsMin=<-10 -20 -20>, angularTarget=<1 0 0>}
      localJointOffset=<0.1 0 0>
     }]
   }
   n21 = Anim Dynamics - Bone: dyn_necklace_bottom_01!(ComponentPose=@n20, Alpha=1, GravityOverride:Vector="0,0,-980", ExternalForce="0,0,0") -> (Pose) {
    angularDampingOverride=0.9999
    bOverrideAngularDamping=true
    bUseGravityOverride=true
    bUsePlanarLimit=false
    simulationSpace=RootRelative
    boundBone {boneName=dyn_necklace_bottom_01}
    componentAppliedLinearAccClamp=<2000 2000 2000>
    gravityOverride {z=-980}
    physicsBodyDefinitions [{
      boundBone {boneName=dyn_necklace_bottom_01}
      boxExtents=<2 2 2>
      constraintSetup {angularLimitsMax=<0 0 70>, angularLimitsMin=<0 -5 10>, angularTarget=<1 0 0>}
      localJointOffset=<0.1 0 0>
     }]
   }
   n22 = Anim Dynamics - Bone: dyn_necklace_bottom_02!(ComponentPose=@n21, Alpha=1, GravityOverride:Vector="0,0,-980", ExternalForce="0,0,0") -> (Pose) {
    angularDampingOverride=0.999
    bOverrideAngularDamping=true
    bUseGravityOverride=true
    simulationSpace=RootRelative
    boundBone {boneName=dyn_necklace_bottom_02}
    componentAppliedLinearAccClamp=<2000 2000 2000>
    gravityOverride {z=-980}
    physicsBodyDefinitions [{
      boundBone {boneName=dyn_necklace_bottom_02}
      boxExtents=<2 2 2>
      constraintSetup {angularLimitsMax=<0 10 70>, angularLimitsMin=<0 -10 0>, angularTarget=<1 0 0>}
      localJointOffset=<0.1 0 0>
     }]
   }
   Output Pose(Result=Component To Local(ComponentPose=@n22)) {layerGroup=None}
   n23 = Control Rig!(
    Source=Use cached pose 'toCTRLrig'()
    Alpha
    Is Right Left Movement:double=Get Is Right Left Movement() -> (Is Right Left Movement:double)
    Crouching ON:double=Get Crouching ON() -> (Crouching ON:double)
    Is Backward Forward Movement:double=Get Is Forward Backward Movement() -> (Is Forward Backward Movement:double)
    Jumping ON:double=Get Jumping ON() -> (Jumping ON:double)
    KneeSliding ON:double=Get KneeSliding ON() -> (KneeSliding ON:double)
   ) -> (Pose) {controlRigAssetReference {blueprintRigClass=&/F_MED_WriteWalrus_CtrlRig.F_MED_WriteWalrus_CtrlRig_C}}
   n24 = RigidBody!(
    ComponentPose=Local To Component(LocalPose=@n23)
    Alpha
    OverrideWorldGravity:Vector="0,0,0"
    ExternalForce
    SimSpaceSettings:Sim Space Settings Structure=(WorldAlpha=0,VelocityScaleZ=1,DampingAlpha=1,MaxLinearVelocity=10000,MaxAngularVelocity=10000,MaxLinearAcceleration=10000,MaxAngularAcceleration=10000,ExternalLinearDragV=(X=0,Y=0,Z=0),ExternalLinearVelocity=(X=0,Y=0,Z=0),ExternalAngularVelocity=(X=0,Y=0,Z=0))
   ) -> (Pose) {
    bClampLinearTranslationLimitToRefPose=true
    bOverrideWorldGravity=true
    overridePhysicsAsset=&/F_MED_WriteWalrus_Physics_Skirt
    simulationSpace=BaseBoneSpace
    baseBoneRef {boneName=pelvis}
   }
   n25 = Anim Dynamics - Bone: dyn_skirt_side_l!(
    ComponentPose=@n24
    Alpha=1
    GravityOverride:Vector="0,0,-980"
    SimSpaceSettings:Anim Phys Sim Space Settings Structure=(SimSpaceAngularAlpha=0,MaxAngularVelocity=10000,MaxAngularAcceleration=10000,ExternalAngularVelocity=(X=0,Y=0,Z=0))
    ExternalForce="0,0,0"
   ) -> (Pose) {
    angularDampingOverride=0.9
    bOverrideAngularDamping=true
    bUseGravityOverride=true
    simulationSpace=RootRelative
    boundBone {boneName=dyn_skirt_side_l}
    componentAppliedLinearAccClamp=<10000 10000 10000>
    gravityOverride {z=-980}
    physicsBodyDefinitions [{
      boundBone {boneName=dyn_skirt_side_l}
      boxExtents=<10 10 10>
      constraintSetup {angularLimitsMax=<0 0 5>, angularLimitsMin=<0 -20 -5>, angularTarget=<-5 0 0>}
      localJointOffset=<0.1 0 0>
     }]
   }
   n26 = Anim Dynamics - Bone: dyn_skirt_side_r!(
    ComponentPose=@n25
    Alpha=1
    GravityOverride:Vector="0,0,-980"
    SimSpaceSettings:Anim Phys Sim Space Settings Structure=(SimSpaceAngularAlpha=0,MaxAngularVelocity=10000,MaxAngularAcceleration=10000,ExternalAngularVelocity=(X=0,Y=0,Z=0))
    ExternalForce="0,0,0"
   ) -> (Pose) {
    angularDampingOverride=0.9
    bOverrideAngularDamping=true
    bUseGravityOverride=true
    simulationSpace=RootRelative
    boundBone {boneName=dyn_skirt_side_r}
    componentAppliedLinearAccClamp=<10000 10000 10000>
    gravityOverride {z=-980}
    physicsBodyDefinitions [{
      boundBone {boneName=dyn_skirt_side_r}
      boxExtents=<10 10 10>
      constraintSetup {angularLimitsMax=<0 20 5>, angularLimitsMin=<0 -5 -5>, angularTarget=<-5 0 0>}
      localJointOffset=<0.1 0 0>
     }]
   }
   n27 = Anim Dynamics - Bone: dyn_skirt_bk_11!(
    ComponentPose=@n26
    Alpha=1
    GravityOverride:Vector="0,0,0"
    SimSpaceSettings:Anim Phys Sim Space Settings Structure=(SimSpaceAngularAlpha=0,MaxAngularVelocity=10000,MaxAngularAcceleration=10000,ExternalAngularVelocity=(X=0,Y=0,Z=0))
    ExternalForce="0,0,0"
   ) -> (Pose) {
    angularDampingOverride=0.9
    bAngularSpring=true
    bOverrideAngularDamping=true
    bUseGravityOverride=true
    simulationSpace=RootRelative
    boundBone {boneName=dyn_skirt_bk_11}
    componentAppliedLinearAccClamp=<3000 3000 3000>
    physicsBodyDefinitions [{
      boundBone {boneName=dyn_skirt_bk_11}
      boxExtents=<5 5 5>
      constraintSetup {angularLimitsMax=<0 20 20>, angularLimitsMin=<0 0 -50>, angularTarget=<1 0 0>}
      localJointOffset=<0.1 0 0>
     }]
    relativeSpaceBone {boneName=pelvis}
   }
   n28 = Anim Dynamics - Bone: dyn_skirt_bk_07!(
    ComponentPose=@n27
    Alpha=1
    GravityOverride:Vector="0,0,0"
    SimSpaceSettings:Anim Phys Sim Space Settings Structure=(SimSpaceAngularAlpha=0,MaxAngularVelocity=10000,MaxAngularAcceleration=10000,ExternalAngularVelocity=(X=0,Y=0,Z=0))
    ExternalForce="0,0,0"
   ) -> (Pose) {
    angularDampingOverride=0.9
    bAngularSpring=true
    bOverrideAngularDamping=true
    bUseGravityOverride=true
    simulationSpace=RootRelative
    boundBone {boneName=dyn_skirt_bk_07}
    componentAppliedLinearAccClamp=<3000 3000 3000>
    physicsBodyDefinitions [{
      boundBone {boneName=dyn_skirt_bk_07}
      boxExtents=<5 5 5>
      constraintSetup {angularLimitsMax=<0 10 20>, angularLimitsMin=<0 -10 -40>, angularTarget=<1 0 0>}
      localJointOffset=<0.1 0 0>
     }]
    relativeSpaceBone {boneName=pelvis}
   }
   n29 = Anim Dynamics - Bone: dyn_skirt_bk_10!(
    ComponentPose=@n28
    Alpha=1
    GravityOverride:Vector="0,0,0"
    SimSpaceSettings:Anim Phys Sim Space Settings Structure=(SimSpaceAngularAlpha=0,MaxAngularVelocity=10000,MaxAngularAcceleration=10000,ExternalAngularVelocity=(X=0,Y=0,Z=0))
    ExternalForce="0,0,0"
   ) -> (Pose) {
    angularDampingOverride=0.9
    bAngularSpring=true
    bOverrideAngularDamping=true
    bUseGravityOverride=true
    simulationSpace=RootRelative
    boundBone {boneName=dyn_skirt_bk_10}
    componentAppliedLinearAccClamp=<3000 3000 3000>
    physicsBodyDefinitions [{
      boundBone {boneName=dyn_skirt_bk_10}
      boxExtents=<5 5 5>
      constraintSetup {angularLimitsMax=<0 10 30>, angularLimitsMin=<0 -5 -5>, angularTarget=<1 0 0>}
      localJointOffset=<0.1 0 0>
     }]
    relativeSpaceBone {boneName=pelvis}
   }
   n30 = Anim Dynamics - Bone: dyn_skirt_fr_l_03(ComponentPose=@n29, Alpha=1, ExternalForce="0,0,0") {
    angularDampingOverride=0.9
    bOverrideAngularDamping=true
    bUseGravityOverride=true
    bUsePlanarLimit=false
    simulationSpace=RootRelative
    boundBone {boneName=dyn_skirt_fr_l_03}
    componentAppliedLinearAccClamp=<3000 3000 3000>
    componentLinearAccScale=<1 1 1>
    componentLinearVelScale=<1 1 1>
    gravityOverride {z=-980}
    physicsBodyDefinitions [{
      boundBone {boneName=dyn_skirt_fr_l_03}
      boxExtents=<6 6 6>
      constraintSetup {angularLimitsMax=<0 0 15>, angularLimitsMin=<0 0 0>, angularTarget=<1 0 0>}
      localJointOffset=<0.2 0 0>
     }]
   }
   n31 = Anim Dynamics - Bone: dyn_skirt_fr_l_04(ComponentPose=@n30, Alpha=1, ExternalForce="0,0,0") {
    angularDampingOverride=0.9
    bOverrideAngularDamping=true
    bUseGravityOverride=true
    bUsePlanarLimit=false
    simulationSpace=RootRelative
    boundBone {boneName=dyn_skirt_fr_l_04}
    componentAppliedLinearAccClamp=<3000 3000 3000>
    componentLinearAccScale=<1 1 1>
    componentLinearVelScale=<1 1 1>
    gravityOverride {z=-980}
    physicsBodyDefinitions [{
      boundBone {boneName=dyn_skirt_fr_l_04}
      boxExtents=<6 6 6>
      constraintSetup {angularLimitsMax=<0 5 30>, angularLimitsMin=<0 -5 0>, angularTarget=<1 0 0>}
      localJointOffset=<0.2 0 0>
     }]
   }
   n32 = Anim Dynamics - Bone: dyn_skirt_arrow_01_01(ComponentPose=@n31, Alpha=1, ExternalForce="0,0,0") {
    angularDampingOverride=0.9
    angularSpringConstant=800
    bAngularSpring=true
    bOverrideAngularDamping=true
    bUseGravityOverride=true
    bUsePlanarLimit=false
    simulationSpace=RootRelative
    boundBone {boneName=dyn_skirt_arrow_01_01}
    componentAppliedLinearAccClamp=<3000 3000 3000>
    componentLinearAccScale=<1 1 1>
    componentLinearVelScale=<1 1 1>
    gravityOverride {z=-980}
    physicsBodyDefinitions [{
      boundBone {boneName=dyn_skirt_arrow_01_01}
      boxExtents=<4 4 4>
      constraintSetup {angularLimitsMax=<0 0 50>, angularLimitsMin=<0 0 0>, angularTarget=<1 0 0>}
      localJointOffset=<0.1 0 0>
     }]
   }
   n33 = Anim Dynamics - Bone: dyn_skirt_arrow_01_02(ComponentPose=@n32, Alpha=1, ExternalForce="0,0,0") {
    angularDampingOverride=0.9
    angularSpringConstant=800
    bOverrideAngularDamping=true
    bUseGravityOverride=true
    bUsePlanarLimit=false
    simulationSpace=RootRelative
    boundBone {boneName=dyn_skirt_arrow_01_02}
    componentAppliedLinearAccClamp=<3000 3000 3000>
    componentLinearAccScale=<1 1 1>
    componentLinearVelScale=<1 1 1>
    gravityOverride {z=-980}
    physicsBodyDefinitions [{
      boundBone {boneName=dyn_skirt_arrow_01_02}
      boxExtents=<4 4 4>
      constraintSetup {angularLimitsMax=<0 0 50>, angularLimitsMin=<0 0 2>, angularTarget=<1 0 0>}
      localJointOffset=<0.1 0 0>
     }]
   }
   n34 = Anim Dynamics - Bone: dyn_skirt_arrow_02_01(ComponentPose=@n33, Alpha=1, ExternalForce="0,0,0") {
    angularDampingOverride=0.9
    angularSpringConstant=800
    bOverrideAngularDamping=true
    bUseGravityOverride=true
    bUsePlanarLimit=false
    simulationSpace=RootRelative
    boundBone {boneName=dyn_skirt_arrow_02_01}
    componentAppliedLinearAccClamp=<3000 3000 3000>
    componentLinearAccScale=<1 1 1>
    componentLinearVelScale=<1 1 1>
    gravityOverride {z=-980}
    physicsBodyDefinitions [{
      boundBone {boneName=dyn_skirt_arrow_02_01}
      boxExtents=<4 4 4>
      constraintSetup {angularLimitsMax=<0 0 35>, angularLimitsMin=<0 0 2>, angularTarget=<1 0 0>}
      localJointOffset=<0.1 0 0>
     }]
   }
   n35 = Anim Dynamics - Bone: dyn_skirt_arrow_02_02(ComponentPose=@n34, Alpha=1, ExternalForce="0,0,0") {
    angularDampingOverride=0.9
    angularSpringConstant=800
    bOverrideAngularDamping=true
    bUseGravityOverride=true
    bUsePlanarLimit=false
    simulationSpace=RootRelative
    boundBone {boneName=dyn_skirt_arrow_02_02}
    componentAppliedLinearAccClamp=<3000 3000 3000>
    componentLinearAccScale=<1 1 1>
    componentLinearVelScale=<1 1 1>
    gravityOverride {z=-980}
    physicsBodyDefinitions [{
      boundBone {boneName=dyn_skirt_arrow_02_02}
      boxExtents=<4 4 4>
      constraintSetup {angularLimitsMax=<0 3 35>, angularLimitsMin=<0 -3 2>, angularTarget=<1 0 0>}
      localJointOffset=<0.1 0 0>
     }]
   }
   n36 = Anim Dynamics - Bone: dyn_skirt_arrow_03_01(ComponentPose=@n35, Alpha=1, ExternalForce="0,0,0") {
    angularDampingOverride=0.9
    angularSpringConstant=800
    bOverrideAngularDamping=true
    bUseGravityOverride=true
    bUsePlanarLimit=false
    simulationSpace=RootRelative
    boundBone {boneName=dyn_skirt_arrow_03_01}
    componentAppliedLinearAccClamp=<3000 3000 3000>
    componentLinearAccScale=<1 1 1>
    componentLinearVelScale=<1 1 1>
    gravityOverride {z=-980}
    physicsBodyDefinitions [{
      boundBone {boneName=dyn_skirt_arrow_03_01}
      boxExtents=<4 4 4>
      constraintSetup {angularLimitsMax=<0 0 50>, angularLimitsMin=<0 0 2>, angularTarget=<1 0 0>}
      localJointOffset=<0.1 0 0>
     }]
   }
   n37 = Anim Dynamics - Bone: dyn_skirt_arrow_03_02(ComponentPose=@n36, Alpha=1, ExternalForce="0,0,0") {
    angularDampingOverride=0.9
    angularSpringConstant=800
    bOverrideAngularDamping=true
    bUseGravityOverride=true
    bUsePlanarLimit=false
    simulationSpace=RootRelative
    boundBone {boneName=dyn_skirt_arrow_03_02}
    componentAppliedLinearAccClamp=<3000 3000 3000>
    componentLinearAccScale=<1 1 1>
    componentLinearVelScale=<1 1 1>
    gravityOverride {z=-980}
    physicsBodyDefinitions [{
      boundBone {boneName=dyn_skirt_arrow_03_02}
      boxExtents=<4 4 4>
      constraintSetup {angularLimitsMax=<0 3 50>, angularLimitsMin=<0 -3 2>, angularTarget=<1 0 0>}
      localJointOffset=<0.1 0 0>
     }]
   }
   n38 = Anim Dynamics - Bone: dyn_skirt_arrow_04_01(ComponentPose=@n37, Alpha=1, ExternalForce="0,0,0") {
    angularDampingOverride=0.9
    angularSpringConstant=800
    bOverrideAngularDamping=true
    bUseGravityOverride=true
    bUsePlanarLimit=false
    simulationSpace=RootRelative
    boundBone {boneName=dyn_skirt_arrow_04_01}
    componentAppliedLinearAccClamp=<3000 3000 3000>
    componentLinearAccScale=<1 1 1>
    componentLinearVelScale=<1 1 1>
    gravityOverride {z=-980}
    physicsBodyDefinitions [{
      boundBone {boneName=dyn_skirt_arrow_04_01}
      boxExtents=<4 4 4>
      constraintSetup {angularLimitsMax=<0 0 40>, angularLimitsMin=<0 0 0>, angularTarget=<1 0 0>}
      localJointOffset=<0.1 0 0>
     }]
   }
   n39 = Anim Dynamics - Bone: dyn_skirt_arrow_04_02(ComponentPose=@n38, Alpha=1, ExternalForce="0,0,0") {
    angularDampingOverride=0.9
    angularSpringConstant=800
    bOverrideAngularDamping=true
    bUseGravityOverride=true
    bUsePlanarLimit=false
    simulationSpace=RootRelative
    boundBone {boneName=dyn_skirt_arrow_04_02}
    componentAppliedLinearAccClamp=<3000 3000 3000>
    componentLinearAccScale=<1 1 1>
    componentLinearVelScale=<1 1 1>
    gravityOverride {z=-980}
    physicsBodyDefinitions [{
      boundBone {boneName=dyn_skirt_arrow_04_02}
      boxExtents=<4 4 4>
      constraintSetup {angularLimitsMax=<0 3 40>, angularLimitsMin=<0 -3 2>, angularTarget=<1 0 0>}
      localJointOffset=<0.1 0 0>
     }]
   }
   n40 = Anim Dynamics - Bone: dyn_skirt_arrow_05_01(ComponentPose=@n39, Alpha=1, ExternalForce="0,0,0") {
    angularDampingOverride=0.9
    angularSpringConstant=800
    bOverrideAngularDamping=true
    bUseGravityOverride=true
    bUsePlanarLimit=false
    simulationSpace=RootRelative
    boundBone {boneName=dyn_skirt_arrow_05_01}
    componentAppliedLinearAccClamp=<3000 3000 3000>
    componentLinearAccScale=<1 1 1>
    componentLinearVelScale=<1 1 1>
    gravityOverride {z=-980}
    physicsBodyDefinitions [{
      boundBone {boneName=dyn_skirt_arrow_05_01}
      boxExtents=<4 4 4>
      constraintSetup {angularLimitsMax=<0 0 50>, angularLimitsMin=<0 0 2>, angularTarget=<1 0 0>}
      localJointOffset=<0.1 0 0>
     }]
   }
   n41 = Anim Dynamics - Bone: dyn_skirt_arrow_05_02(ComponentPose=@n40, Alpha=1, ExternalForce="0,0,0") {
    angularDampingOverride=0.9
    angularSpringConstant=800
    bOverrideAngularDamping=true
    bUseGravityOverride=true
    bUsePlanarLimit=false
    simulationSpace=RootRelative
    boundBone {boneName=dyn_skirt_arrow_05_02}
    componentAppliedLinearAccClamp=<3000 3000 3000>
    componentLinearAccScale=<1 1 1>
    componentLinearVelScale=<1 1 1>
    gravityOverride {z=-980}
    physicsBodyDefinitions [{
      boundBone {boneName=dyn_skirt_arrow_05_02}
      boxExtents=<4 4 4>
      constraintSetup {angularLimitsMax=<0 3 40>, angularLimitsMin=<0 -3 2>, angularTarget=<1 0 0>}
      localJointOffset=<0.1 0 0>
     }]
   }
   n42 = Anim Dynamics - Bone: dyn_skirt_arrow_06_01(ComponentPose=@n41, Alpha=1, ExternalForce="0,0,0") {
    angularDampingOverride=0.9
    angularSpringConstant=800
    bOverrideAngularDamping=true
    bUseGravityOverride=true
    bUsePlanarLimit=false
    simulationSpace=RootRelative
    boundBone {boneName=dyn_skirt_arrow_06_01}
    componentAppliedLinearAccClamp=<3000 3000 3000>
    componentLinearAccScale=<1 1 1>
    componentLinearVelScale=<1 1 1>
    gravityOverride {z=-980}
    physicsBodyDefinitions [{
      boundBone {boneName=dyn_skirt_arrow_06_01}
      boxExtents=<4 4 4>
      constraintSetup {angularLimitsMax=<0 0 35>, angularLimitsMin=<0 0 0>, angularTarget=<1 0 0>}
      localJointOffset=<0.1 0 0>
     }]
   }
   n43 = Anim Dynamics - Bone: dyn_skirt_arrow_06_02(ComponentPose=@n42, Alpha=1, ExternalForce="0,0,0") {
    angularDampingOverride=0.9
    angularSpringConstant=800
    bOverrideAngularDamping=true
    bUseGravityOverride=true
    bUsePlanarLimit=false
    simulationSpace=RootRelative
    boundBone {boneName=dyn_skirt_arrow_06_02}
    componentAppliedLinearAccClamp=<3000 3000 3000>
    componentLinearAccScale=<1 1 1>
    componentLinearVelScale=<1 1 1>
    gravityOverride {z=-980}
    physicsBodyDefinitions [{
      boundBone {boneName=dyn_skirt_arrow_06_02}
      boxExtents=<4 4 4>
      constraintSetup {angularLimitsMax=<0 3 35>, angularLimitsMin=<0 -3 0>, angularTarget=<1 0 0>}
      localJointOffset=<0.1 0 0>
     }]
   }
   n44 = Anim Dynamics - Bone: dyn_skirt_arrow_07_01(ComponentPose=@n43, Alpha=1, ExternalForce="0,0,0") {
    angularDampingOverride=0.9
    angularSpringConstant=800
    bOverrideAngularDamping=true
    bUseGravityOverride=true
    bUsePlanarLimit=false
    simulationSpace=RootRelative
    boundBone {boneName=dyn_skirt_arrow_07_01}
    componentAppliedLinearAccClamp=<3000 3000 3000>
    componentLinearAccScale=<1 1 1>
    componentLinearVelScale=<1 1 1>
    gravityOverride {z=-980}
    physicsBodyDefinitions [{
      boundBone {boneName=dyn_skirt_arrow_07_01}
      boxExtents=<4 4 4>
      constraintSetup {angularLimitsMax=<0 0 50>, angularLimitsMin=<0 0 0>, angularTarget=<1 0 0>}
      localJointOffset=<0.1 0 0>
     }]
   }
   Save cached pose 'toTorsoDyn'(Pose=Component To Local(ComponentPose=@n44)) {cacheName=toTorsoDyn}
   n45 = "Constraint - dyn_leg_strap_r_base (PPPPPP)"!(
    ComponentPose=Local To Component(LocalPose=Copy Pose From Mesh() {bUseAttachedParent=true})
    Alpha=1
    ConstraintWeights_0=0
    ConstraintWeights_1:float=0.1178
    ConstraintWeights_2:float=0.8821
    ConstraintWeights_3:float=0
    ConstraintWeights_4:float=0
    ConstraintWeights_5:float=0.0001
   ) -> (Pose) {
    boneToModify {boneName=dyn_leg_strap_r_base}
    constraintSetup [
     {targetBone {boneName=pelvis}}
     {targetBone {boneName=thigh_twist_01_r}}
     {targetBone {boneName=thigh_r}}
     {targetBone {boneName=calf_r}}
     {targetBone {boneName=deform_glute_r}}
     {targetBone {boneName=deform_knee_in_r}}
    ]
    constraintWeights [1,1,1,1,1,1]
   }
   n46 = "Constraint - dyn_skirt_arrow_01_base (PPPPPPPP)"!(
    ComponentPose=@n45
    Alpha=1
    ConstraintWeights_0=0.2886
    ConstraintWeights_1:float=0.3003
    ConstraintWeights_2:float=0.0224
    ConstraintWeights_3:float=0.0454
    ConstraintWeights_4:float=0.0018
    ConstraintWeights_5:float=0.1767
    ConstraintWeights_6:float=0.0084
    ConstraintWeights_7:float=0.1563
   ) -> (Pose) {
    boneToModify {boneName=dyn_skirt_arrow_01_base}
    constraintSetup [
     {targetBone {boneName=pelvis}}
     {targetBone {boneName=thigh_twist_01_l}}
     {targetBone {boneName=thigh_twist_01_r}}
     {targetBone {boneName=thigh_l}}
     {targetBone {boneName=thigh_r}}
     {targetBone {boneName=deform_groin_l}}
     {targetBone {boneName=deform_groin_r}}
     {targetBone {boneName=dyn_skirt_fr_l_01}}
    ]
    constraintWeights [1,1,1,1,1,1,1,1]
   }
   n47 = "Constraint - dyn_skirt_arrow_02_base (PPPPPPPP)"!(
    ComponentPose=@n46
    Alpha=1
    ConstraintWeights_0=0.3164
    ConstraintWeights_1:float=0.2263
    ConstraintWeights_2:float=0.0585
    ConstraintWeights_3:float=0.0338
    ConstraintWeights_4:float=0.002
    ConstraintWeights_5:float=0.1342
    ConstraintWeights_6:float=0.0205
    ConstraintWeights_7:float=0.2083
   ) -> (Pose) {
    boneToModify {boneName=dyn_skirt_arrow_02_base}
    constraintSetup [
     {targetBone {boneName=pelvis}}
     {targetBone {boneName=thigh_twist_01_l}}
     {targetBone {boneName=thigh_twist_01_r}}
     {targetBone {boneName=thigh_l}}
     {targetBone {boneName=thigh_r}}
     {targetBone {boneName=deform_groin_l}}
     {targetBone {boneName=deform_groin_r}}
     {targetBone {boneName=dyn_skirt_fr_l_01}}
    ]
    constraintWeights [1,1,1,1,1,1,1,1]
   }
   n48 = "Constraint - dyn_skirt_arrow_03_base (PPPPPPPP)"!(
    ComponentPose=@n47
    Alpha=1
    ConstraintWeights_0=0.3609
    ConstraintWeights_1:float=0.2086
    ConstraintWeights_2:float=0.1039
    ConstraintWeights_3:float=0.0263
    ConstraintWeights_4:float=0.1164
    ConstraintWeights_5:float=0.042
    ConstraintWeights_6:float=0.1365
    ConstraintWeights_7:float=0.0054
   ) -> (Pose) {
    boneToModify {boneName=dyn_skirt_arrow_03_base}
    constraintSetup [
     {targetBone {boneName=pelvis}}
     {targetBone {boneName=thigh_twist_01_l}}
     {targetBone {boneName=thigh_twist_01_r}}
     {targetBone {boneName=thigh_l}}
     {targetBone {boneName=deform_groin_l}}
     {targetBone {boneName=deform_groin_r}}
     {targetBone {boneName=dyn_skirt_fr_l_01}}
     {targetBone {boneName=dyn_skirt_fr_r_01}}
    ]
    constraintWeights [1,1,1,1,1,1,1,1]
   }
   n49 = "Constraint - dyn_skirt_arrow_04_base (PPPPPPPP)"!(
    ComponentPose=@n48
    Alpha=1
    ConstraintWeights_0=0.4184
    ConstraintWeights_1:float=0.1403
    ConstraintWeights_2:float=0.1622
    ConstraintWeights_3:float=0.0142
    ConstraintWeights_4:float=0.0707
    ConstraintWeights_5:float=0.0739
    ConstraintWeights_6:float=0.0688
    ConstraintWeights_7:float=0.0514
   ) -> (Pose) {
    boneToModify {boneName=dyn_skirt_arrow_04_base}
    constraintSetup [
     {targetBone {boneName=pelvis}}
     {targetBone {boneName=thigh_twist_01_l}}
     {targetBone {boneName=thigh_twist_01_r}}
     {targetBone {boneName=thigh_r}}
     {targetBone {boneName=deform_groin_l}}
     {targetBone {boneName=deform_groin_r}}
     {targetBone {boneName=dyn_skirt_fr_l_01}}
     {targetBone {boneName=dyn_skirt_fr_r_01}}
    ]
    constraintWeights [1,1,1,1,1,1,1,1]
   }
   n50 = "Constraint - dyn_skirt_arrow_05_base (PPPPPPPPP)"!(
    ComponentPose=@n49
    Alpha=1
    ConstraintWeights_0=0.2646
    ConstraintWeights_1:float=0.0665
    ConstraintWeights_2:float=0.215
    ConstraintWeights_3:float=0.0183
    ConstraintWeights_4:float=0.0131
    ConstraintWeights_5:float=0.0962
    ConstraintWeights_6:float=0.0535
    ConstraintWeights_7:float=0.2434
    ConstraintWeights_8:float=0.0293
   ) -> (Pose) {
    boneToModify {boneName=dyn_skirt_arrow_05_base}
    constraintSetup [
     {targetBone {boneName=pelvis}}
     {targetBone {boneName=thigh_twist_01_l}}
     {targetBone {boneName=thigh_twist_01_r}}
     {targetBone {boneName=thigh_r}}
     {targetBone {boneName=deform_groin_l}}
     {targetBone {boneName=deform_groin_r}}
     {targetBone {boneName=dyn_skirt_fr_l_01}}
     {targetBone {boneName=dyn_skirt_fr_r_01}}
     {targetBone {boneName=dyn_skirt_fr_mid_01}}
    ]
    constraintWeights [1,1,1,1,1,1,1,1,1]
   }
   n51 = "Constraint - dyn_skirt_arrow_06_base (PPPPPPPP)"!(
    ComponentPose=@n50
    Alpha=1
    ConstraintWeights_0=0.0648
    ConstraintWeights_1:float=0.0149
    ConstraintWeights_2:float=0.09
    ConstraintWeights_3:float=0.0323
    ConstraintWeights_4:float=0.028
    ConstraintWeights_5:float=0.7424
    ConstraintWeights_6:float=0.0086
    ConstraintWeights_7:float=0.0191
   ) -> (Pose) {
    boneToModify {boneName=dyn_skirt_arrow_06_base}
    constraintSetup [
     {targetBone {boneName=pelvis}}
     {targetBone {boneName=thigh_twist_01_l}}
     {targetBone {boneName=thigh_twist_01_r}}
     {targetBone {boneName=deform_groin_r}}
     {targetBone {boneName=dyn_skirt_fr_l_01}}
     {targetBone {boneName=dyn_skirt_fr_r_01}}
     {targetBone {boneName=dyn_skirt_fr_mid_base}}
     {targetBone {boneName=dyn_skirt_fr_mid_01}}
    ]
    constraintWeights [1,1,1,1,1,1,1,1]
   }
   n52 = "Constraint - dyn_skirt_arrow_07_base (PPPPPPPP)"!(
    ComponentPose=@n51
    Alpha=1
    ConstraintWeights_0=0.0578
    ConstraintWeights_1:float=0.0103
    ConstraintWeights_2:float=0.1022
    ConstraintWeights_3:float=0.0235
    ConstraintWeights_4:float=0.0003
    ConstraintWeights_5:float=0.0354
    ConstraintWeights_6:float=0.0149
    ConstraintWeights_7:float=0.7557
   ) -> (Pose) {
    boneToModify {boneName=dyn_skirt_arrow_07_base}
    constraintSetup [
     {targetBone {boneName=pelvis}}
     {targetBone {boneName=thigh_twist_01_l}}
     {targetBone {boneName=thigh_twist_01_r}}
     {targetBone {boneName=thigh_r}}
     {targetBone {boneName=deform_groin_l}}
     {targetBone {boneName=deform_groin_r}}
     {targetBone {boneName=dyn_skirt_fr_l_01}}
     {targetBone {boneName=dyn_skirt_fr_r_01}}
    ]
    constraintWeights [1,1,1,1,1,1,1,1]
   }
   Save cached pose 'toCTRLrig'(Pose=Component To Local(ComponentPose=@n52)) {cacheName=toCTRLrig}
  ]}
 EventGraph {nodes [
   Event BlueprintUpdateAnimation()
   > n1 = Sequence!(execute) -> (then_0, then_1, then_2:Exec)
   n2 = Branch(execute=@n1.then_1, Condition=Get bIsInFrontEnd())
   > Set RBAN Alpha(RBAN Alpha:double=0.6) -> (Output_Get:double)
   > Set AnimDynAlpha(AnimDynAlpha:double=0.8) -> (Output_Get:double)
   Set RBAN Alpha(execute=@n2.else, RBAN Alpha=1)
   > Set AnimDynAlpha(AnimDynAlpha=1)
   n3 = PlayerMovement(execute=@n1.then_2, AnimInstance=Self-Reference(), XDotProductCoefficient=1, YDotProductCoefficient=1, ZDotProductCoefficient=1)
   > Set Is Right Left Movement(Is Right Left Movement:double=@n3) -> (Output_Get:double)
   > Set Is Forward Backward Movement(Is Forward Backward Movement:double=@n3.YMovement) -> (Output_Get:double)
   > Set Jumping ON(Jumping ON:double="To Float (Boolean)"(InBool=Get Is Jumping() -> (Is Jumping:bool))) -> (Output_Get:double)
   > Set Crouching ON(Crouching ON:double="To Float (Boolean)"(InBool=Get Is Crouching() -> (Is Crouching:bool))) -> (Output_Get:double)
   > Set KneeSliding ON(KneeSliding ON:double="To Float (Boolean)"(InBool=Get Is Kneesliding() -> (Is Kneesliding:bool))) -> (Output_Get:double)
   n4 = "Gravity Override State Function (Parameter Asset) (Nativized)"(
    execute=@n1.then_0
    CharacterPartAnimInstance=Self-Reference()
    GameTimeInSeconds=GetGameTimeInSeconds()
    Parameters=&/M_MED_WriteWalrus_Parameters
   )
   > Set Gravity Override(Gravity Override=@n4)
   > Set RBAN Settings Out(RBAN Settings Out=@n4.OutRigidBodyAnimNodeParameters)
   Event OnBeginCrouching()
   > Set Is Crouching(Is Crouching:bool=true) -> (Output_Get:bool)
   Event OnEndCrouching()
   > Set Is Crouching(Is Crouching=false)
   Event OnBeginJumping()
   > Set Is Jumping(Is Jumping:bool=true) -> (Output_Get:bool)
   Event OnEndJumping()
   > Set Is Jumping(Is Jumping=false)
   Event OnBeginKneeSliding() {is=title K2Node_Event}
   > Set Is Kneesliding(Is Kneesliding:bool=true) -> (Output_Get:bool)
   Event OnEndKneeSliding() {is=title K2Node_Event}
   > Set Is Kneesliding(Is Kneesliding=false)
  ]}
]
variables [
 AnimDynAlpha {type=number}
 Crouching ON {type=number}
 Gravity Override {type=Vector}
 Is Crouching {type=boolean}
 Is Forward Backward Movement {type=number}
 Is Jumping {type=boolean}
 Is Kneesliding {type=boolean}
 Is Right Left Movement {type=number}
 Jumping ON {type=number}
 KneeSliding ON {type=number}
 RBAN Alpha {type=number}
 RBAN Settings Out {type=FortRigidBodyAnimNodeParameters}
 Sim Space Settings {
  type=AnimPhysSimSpaceSettings
  default {maxAngularAcceleration=10000, maxAngularVelocity=10000, simSpaceAngularAlpha=0, externalAngularVelocity=<0 0 0>}
 }
]
