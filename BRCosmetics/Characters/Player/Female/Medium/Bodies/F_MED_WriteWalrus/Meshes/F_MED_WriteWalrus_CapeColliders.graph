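# Physics asset of cape colliders for the F_MED_WriteWalrus female player body, built from capsules.
# Body capsules cover the head, clavicles, spine, pelvis and feet, while arm, thigh, calf and neck bodies carry no shapes.
# Three dyn_skirt bone chains (back middle, back left, left side-back) hang from the pelvis, with no right-side counterparts.
# Their constraints leave swing and twist free, so the stored angle limits are inactive.
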PhysicsAsset /BRCosmetics/Characters/Player/Female/Medium/Bodies/F_MED_WriteWalrus/Meshes/F_MED_WriteWalrus_CapeColliders
bodies {
 calf_l []
 calf_r []
 clavicle_l [{shapeType=Capsule, length=6.5623, radius=7.5194, center=<3.9566 -1.0862 -2.1127>, rotation=<rot -77.5566 177.3915 7.7314>}]
 clavicle_r [{shapeType=Capsule, length=6.5623, radius=7.5194, center=<-3.9566 1.0862 2.1127>, rotation=<rot -77.5566 177.3915 -172.2686>}]
 dyn_skirt_bk_l_01 [{shapeType=Capsule, length=8.8335, radius=5.1783, shapeName=dyn_skirt_bk_l_01_capsule, center=<5.2923 0 0>, rotation=<rot 90 0 0>}]
 dyn_skirt_bk_l_02 [{shapeType=Capsule, length=7.5212, radius=5.2026, shapeName=dyn_skirt_bk_l_02_capsule, center=<5.634 0 0>, rotation=<rot 90 0 0>}]
 dyn_skirt_bk_l_03 [{shapeType=Capsule, length=10.0169, radius=4.9316, shapeName=dyn_skirt_bk_l_03_capsule, center=<7.7976 0 0>, rotation=<rot 90 0 0>}]
 dyn_skirt_bk_l_base [{shapeType=Capsule, length=0.505, radius=0.505, shapeName=dyn_skirt_bk_l_base_capsule}]
 dyn_skirt_bk_mid_01 [{shapeType=Capsule, length=15.5089, radius=7.6279, shapeName=dyn_skirt_bk_mid_01_capsule}]
 dyn_skirt_bk_mid_02 [{shapeType=Capsule, length=13.8316, radius=7.9795, shapeName=dyn_skirt_bk_mid_02_capsule}]
 dyn_skirt_bk_mid_03 [{shapeType=Capsule, length=14.2348, radius=6.9315, shapeName=dyn_skirt_bk_mid_03_capsule}]
 dyn_skirt_bk_mid_04 [{shapeType=Capsule, length=12.6695, radius=8.0745, shapeName=dyn_skirt_bk_mid_04_capsule}]
 dyn_skirt_bk_mid_05 [{shapeType=Capsule, length=11.8546, radius=9.4294, shapeName=dyn_skirt_bk_mid_05_capsule}]
 dyn_skirt_bk_mid_base [{shapeType=Capsule, length=0.505, radius=0.505, shapeName=dyn_skirt_bk_mid_base_capsule}]
 dyn_skirt_side_l_bk_01 [{shapeType=Capsule, length=11.0283, radius=6.2246, shapeName=dyn_skirt_side_l_bk_01_capsule, center=<9.2853 0 0>, rotation=<rot 0 90 90>}]
 dyn_skirt_side_l_bk_02 [{shapeType=Capsule, length=15.7878, radius=9.2395, shapeName=dyn_skirt_side_l_bk_02_capsule, center=<13.423 0 0>, rotation=<rot 0 90 90>}]
 dyn_skirt_side_l_bk_base [{shapeType=Capsule, length=0.505, radius=0.505, shapeName=dyn_skirt_side_l_bk_base_capsule, rotation=<rot 90 0 90>}]
 foot_l [{shapeType=Capsule, length=11.0765, radius=8.0344, center=<-4.1811 6.0027 -0.338>, rotation=<rot 0.0388 1.4701 -89.3982>}]
 foot_r [{shapeType=Capsule, length=11.0765, radius=8.0344, center=<4.1811 -6.0027 0.338>, rotation=<rot 0.0388 1.4701 90.6018>}]
 head [{shapeType=Capsule, length=0, radius=18, center=<9.0697 0.0746 0>, rotation=<rot 0 -90 0>}]
 lowerarm_l []
 lowerarm_r []
 neck_02 []
 pelvis [{shapeType=Capsule, length=9.6042, radius=11.1441, center=<-2.4471 -2.5289 -0.1981>}]
 spine_01 [{shapeType=Capsule, length=19.2313, radius=15.0614, center=<-11.7029 -0.5237 0>}]
 spine_02 [
  {shapeType=Capsule, length=9.6719, radius=9.7317, center=<-2.6885 -1.2641 0>}
  {shapeType=Capsule, length=20.6875, radius=9.3125, center=<-1.8317 1.9362 0>, rotation=<rot 90 0 0>}
 ]
 spine_03 [
  {shapeType=Capsule, length=15.5971, radius=16.9153, center=<0.0574 0.4416 0>}
  {shapeType=Capsule, length=20.7188, radius=9.2812, center=<-1.2163 0.549 0>, rotation=<rot 90 180 -175.0732>}
 ]
 spine_04 [
  {shapeType=Capsule, length=17.9853, radius=12.8587, center=<5.6234 -0.6779 0>}
  {shapeType=Capsule, length=10.0312, radius=9.2812, center=<-1.2326 1.308 0>, rotation=<rot 90 0 0>}
 ]
 spine_05 [{shapeType=Capsule, length=9.678, radius=12.3571, center=<0.6969 1.91 0>}]
 thigh_l []
 thigh_r []
 upperarm_l []
 upperarm_r []
}
constraints [
 {
  bone1Name=dyn_skirt_bk_mid_base
  bone2Name=pelvis
  swing1LimitDegrees=45
  swing1Motion=Free
  swing2LimitDegrees=45
  swing2Motion=Free
  twistLimitDegrees=45
  twistMotion=Free
 }
 {
  bone1Name=dyn_skirt_bk_mid_01
  bone2Name=dyn_skirt_bk_mid_base
  swing1LimitDegrees=45
  swing1Motion=Free
  swing2LimitDegrees=45
  swing2Motion=Free
  twistLimitDegrees=45
  twistMotion=Free
 }
 {
  bone1Name=dyn_skirt_bk_mid_02
  bone2Name=dyn_skirt_bk_mid_01
  swing1LimitDegrees=45
  swing1Motion=Free
  swing2LimitDegrees=45
  swing2Motion=Free
  twistLimitDegrees=45
  twistMotion=Free
 }
 {
  bone1Name=dyn_skirt_bk_mid_03
  bone2Name=dyn_skirt_bk_mid_02
  swing1LimitDegrees=45
  swing1Motion=Free
  swing2LimitDegrees=45
  swing2Motion=Free
  twistLimitDegrees=45
  twistMotion=Free
 }
 {
  bone1Name=dyn_skirt_bk_mid_04
  bone2Name=dyn_skirt_bk_mid_03
  swing1LimitDegrees=45
  swing1Motion=Free
  swing2LimitDegrees=45
  swing2Motion=Free
  twistLimitDegrees=45
  twistMotion=Free
 }
 {
  bone1Name=dyn_skirt_bk_mid_05
  bone2Name=dyn_skirt_bk_mid_04
  swing1LimitDegrees=45
  swing1Motion=Free
  swing2LimitDegrees=45
  swing2Motion=Free
  twistLimitDegrees=45
  twistMotion=Free
 }
 {
  bone1Name=dyn_skirt_bk_l_base
  bone2Name=pelvis
  swing1LimitDegrees=45
  swing1Motion=Free
  swing2LimitDegrees=45
  swing2Motion=Free
  twistLimitDegrees=45
  twistMotion=Free
 }
 {
  bone1Name=dyn_skirt_bk_l_01
  bone2Name=dyn_skirt_bk_l_base
  swing1LimitDegrees=45
  swing1Motion=Free
  swing2LimitDegrees=45
  swing2Motion=Free
  twistLimitDegrees=45
  twistMotion=Free
 }
 {
  bone1Name=dyn_skirt_bk_l_02
  bone2Name=dyn_skirt_bk_l_01
  swing1LimitDegrees=45
  swing1Motion=Free
  swing2LimitDegrees=45
  swing2Motion=Free
  twistLimitDegrees=45
  twistMotion=Free
 }
 {
  bone1Name=dyn_skirt_bk_l_03
  bone2Name=dyn_skirt_bk_l_02
  swing1LimitDegrees=45
  swing1Motion=Free
  swing2LimitDegrees=45
  swing2Motion=Free
  twistLimitDegrees=45
  twistMotion=Free
 }
 {
  bone1Name=dyn_skirt_side_l_bk_base
  bone2Name=pelvis
  swing1LimitDegrees=45
  swing1Motion=Free
  swing2LimitDegrees=45
  swing2Motion=Free
  twistLimitDegrees=45
  twistMotion=Free
 }
 {
  bone1Name=dyn_skirt_side_l_bk_01
  bone2Name=dyn_skirt_side_l_bk_base
  swing1LimitDegrees=45
  swing1Motion=Free
  swing2LimitDegrees=45
  swing2Motion=Free
  twistLimitDegrees=45
  twistMotion=Free
 }
 {
  bone1Name=dyn_skirt_side_l_bk_02
  bone2Name=dyn_skirt_side_l_bk_01
  swing1LimitDegrees=45
  swing1Motion=Free
  swing2LimitDegrees=45
  swing2Motion=Free
  twistLimitDegrees=45
  twistMotion=Free
 }
]
