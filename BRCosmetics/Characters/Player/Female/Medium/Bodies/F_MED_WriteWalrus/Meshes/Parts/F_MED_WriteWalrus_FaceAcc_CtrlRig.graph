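# Control Rig for the WriteWalrus female character's face accessory part, correcting hair bones as the body moves.
# The back hair base blends between head and upper spine, then takes weighted offsets from head rotation, crouching with both arms posed, and skydive floating.
# A right upper-arm spherical pose reader also drives side hair.
# With Is Backbling set, spine and clavicle-constrained controls drive the back hair chain through twist and Hermite spline IK.
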ControlRigBlueprint /BRCosmetics/Characters/Player/Female/Medium/Bodies/F_MED_WriteWalrus/Meshes/Parts/F_MED_WriteWalrus_FaceAcc_CtrlRig
graphs [
 Multiply_1_ContainedGraph {nodes [
   n1 = RigVMFunctionEntryNode()
   Multiply = &RigVMFunction_MathDoubleMul(A=@n1.A, B=@n1.B)
   RigVMFunctionReturnNode(Result=&RigVMFunction_MathDoubleMul(B=@n1.D, A=@Multiply.Result).Result)
  ]}
 RigVMFunctionLibrary {links [], nodes []}
 RigVMModel {nodes [
   RigUnit_BeginExecution = &RigUnit_BeginExecution()
   n1 = &RigUnit_ParentConstraint(
    AdvancedSettings=(InterpolationType=Average,RotationOrderForFilter=XZY)
    Child=(Type=Bone,Name="dyn_hair_bk_base")
    Filter=(TranslationFilter=(bX=true,bY=true,bZ=true),RotationFilter=(bX=true,bY=true,bZ=true),ScaleFilter=(bX=True,bY=True,bZ=True))
    Parents=((Item=(Type=Bone,Name="head"),Weight=0.8),(Item=(Type=Bone,Name="spine_05"),Weight=0.2))
    Weight=1
    ExecutePin=@RigUnit_BeginExecution.ExecutePin
   )
   RigVMFunction_Sequence = &RigVMFunction_Sequence(ExecuteContext=@n1.ExecutePin)
   Get Delta Ref = RigVMFunctionReferenceNode(Item=(Type=Bone,Name="head"), ExecuteContext=@RigVMFunction_Sequence.A)
   AlphaInterp = &RigVMFunction_AlphaInterp(
    InterpSpeedDecreasing=6
    InterpSpeedIncreasing=6
    Scale=1
    bClampResult=False
    bInterpResult=False
    Value=&RigVMFunction_MathDoubleRemap(SourceMaximum=-30, SourceMinimum=-15, TargetMaximum=1, TargetMinimum=0, Value=@"Get Delta Ref".Result.Rotation.Yaw).Result
   )
   > &RigUnit_ModifyTransforms(
    ItemToModify=((Item=(Type=Bone,Name="dyn_hair_bk_base"),Transform=(Rotation=(X=0,Y=0,Z=0,W=1),Translation=(X=-2.9051,Y=-10,Z=1),Scale3D=(X=1,Y=1,Z=1))))
    WeightMinimum=0
    Weight=@AlphaInterp.Result
   )
   > Get Delta Ref_1 = RigVMFunctionReferenceNode(Item=(Type=Bone,Name="head"))
   n2 = &RigVMFunction_MathDoubleRemap(SourceMaximum=4, SourceMinimum=0, TargetMaximum=1, TargetMinimum=0, Value=@"Get Delta Ref_1".Result.Rotation.Roll)
   n3 = &RigVMFunction_AlphaInterp(
    InterpSpeedDecreasing=6
    InterpSpeedIncreasing=6
    Scale=1
    bClampResult=False
    bInterpResult=False
    Value=&RigVMFunction_MathDoubleMul(
     B=&RigVMFunction_MathDoubleRemap(SourceMaximum=-10, SourceMinimum=0, TargetMaximum=1, TargetMinimum=0, Value=@"Get Delta Ref_1".Result.Rotation.Yaw).Result
     A=@n2.Result
    ).Result
   )
   > &RigUnit_ModifyTransforms(
    ItemToModify=((Item=(Type=Bone,Name="dyn_hair_bk_base"),Transform=(Rotation=(X=0.2164,Y=0,Z=0,W=0.9763),Translation=(X=0,Y=0,Z=0),Scale3D=(X=1,Y=1,Z=1))))
    WeightMinimum=0
    Weight=@n3.Result
   )
   n4 = &RigVMFunction_AlphaInterp(InterpSpeedDecreasing=6, InterpSpeedIncreasing=6, Scale=1, bClampResult=False, bInterpResult=False, Value=@n2.Result)
   > &RigUnit_ModifyTransforms(
    ItemToModify="((Item=(Type=Bone,Name=\"dyn_hair_bk_base\"),Transform=(Rotation=(X=0,Y=0,Z=0,W=1),Translation=(X=-2,Y=0,Z=0),Scale3D=(X=1,Y=1,Z=1))),(Item=(Type=Bone,Name=\"hair_r_support\"),Transform=(Rotation=(X=0,Y=0,Z=0,W=1),Translation=(X=0,Y=-2.222,Z=0),Scale3D=(X=1,Y= ...(truncated)"
    WeightMinimum=0
    Weight=@n4.Result
   )
   > n5 = &RigUnit_SphericalPoseReader(
    ActiveRegionScaleFactors=(PositiveWidth=1,NegativeWidth=1,PositiveHeight=1,NegativeHeight=1)
    ActiveRegionSize=0.15
    Debug=(bDrawDebug=True,bDraw2D=False,bDrawLocalAxes=False,DebugScale=25,DebugSegments=20,DebugThickness=0.25)
    DriverAxis=(X=1,Y=0,Z=0)
    DriverItem=(Type=Bone,Name="upperarm_l")
    FalloffRegionScaleFactors=(PositiveWidth=1,NegativeWidth=1,PositiveHeight=1,NegativeHeight=1)
    FalloffSize=0.3
    FlipHeightScaling=False
    FlipWidthScaling=False
    OptionalParentItem=(Type=Bone,Name="None")
    RotationOffset=(X=50,Y=0,Z=-40)
   )
   > n6 = &RigUnit_SphericalPoseReader(
    ActiveRegionScaleFactors=(PositiveWidth=1,NegativeWidth=1,PositiveHeight=1,NegativeHeight=1)
    ActiveRegionSize=0.15
    Debug=(bDrawDebug=True,bDraw2D=False,bDrawLocalAxes=False,DebugScale=25,DebugSegments=20,DebugThickness=0.25)
    DriverAxis=(X=-1,Y=0,Z=0)
    DriverItem=(Type=Bone,Name="upperarm_r")
    FalloffRegionScaleFactors=(PositiveWidth=1,NegativeWidth=1,PositiveHeight=1,NegativeHeight=1)
    FalloffSize=0.3
    FlipHeightScaling=False
    FlipWidthScaling=False
    OptionalParentItem=(Type=Bone,Name="None")
    RotationOffset=(X=-110,Y=20,Z=40)
   )
   RigVMFunction_MathBoolToFloat = &RigVMFunction_MathBoolToFloat(Value=RigVMVariableNode() {variable=Crouching}.Value)
   Multiply = RigVMAggregateNode(D=@RigVMFunction_MathBoolToFloat.Result, A=@n6.OutputParam, B=@n5.OutputParam)
   n7 = &RigVMFunction_AlphaInterp(InterpSpeedDecreasing=8, InterpSpeedIncreasing=8, Scale=1, bClampResult=False, bInterpResult=False, Value=@Multiply.Result)
   > &RigUnit_ModifyTransforms(
    ItemToModify=((Item=(Type=Bone,Name="dyn_hair_bk_base"),Transform=(Rotation=(X=0,Y=0,Z=0,W=1),Translation=(X=-4,Y=0,Z=2),Scale3D=(X=1,Y=1,Z=1))))
    WeightMinimum=0
    Weight=@n7.Result
   )
   n8 = &RigVMFunction_AlphaInterp(InterpSpeedDecreasing=8, InterpSpeedIncreasing=8, Scale=1, bClampResult=False, bInterpResult=False, Value=@n6.OutputParam)
   > &RigUnit_ModifyTransforms(
    ItemToModify="((Item=(Type=Bone,Name=\"dyn_curl__side_r_base\"),Transform=(Rotation=(X=0.1919,Y=-0.1472,Z=0.1565,W=0.9576),Translation=(X=0,Y=0,Z=1),Scale3D=(X=1,Y=1,Z=1))),(Item=(Type=Bone,Name=\"dyn_hair_side_r_base\"),Transform=(Rotation=(X=0.0274,Y=0.362,Z=0.0725,W=0.92 ...(truncated)"
    WeightMinimum=0
    Weight=@n8.Result
   )
   > &RigUnit_ModifyTransforms(
    ItemToModify=((Item=(Type=Bone,Name="dyn_hair_bk_base"),Transform=(Rotation=(X=0,Y=0,Z=0,W=1),Translation=(X=0,Y=0,Z=4),Scale3D=(X=1,Y=1,Z=1))))
    WeightMinimum=0
    Weight=&RigVMFunction_MathBoolToFloat(Value=RigVMVariableNode() {variable=Is Skydive Floating}.Value).Result
   )
   n9 = &RigUnit_SphericalPoseReader(
    ActiveRegionScaleFactors=(PositiveWidth=1,NegativeWidth=1,PositiveHeight=1,NegativeHeight=1)
    ActiveRegionSize=0.15
    Debug=(bDrawDebug=True,bDraw2D=False,bDrawLocalAxes=False,DebugScale=25,DebugSegments=20,DebugThickness=0.25)
    DriverAxis=(X=-1,Y=0,Z=0)
    DriverItem=(Type=Bone,Name="upperarm_r")
    FalloffRegionScaleFactors=(PositiveWidth=1,NegativeWidth=1,PositiveHeight=1,NegativeHeight=1)
    FalloffSize=0.3
    FlipHeightScaling=False
    FlipWidthScaling=False
    OptionalParentItem=(Type=Bone,Name="None")
    RotationOffset=(X=-110,Y=20,Z=40)
    ExecutePin=@RigVMFunction_Sequence.B
   )
   n10 = &RigVMFunction_AlphaInterp(InterpSpeedDecreasing=8, InterpSpeedIncreasing=8, Scale=1, bClampResult=False, bInterpResult=False, Value=@n9.OutputParam)
   > n11 = &RigUnit_ModifyTransforms(
    ItemToModify="((Item=(Type=Bone,Name=\"dyn_curl__side_r_base\"),Transform=(Rotation=(X=0,Y=0,Z=0,W=1),Translation=(X=0,Y=0,Z=0),Scale3D=(X=1,Y=1,Z=1))),(Item=(Type=Bone,Name=\"dyn_hair_side_r_base\"),Transform=(Rotation=(X=0,Y=0,Z=0,W=1),Translation=(X=0,Y=0,Z=0),Scale3D=(X ...(truncated)"
    WeightMinimum=0
    Weight=@n10.Result
   )
   RigVMFunction_ControlFlowBranch = &RigVMFunction_ControlFlowBranch(ExecuteContext=@n11.ExecutePin, Condition=RigVMVariableNode() {variable=Is Backbling}.Value)
   &RigUnit_ParentConstraint(
    AdvancedSettings=(InterpolationType=Average,RotationOrderForFilter=XZY)
    Child=(Type=Null,Name="dyn_hair_bk_01_null")
    Filter=(TranslationFilter=(bX=true,bY=true,bZ=true),RotationFilter=(bX=true,bY=true,bZ=true),ScaleFilter=(bX=True,bY=True,bZ=True))
    Parents=((Item=(Type=Bone,Name="head"),Weight=1),(Item=(Type=Bone,Name="spine_04"),Weight=1))
    Weight=1
    ExecutePin=@RigVMFunction_ControlFlowBranch.True
   )
   > &RigUnit_ParentConstraint(
    AdvancedSettings=(InterpolationType=Average,RotationOrderForFilter=XZY)
    Child=(Type=Null,Name="dyn_hair_bk_02_null")
    Parents=((Item=(Type=Bone,Name="head"),Weight=0),(Item=(Type=Bone,Name="spine_04"),Weight=1),(Item=(Type=Bone,Name="clavicle_l"),Weight=1),(Item=(Type=Bone,Name="clavicle_r"),Weight=1))
    Weight=1
   )
   > &RigUnit_ParentConstraint(
    AdvancedSettings=(InterpolationType=Average,RotationOrderForFilter=XZY)
    Child=(Type=Null,Name="dyn_hair_bk_03_null")
    Parents=((Item=(Type=Bone,Name="head"),Weight=0),(Item=(Type=Bone,Name="spine_04"),Weight=1),(Item=(Type=Bone,Name="clavicle_l"),Weight=1),(Item=(Type=Bone,Name="clavicle_r"),Weight=1))
    Weight=1
   )
   > &RigUnit_ParentConstraint(
    AdvancedSettings=(InterpolationType=Average,RotationOrderForFilter=XZY)
    Child=(Type=Null,Name="dyn_hair_bk_04_null")
    Filter=(TranslationFilter=(bX=true,bY=true,bZ=true),RotationFilter=(bX=true,bY=true,bZ=true),ScaleFilter=(bX=True,bY=True,bZ=True))
    Parents=((Item=(Type=Bone,Name="head"),Weight=0),(Item=(Type=Bone,Name="spine_03"),Weight=1))
    Weight=1
   )
   ItemArray_2 = RigVMDispatchNode(Value="((Type=Bone,Name=\"dyn_hair_bk_01\"),(Type=Bone,Name=\"dyn_hair_bk_02\"),(Type=Bone,Name=\"dyn_hair_bk_03\"),(Type=Bone,Name=\"dyn_hair_bk_04\"),(Type=None,Name=\"None\"),(Type=None,Name=\"None\"),(Type=None,Name=\"None\"),(Type=None,Name=\"None\"),(Type=None,Name=\"None\") ...(truncated)") {template="DISPATCH_RigVMDispatch_Constant(io Value)"}
   > TwistBones_1 = RigVMFunctionReferenceNode(
    Axis to Twist=(bX=True,bY=true,bZ=true)
    Debug=false
    End=(Type=Bone,Name="neck_01")
    Start=(Type=Bone,Name="pelvis")
    Bones to Twist=@ItemArray_2.Value
   )
   Make_Array_1 = RigVMDispatchNode(Values="(((),Type=Bone,Name=\"dyn_hair_bk_01\"),((),Type=Bone,Name=\"dyn_hair_bk_02\"),((),Type=Bone,Name=\"dyn_hair_bk_03\"),((),Type=Bone,Name=\"dyn_hair_bk_04\"),((),Type=None,Name=\"None\"),((),Type=None,Name=\"None\"),((),Type=None,Name=\"None\"),((),Type=None,Name=\"None\") ...(truncated)") {template="DISPATCH_RigVMDispatch_ArrayMake(in Values,out Array)"}
   > SplineIK = RigVMFunctionReferenceNode(
    Primary Axis=(X=0.1,Y=0,Z=0)
    Samples Per Segment=3
    Secondary Spline Direction=(X=0,Y=0,Z=0)
    Spline Mode=Hermite
    Stretch=false
    Up Axis=(X=0,Y=0,Z=0)
    debug=false
    Bones=@Make_Array_1.Array
    Controls=RigVMVariableNode() {variable=Dyn Hair Ctls}.Value
   )
  ]}
 RigVMModel Constraction {nodes [
   n1 = &RigUnit_PrepareForExecution()
   Make_Array_1 = RigVMDispatchNode(Values=(((),Type=Control,Name="dyn_hair_bk_01_ctrl"),((),Type=Control,Name="dyn_hair_bk_02_ctrl"),((),Type=Control,Name="dyn_hair_bk_03_ctrl"),((),Type=Control,Name="dyn_hair_bk_04_ctrl"))) {template="DISPATCH_RigVMDispatch_ArrayMake(in Values,out Array)"}
   RigVMVariableNode(Value=@Make_Array_1.Array, ExecuteContext=@n1.ExecutePin) {variable=Dyn Hair Ctls}
  ]}
]
hierarchy [
 root {loc=<0 0 0>, rot=<0 0 0 1>}
 pelvis {parent=root, loc=<0 -0.5301 98.0571>, rot=<0.0127 0.707 -0.0127 -0.707>}
 spine_01 {parent=pelvis, loc=<6.1208 -0.9328 0>, rot=<0 0 0.034 -0.9994>}
 spine_02 {parent=spine_01, loc=<6.5997 -0.2095 0>, rot=<0 0 0.0238 -0.9997>}
 spine_03 {parent=spine_02, loc=<6.4914 -0.1237 0>, rot=<0 0 0.0149 -0.9999>}
 spine_04 {parent=spine_03, loc=<6.9598 0.0794 0>, rot=<0 0 -0.0132 -0.9999>}
 spine_05 {parent=spine_04, loc=<6.9987 0 0>, rot=<0 0 -0.0015 -1>}
 neck_01 {parent=spine_05, loc=<11.3138 0.1529 0>, rot=<0 0 -0.2003 -0.9797>}
 neck_02 {parent=neck_01, loc=<4.5627 0 0>, rot=<0 0 0.0504 -0.9987>}
 head {parent=neck_02, loc=<5.3914 0 0>, rot=<0 0 0.115 -0.9934>}
 dyn_bang_01 {parent=head, loc=<7.4342 11.126 1.8662>, rot=<0.0034 -0.8854 -0.4624 0.0464>}
 dyn_bang_02 {parent=head, loc=<6.2326 12.3277 4.357>, rot=<0.0309 -0.9329 -0.3498 0.0803>}
 dyn_curl_bk_02 {parent=head, loc=<15.6644 -6.2518 4.0097>, rot=<0.7315 -0.3719 0.5217 -0.2333>}
 dyn_curl_bk_03 {parent=head, loc=<18.6869 -0.989 3.1312>, rot=<0.7745 -0.019 0.6321 0.0159>}
 dyn_curl_side_r_01 {parent=head, loc=<13.0637 1.4095 10.3846>, rot=<0.494 0.0341 0.8489 -0.185>}
 dyn_curl_side_r_04 {parent=head, loc=<5.3573 4.799 12.3701>, rot=<-0.0334 -0.9625 0.2022 0.1776>}
 dyn_curl_bk_01 {parent=head, loc=<12.9739 -11.5857 3.5647>, rot=<0.6013 -0.4417 0.5861 -0.316>}
 dyn_pigtail_base {parent=head, loc=<8.7928 -11.5243 -5.5385>, rot=<-0.0628 -0.4818 0.8736 -0.0279>}
 dyn_pigtail_01 {parent=dyn_pigtail_base, loc=<0 0 0>, rot=<0 0 0 1>}
 dyn_pigtail_02 {parent=dyn_pigtail_01, loc=<6.3367 0 0>, rot=<0 0 0 1>}
 dyn_pigtail_03 {parent=dyn_pigtail_02, loc=<5.8835 0 0>, rot=<0 0 0 1>}
 dyn_pigtail_04 {parent=dyn_pigtail_03, loc=<6.1792 0 0>, rot=<0 0 0 1>}
 dyn_pigtail_05 {parent=dyn_pigtail_04, loc=<6.3247 0 0>, rot=<0 0 0 1>}
 dyn_pigtail_06 {parent=dyn_pigtail_05, loc=<5.894 0 0>, rot=<0 0 0 1>}
 dyn_pigtail_07 {parent=dyn_pigtail_06, loc=<6.6423 0 0>, rot=<0 0 0 1>}
 dyn_curl_pigtail_01 {parent=dyn_pigtail_04, loc=<0.9766 -0.927 2.1711>, rot=<0.7834 -0.1882 0.4364 -0.4005>}
 dyn_curl_pigtail_03 {parent=dyn_pigtail_02, loc=<1.0723 -0.96 1.4716>, rot=<0.696 -0.0679 0.3959 -0.5952>}
 dyn_curl_pigtail_04 {parent=dyn_pigtail_02, loc=<-0.9179 -1.3695 -0.8998>, rot=<0.5993 -0.718 -0.3504 -0.0497>}
 dyn_curl_pigtail_02 {parent=dyn_pigtail_01, loc=<3.2406 -0.0037 1.7155>, rot=<0.71 0.3531 -0.018 -0.6091>}
 dyn_curl__side_r_base {parent=head, loc=<2.528 -2.9336 11.4633>, rot=<0.0148 -0.9976 0.0086 0.0674>}
 dyn_curl__side_r_01 {parent=dyn_curl__side_r_base, loc=<0 0 0>, rot=<0 0 0 1>}
 dyn_curl__side_r_02 {parent=dyn_curl__side_r_01, loc=<5.5934 -0.735 0.0171>, rot=<0.1086 -0.0613 0.0522 -0.9908>}
 dyn_curl__side_r_03 {parent=dyn_curl__side_r_01, loc=<9.5804 -0.6648 0.214>, rot=<-0.1985 -0.3761 -0.1449 -0.8934>}
 dyn_hair_frnt_base {parent=head, loc=<2.96 7.9681 8.0382>, rot=<0.0208 -0.9692 0.2348 0.0712>}
 dyn_hair_frnt_01 {parent=dyn_hair_frnt_base, loc=<0 0 0>, rot=<0 0 0 1>}
 dyn_hair_frnt_02 {parent=dyn_hair_frnt_01, loc=<3.0586 1.9788 -1.4972>, rot=<-0.2124 0.0155 0.046 -0.976>}
 dyn_hair_frnt_03 {parent=dyn_hair_frnt_01, loc=<1.1817 -2.391 -2.7075>, rot=<0.0963 -0.1186 0.0605 -0.9864>}
 dyn_hair_frnt_04 {parent=dyn_hair_frnt_01, loc=<6.5363 0.1089 -0.0951>, rot=<0.0384 -0.1234 0.0112 -0.9916>}
 dyn_hair_side_r_base {parent=head, loc=<-0.8694 2.8787 9.7686>, rot=<0.0358 -0.9914 0.1252 0.0102>}
 dyn_hair_side_r_01 {parent=dyn_hair_side_r_base, loc=<0 0 0>, rot=<0 0 0 1>}
 dyn_hair_side_r_02 {parent=dyn_hair_side_r_01, loc=<4.5917 -0.0365 -0.0317>, rot=<0.008 0.0614 -0.0795 -0.9949>}
 dyn_hair_side_r_06 {parent=dyn_hair_side_r_02, loc=<4.9639 -1.6311 -0.7884>, rot=<-0.5864 -0.3258 0.1429 -0.7277>}
 dyn_hair_side_r_05 {parent=dyn_hair_side_r_02, loc=<4.3152 0.3713 -0.9998>, rot=<-0.6616 -0.111 -0.1086 -0.7336>}
 dyn_hair_side_r_04 {parent=dyn_hair_side_r_02, loc=<1.0069 0.8681 1.4729>, rot=<-0.2978 0.0035 -0.0504 -0.9533>}
 dyn_curl_side_r_03 {parent=dyn_hair_side_r_01, loc=<0.571 -2.1072 -2.1509>, rot=<0.0268 -0.163 -0.0328 -0.9857>}
 dyn_hair_bk_base {parent=head, loc=<1.7344 -11.1568 3.6223>, rot=<0.134 -0.9907 0.0002 -0.0251>}
 dyn_curl_bk_04 {parent=dyn_hair_bk_base, loc=<2.5918 -1.622 -5.6437>, rot=<-0.3051 -0.0802 0.0841 -0.9452>}
 dyn_hair_bk_01 {parent=dyn_hair_bk_base, loc=<0.0264 0.0032 0>, rot=<0 0 0 1>}
 dyn_hair_bk_02 {parent=dyn_hair_bk_01, loc=<7.3404 1.9673 -1.1484>, rot=<0 0 -0.0317 -0.9995>}
 dyn_hair_bk_03 {parent=dyn_hair_bk_02, loc=<7.513 0.4528 -0.5696>, rot=<0 0 -0.1296 -0.9916>}
 dyn_hair_bk_04 {parent=dyn_hair_bk_03, loc=<7.0336 -1.0468 0>, rot=<0 0 0 1>}
 dyn_curl_bk_10 {parent=dyn_hair_bk_04, loc=<9.9825 -1.2309 5.4902>, rot=<-0.0383 0.0454 0.0582 -0.9965>}
 dyn_curl_bk_11 {parent=dyn_hair_bk_04, loc=<10.3936 -0.1146 2.5126>, rot=<-0.0272 -0.0273 0.1094 -0.9933>}
 dyn_curl_bk_12 {parent=dyn_hair_bk_04, loc=<9.7944 -2.4743 0.4967>, rot=<-0.0272 -0.0273 0.1094 -0.9933>}
 dyn_curl_bk_13 {parent=dyn_hair_bk_04, loc=<8.1441 -3.1797 -1.7074>, rot=<-0.0296 -0.0247 0.0173 -0.9991>}
 dyn_curl_bk_08 {parent=dyn_hair_bk_04, loc=<6.3955 2.5124 -4.7429>, rot=<-0.085 -0.1239 0.1278 -0.9804>}
 dyn_curl_bk_09 {parent=dyn_hair_bk_03, loc=<11.2491 1.0721 -7.592>, rot=<-0.1086 -0.214 -0.1122 -0.9643>}
 dyn_curl_bk_07 {parent=dyn_hair_bk_03, loc=<7.7992 -2.3861 -8.1438>, rot=<-0.2877 -0.0189 0.1255 -0.9493>}
 dyn_curl_bk_14 {parent=dyn_hair_bk_03, loc=<7.1055 0.3543 8.4221>, rot=<0.1954 0.0783 0.0576 -0.9759>}
 dyn_curl_bk_16 {parent=dyn_hair_bk_03, loc=<6.2241 -3.8864 2.5904>, rot=<-0.0226 -0.0312 0.2609 -0.9646>}
 dyn_curl_bk_06 {parent=dyn_hair_bk_02, loc=<12.117 -2.8386 -4.983>, rot=<-0.077 -0.0547 0.0618 -0.9936>}
 dyn_curl_bk_17 {parent=dyn_hair_bk_02, loc=<7.1214 -2.6808 -1.1168>, rot=<-0.1276 -0.0808 0.1414 -0.9784>}
 dyn_curl_bk_18 {parent=dyn_hair_bk_02, loc=<10.8238 1.8954 8.2583>, rot=<0.1532 0.1925 0.0527 -0.9678>}
 dyn_curl_bk_15 {parent=dyn_hair_bk_02, loc=<9.4523 -1.8047 4.0388>, rot=<-0.0184 -0.0339 0.1337 -0.9903>}
 dyn_curl_bk_05 {parent=dyn_hair_bk_01, loc=<7.0268 1.8164 -8.679>, rot=<-0.3691 -0.2852 0.3684 -0.8042>}
 dyn_curl_bk_19 {parent=dyn_hair_bk_01, loc=<11.0337 1.8926 4.7128>, rot=<0.1486 0.3146 0.0037 -0.9375>}
 dyn_curl_bk_20 {parent=dyn_hair_bk_01, loc=<12.1038 5.4322 7.738>, rot=<0.1376 0.2039 0.0675 -0.9669>}
 hair_r_support {parent=dyn_hair_bk_base, loc=<7.5687 5.0326 -7.6743>, rot=<-0.3309 -0.5814 0.4895 -0.5594>}
 clavicle_l {parent=spine_05, loc=<5.4831 -0.1053 -3.2946>, rot=<0.0465 -0.7289 -0.009 -0.6829>}
 upperarm_l {parent=clavicle_l, loc=<9.8658 -1.2136 -0.4105>, rot=<0.0164 -0.3245 -0.0269 -0.9454>}
 lowerarm_l {parent=upperarm_l, loc=<26.465 -0.8345 0.259>, rot=<0.0036 0.0161 -0.2298 -0.9731>}
 hand_l {parent=lowerarm_l, loc=<22.7355 -0.3203 -0.6105>, rot=<-0.6784 -0.0223 0.0735 -0.7307>}
 thumb_01_l {parent=hand_l, loc=<1.3118 -0.3191 -2.459>, rot=<0.6546 -0.4042 0.0664 -0.6354>}
 thumb_02_l {parent=thumb_01_l, loc=<5.9336 -0.0448 0.0401>, rot=<-0.0037 0.0213 0.298 -0.9543>}
 thumb_03_l {parent=thumb_02_l, loc=<2.2223 0.3682 -0.0907>, rot=<0.0118 -0.013 0.0066 -0.9998>}
 index_metacarpal_l {parent=hand_l, loc=<1.6535 0.0183 -2.96>, rot=<0.0158 -0.0124 -0.0173 -0.9996>}
 index_01_l {parent=index_metacarpal_l, loc=<8.3008 0.0357 -0.0524>, rot=<-0.0044 0.0485 0.0924 -0.9945>}
 index_02_l {parent=index_01_l, loc=<3.3587 -0.3004 0.0452>, rot=<-0.002 0.0027 0.3225 -0.9466>}
 index_03_l {parent=index_02_l, loc=<2.5578 -0.123 -0.0252>, rot=<-0.002 0.0067 0.2558 -0.9667>}
 middle_metacarpal_l {parent=hand_l, loc=<2.1856 -0.1725 -0.5984>, rot=<0.006 0.0168 -0.0329 -0.9993>}
 middle_01_l {parent=middle_metacarpal_l, loc=<7.3275 -0.2447 0.0129>, rot=<-0.0208 0.0042 0.0858 -0.9961>}
 middle_02_l {parent=middle_01_l, loc=<4.0123 0.1401 0.1175>, rot=<-0.0218 0.0258 0.3512 -0.9357>}
 middle_03_l {parent=middle_02_l, loc=<2.6077 0.0534 0.1373>, rot=<0.0082 0.0188 0.1708 -0.9851>}
 ring_metacarpal_l {parent=hand_l, loc=<2.2074 -0.3625 1.3638>, rot=<-0.0284 0.0722 -0.0458 -0.9959>}
 ring_01_l {parent=ring_metacarpal_l, loc=<7.0295 -0.6197 -0.1446>, rot=<-0.0489 -0.0252 0.1892 -0.9804>}
 ring_02_l {parent=ring_01_l, loc=<3.6174 0.4647 -0.1027>, rot=<0 0 0.2806 -0.9598>}
 ring_03_l {parent=ring_02_l, loc=<2.4152 0.114 0.0199>, rot=<0 0 0.1189 -0.9929>}
 pinky_metacarpal_l {parent=hand_l, loc=<1.7616 -0.9753 3.3814>, rot=<-0.0397 0.0814 -0.069 -0.9935>}
 pinky_01_l {parent=pinky_metacarpal_l, loc=<7.1709 -0.4831 -0.1094>, rot=<-0.0654 -0.0035 0.2242 -0.9723>}
 pinky_02_l {parent=pinky_01_l, loc=<2.8274 0.0893 -0.1043>, rot=<0.0037 0.0011 0.2854 -0.9584>}
 pinky_03_l {parent=pinky_02_l, loc=<2.2046 0.0307 -0.0078>, rot=<-0.0002 0.0098 0.1445 -0.9895>}
 deform_wrist_dn_base_l {parent=hand_l, loc=<0 -1.5326 0>, rot=<0 0 0 -1>}
 deform_wrist_dn_l {parent=deform_wrist_dn_base_l, loc=<0 0 0>, rot=<0 0 0 -1>}
 deform_wrist_up_base_l {parent=hand_l, loc=<0 2.3009 0>, rot=<0 0 0 -1>}
 deform_wrist_up_l {parent=deform_wrist_up_base_l, loc=<0 0 0>, rot=<0 0 0 -1>}
 hand_attach_l {parent=hand_l, loc=<8.5 -3 0>, rot=<0.7071 0.7071 0 0>}
 weapon_l {parent=hand_l, loc=<-0.1974 -3.4111 0.3811>, rot=<0 0 0 1>}
 lowerarm_twist_02_l {parent=lowerarm_l, loc=<10.6707 0 0>, rot=<0 0 0 1>}
 lowerarm_twist_01_l {parent=lowerarm_l, loc=<18.316 0 0>, rot=<0 0 0 1>}
 deform_elbow_in_base_l {parent=lowerarm_l, loc=<0.1834 3.6062 -0.2572>, rot=<0 0 0 -1>}
 deform_elbow_in_l {parent=deform_elbow_in_base_l, loc=<0 0 0>, rot=<0 0 0 -1>}
 deform_elbow_out_base_l {parent=lowerarm_l, loc=<0 -4 0>, rot=<0 0 0 1>}
 deform_elbow_out_l {parent=deform_elbow_out_base_l, loc=<0 0 0>, rot=<0 0 0 1>}
 upperarm_twist_01_l {parent=upperarm_l, loc=<7.2135 -0.1134 0>, rot=<0 0 0 1>}
 deform_upperarm_bk_base_l {parent=upperarm_twist_01_l, loc=<-6.4053 -4.3746 0.314>, rot=<0 0 0 -1>}
 deform_upperarm_bk_l {parent=deform_upperarm_bk_base_l, loc=<0 0 0>, rot=<0 0 0 -1>}
 deform_upperarm_fr_base_l {parent=upperarm_twist_01_l, loc=<-5.6312 5.092 0.3354>, rot=<0 0 0 -1>}
 deform_upperarm_fr_l {parent=deform_upperarm_fr_base_l, loc=<0 0 0>, rot=<0 0 0 -1>}
 deform_upperarm_up_base_l {parent=upperarm_twist_01_l, loc=<-8.891 0.4825 3.9517>, rot=<0 0 0 -1>}
 deform_upperarm_up_l {parent=deform_upperarm_up_base_l, loc=<0 0 0>, rot=<0 0 0 -1>}
 dyn_shoulder_l_01 {parent=upperarm_twist_01_l, loc=<-2.7365 -7.0535 0.2723>, rot=<0 -0.2876 0 -0.9578>}
 dyn_shoulder_l_03 {parent=upperarm_twist_01_l, loc=<-3.2461 -1.4034 7.7726>, rot=<-0.0425 -0.1776 0.025 -0.9829>}
 upperarm_twist_02_l {parent=upperarm_l, loc=<19.7998 -0.1134 0>, rot=<0 0 0 1>}
 support_shoulder_l_rotator {parent=upperarm_l, loc=<0 0 0>, rot=<0 0 0 1>}
 support_shoulder_l {parent=support_shoulder_l_rotator, loc=<-4.2626 0 5.1178>, rot=<0 0.5337 0 -0.8457>}
 dyn_shoulder_l_02 {parent=support_shoulder_l_rotator, loc=<6.7127 6.1475 2.3883>, rot=<-0.244 -0.3532 0.0351 -0.9025>}
 clavicle_r {parent=spine_05, loc=<5.4831 -0.1053 3.2946>, rot=<0.7289 0.0465 -0.6829 0.009>}
 upperarm_r {parent=clavicle_r, loc=<-9.8658 1.2136 0.4105>, rot=<0.0164 -0.3245 -0.0269 -0.9454>}
 lowerarm_r {parent=upperarm_r, loc=<-26.465 0.8345 -0.259>, rot=<0.0036 0.0161 -0.2298 -0.9731>}
 hand_r {parent=lowerarm_r, loc=<-22.7355 0.3203 0.6106>, rot=<-0.6784 -0.0223 0.0735 -0.7307>}
 thumb_01_r {parent=hand_r, loc=<-1.3118 0.3191 2.459>, rot=<0.6546 -0.4042 0.0664 -0.6354>}
 thumb_02_r {parent=thumb_01_r, loc=<-5.9336 0.0448 -0.0401>, rot=<-0.0037 0.0213 0.298 -0.9543>}
 thumb_03_r {parent=thumb_02_r, loc=<-2.2223 -0.3682 0.0907>, rot=<0.0118 -0.013 0.0066 -0.9998>}
 index_metacarpal_r {parent=hand_r, loc=<-1.6535 -0.0183 2.96>, rot=<0.0158 -0.0124 -0.0173 -0.9996>}
 index_01_r {parent=index_metacarpal_r, loc=<-8.3008 -0.0357 0.0524>, rot=<-0.0044 0.0485 0.0924 -0.9945>}
 index_02_r {parent=index_01_r, loc=<-3.3587 0.3004 -0.0452>, rot=<-0.002 0.0027 0.3225 -0.9466>}
 index_03_r {parent=index_02_r, loc=<-2.5578 0.123 0.0252>, rot=<-0.002 0.0067 0.2558 -0.9667>}
 middle_metacarpal_r {parent=hand_r, loc=<-2.1856 0.1725 0.5984>, rot=<0.006 0.0168 -0.0329 -0.9993>}
 middle_01_r {parent=middle_metacarpal_r, loc=<-7.3275 0.2447 -0.0129>, rot=<-0.0208 0.0042 0.0858 -0.9961>}
 middle_02_r {parent=middle_01_r, loc=<-4.0123 -0.1401 -0.1175>, rot=<-0.0218 0.0258 0.3512 -0.9357>}
 middle_03_r {parent=middle_02_r, loc=<-2.6077 -0.0534 -0.1373>, rot=<0.0082 0.0188 0.1708 -0.9851>}
 ring_metacarpal_r {parent=hand_r, loc=<-2.2074 0.3625 -1.3638>, rot=<-0.0284 0.0722 -0.0458 -0.9959>}
 ring_01_r {parent=ring_metacarpal_r, loc=<-7.0295 0.6197 0.1446>, rot=<-0.0489 -0.0252 0.1892 -0.9804>}
 ring_02_r {parent=ring_01_r, loc=<-3.6174 -0.4647 0.1027>, rot=<0 0 0.2806 -0.9598>}
 ring_03_r {parent=ring_02_r, loc=<-2.4152 -0.114 -0.0199>, rot=<0 0 0.1189 -0.9929>}
 pinky_metacarpal_r {parent=hand_r, loc=<-1.7616 0.9753 -3.3814>, rot=<-0.0397 0.0814 -0.069 -0.9935>}
 pinky_01_r {parent=pinky_metacarpal_r, loc=<-7.1709 0.4831 0.1094>, rot=<-0.0654 -0.0035 0.2242 -0.9723>}
 pinky_02_r {parent=pinky_01_r, loc=<-2.8274 -0.0893 0.1043>, rot=<0.0037 0.0011 0.2854 -0.9584>}
 pinky_03_r {parent=pinky_02_r, loc=<-2.2046 -0.0307 0.0078>, rot=<-0.0002 0.0098 0.1445 -0.9895>}
 deform_wrist_dn_base_r {parent=hand_r, loc=<-0.0001 1.5328 0>, rot=<0 0 0 -1>}
 deform_wrist_dn_r {parent=deform_wrist_dn_base_r, loc=<0 0 0>, rot=<0 0 0 -1>}
 deform_wrist_up_base_r {parent=hand_r, loc=<-0.0002 -2.3007 -0.0001>, rot=<0 0 0 -1>}
 deform_wrist_up_r {parent=deform_wrist_up_base_r, loc=<0 0 0>, rot=<0 0 0 -1>}
 weapon_r {parent=hand_r, loc=<0.1974 3.4111 -0.3811>, rot=<0 0 0 1>}
 hand_attach_r {parent=hand_r, loc=<-8.5001 2.9999 0>, rot=<0 0 -0.7071 -0.7071>}
 lowerarm_twist_02_r {parent=lowerarm_r, loc=<-10.6707 0 0>, rot=<0 0 0 -1>}
 lowerarm_twist_01_r {parent=lowerarm_r, loc=<-18.316 0 0>, rot=<0 0 0 -1>}
 deform_elbow_in_base_r {parent=lowerarm_r, loc=<-0.1834 -3.6062 0.2572>, rot=<0 0 0 -1>}
 deform_elbow_in_r {parent=deform_elbow_in_base_r, loc=<0 0 0>, rot=<0 0 0 -1>}
 deform_elbow_out_base_r {parent=lowerarm_r, loc=<0 4 -0.0001>, rot=<0 0 0 -1>}
 deform_elbow_out_r {parent=deform_elbow_out_base_r, loc=<0 0 0>, rot=<0 0 0 1>}
 upperarm_twist_01_r {parent=upperarm_r, loc=<-7.2135 0.1134 0>, rot=<0 0 0 -1>}
 deform_upperarm_bk_base_r {parent=upperarm_twist_01_r, loc=<6.4053 4.3746 -0.314>, rot=<0 0 0 -1>}
 deform_upperarm_bk_r {parent=deform_upperarm_bk_base_r, loc=<0 0 0>, rot=<0 0 0 -1>}
 deform_upperarm_fr_base_r {parent=upperarm_twist_01_r, loc=<5.6312 -5.092 -0.3354>, rot=<0 0 0 -1>}
 deform_upperarm_fr_r {parent=deform_upperarm_fr_base_r, loc=<0 0 0>, rot=<0 0 0 -1>}
 deform_upperarm_up_base_r {parent=upperarm_twist_01_r, loc=<8.891 -0.4825 -3.9517>, rot=<0 0 0 -1>}
 deform_upperarm_up_r {parent=deform_upperarm_up_base_r, loc=<0 0 0>, rot=<0 0 0 -1>}
 upperarm_twist_02_r {parent=upperarm_r, loc=<-19.7998 0.1134 0>, rot=<0 0 0 -1>}
 support_shoulder_r_rotator {parent=upperarm_r, loc=<0 0 0>, rot=<0 0 0 1>}
 support_shoulder_r {parent=support_shoulder_r_rotator, loc=<4.8155 0 -4.1954>, rot=<0 -0.4857 0 -0.8741>}
 attach_backpack {parent=spine_05, loc=<-0.123 -13.473 0>, rot=<-0.0093 -0.707 0.0093 -0.707>}
 attach_cape {parent=spine_05, loc=<9.017 -11.714 0>, rot=<0.0616 -0.7044 -0.0616 -0.7044>}
 dyn_necklace_bottom_base {parent=spine_05, loc=<1.2955 9.4012 0>, rot=<-0.2733 -0.9619 0 0>}
 dyn_necklace_bottom_01 {parent=dyn_necklace_bottom_base, loc=<-0.0009 0 0>, rot=<0 0 0 1>}
 dyn_necklace_bottom_02 {parent=dyn_necklace_bottom_01, loc=<9.1602 -0.5918 -1.6452>, rot=<-0.0209 0.0074 0.3331 -0.9426>}
 deform_pec_aim_base_l {parent=spine_04, loc=<3.2456 9.3309 -1.2237>, rot=<0.0603 -0.4033 0.3318 -0.8507>}
 deform_pec_aim_l {parent=deform_pec_aim_base_l, loc=<0 0 0>, rot=<0 0 0 1>}
 deform_pec_base_l {parent=deform_pec_aim_l, loc=<9.3665 0 0>, rot=<0 0 0 1>}
 deform_pec_l {parent=deform_pec_base_l, loc=<0 0 0>, rot=<0 0 0 -1>}
 deform_pec_aim_base_r {parent=spine_04, loc=<3.2452 9.3309 1.2237>, rot=<-0.4033 -0.0603 0.8507 0.3318>}
 deform_pec_aim_r {parent=deform_pec_aim_base_r, loc=<0 0 0>, rot=<0 0 0 1>}
 deform_pec_base_r {parent=deform_pec_aim_r, loc=<-9.3667 -0.0003 -0.0004>, rot=<0 0 0 1>}
 deform_pec_r {parent=deform_pec_base_r, loc=<0 0 0>, rot=<0 0 0 -1>}
 dyn_torso_l_base {parent=spine_04, loc=<0.5 5 -5.2>, rot=<-0.5 -0.5 -0.5 -0.5>}
 dyn_torso_l {parent=dyn_torso_l_base, loc=<0 0 0>, rot=<0 0 0 1>}
 dyn_torso_r_base {parent=spine_04, loc=<0.5 5 5.2>, rot=<-0.5 -0.5 -0.5 -0.5>}
 dyn_torso_r {parent=dyn_torso_r_base, loc=<0 0 0>, rot=<0 0 0 1>}
 thigh_l {parent=pelvis, loc=<-1.0855 -0.2109 -8.7053>, rot=<0.0009 0.0355 0.0155 -0.9993>}
 calf_l {parent=thigh_l, loc=<-42.5284 -0.0019 -0.7297>, rot=<0.0006 -0.0161 -0.0391 -0.9991>}
 foot_l {parent=calf_l, loc=<-46.7401 -2.5014 -0.592>, rot=<0.0427 -0.0238 0.0353 -0.9982>}
 ball_l {parent=foot_l, loc=<-3.5658 10.8066 -0.1544>, rot=<0.0001 -0.0001 -0.7063 -0.7079>}
 support_foot_l {parent=foot_l, loc=<3.5 -6 0>, rot=<0 0 0 1>}
 calf_twist_02_l {parent=calf_l, loc=<-19.1546 -1.7943 -0.2638>, rot=<0.0237 0.0095 -0.0035 -0.9997>}
 calf_twist_01_l {parent=calf_l, loc=<-35.8483 -1.9197 -0.5794>, rot=<0.0237 0.0095 -0.0035 -0.9997>}
 dyn_boot_fur_l_01 {parent=calf_twist_01_l, loc=<3.3435 2.2585 -4.8484>, rot=<-0.0502 -0.9748 -0.1634 -0.1437>}
 dyn_boot_fur_l_02 {parent=calf_twist_01_l, loc=<3.564 -3.4498 4.2088>, rot=<0.1322 -0.3135 0.9402 -0.0151>}
 dyn_boot_fur_l_03 {parent=calf_twist_01_l, loc=<2.8837 -5.7173 -2.6655>, rot=<0.0654 -0.839 0.5367 -0.0611>}
 dyn_boot_fur_l_04 {parent=calf_twist_01_l, loc=<3.7081 5.0066 1.6076>, rot=<0.2119 0.5144 0.8173 0.1498>}
 deform_knee_in_base_l {parent=calf_l, loc=<0.8031 -5.0231 0.2456>, rot=<0 0 0 -1>}
 deform_knee_in_l {parent=deform_knee_in_base_l, loc=<0 0 0>, rot=<0 0 0 -1>}
 deform_knee_out_base_l {parent=calf_l, loc=<0 4 0>, rot=<0 0 0 -1>}
 deform_knee_out_l {parent=deform_knee_out_base_l, loc=<0 0 0>, rot=<0 0 0 1>}
 thigh_twist_01_l {parent=thigh_l, loc=<-27.952 0 0>, rot=<0.0013 0 0 -1>}
 deform_groin_base_l {parent=thigh_l, loc=<0.4801 5.6776 -0.2486>, rot=<0 0 0 -1>}
 deform_groin_l {parent=deform_groin_base_l, loc=<0 0 0>, rot=<0 0 0 -1>}
 dyn_skirt_side_l {parent=thigh_l, loc=<-19.0152 -2.7689 -9.6281>, rot=<-0.0168 -0.9872 0.0777 -0.1379>}
 thigh_r {parent=pelvis, loc=<-1.0855 -0.2109 8.7053>, rot=<0.0355 -0.0009 0.9993 0.0155>}
 calf_r {parent=thigh_r, loc=<42.5284 0.0019 0.7297>, rot=<0.0006 -0.0161 -0.0391 -0.9991>}
 foot_r {parent=calf_r, loc=<46.7401 2.5014 0.592>, rot=<0.0427 -0.0238 0.0353 -0.9982>}
 ball_r {parent=foot_r, loc=<3.5658 -10.8066 0.1544>, rot=<0.0001 -0.0001 -0.7063 -0.7079>}
 support_foot_r {parent=foot_r, loc=<-3.5 6 0>, rot=<0 0 0 1>}
 calf_twist_02_r {parent=calf_r, loc=<19.1546 1.7943 0.2638>, rot=<0.0237 0.0095 -0.0035 -0.9997>}
 calf_twist_01_r {parent=calf_r, loc=<35.8483 1.9197 0.5794>, rot=<0.0237 0.0095 -0.0035 -0.9997>}
 dyn_boot_fur_r_04 {parent=calf_twist_01_r, loc=<-3.7082 -5.0066 -1.6076>, rot=<-0.8173 0.1498 0.2119 -0.5144>}
 dyn_boot_fur_r_01 {parent=calf_twist_01_r, loc=<-3.3434 -2.2585 4.8484>, rot=<-0.1634 0.1437 0.0502 -0.9748>}
 dyn_boot_fur_r_02 {parent=calf_twist_01_r, loc=<-3.564 3.4498 -4.2088>, rot=<0.9402 0.0151 -0.1322 -0.3135>}
 dyn_boot_fur_r_03 {parent=calf_twist_01_r, loc=<-2.8837 5.7173 2.6655>, rot=<0.5367 0.0611 -0.0654 -0.839>}
 deform_knee_in_base_r {parent=calf_r, loc=<-0.8031 5.0231 -0.2457>, rot=<0 0 0 -1>}
 deform_knee_in_r {parent=deform_knee_in_base_r, loc=<0 0 0>, rot=<0 0 0 -1>}
 deform_knee_out_base_r {parent=calf_r, loc=<0 -4 0>, rot=<0 0 0 -1>}
 deform_knee_out_r {parent=deform_knee_out_base_r, loc=<0 0 0>, rot=<0 0 0 1>}
 thigh_twist_01_r {parent=thigh_r, loc=<27.952 0 0>, rot=<0.0013 0 0 -1>}
 dyn_skirt_side_r {parent=thigh_twist_01_r, loc=<-2.0175 3.957 8.7887>, rot=<0.9879 0.0487 0.0964 0.1112>}
 dyn_leg_strap_r_base {parent=thigh_twist_01_r, loc=<1.9838 5.8333 -4.7221>, rot=<0.9343 0.0148 -0.0182 0.3558>}
 dyn_leg_strap_r_02 {parent=dyn_leg_strap_r_base, loc=<0.2318 0.3922 1.6119>, rot=<0.1631 0.2053 -0.1224 -0.9572>}
 dyn_leg_strap_r_01 {parent=dyn_leg_strap_r_base, loc=<0.3749 0.4164 -1.4519>, rot=<-0.0131 -0.3065 -0.0406 -0.9509>}
 deform_groin_base_r {parent=thigh_r, loc=<-0.4801 -5.6776 0.2486>, rot=<0 0 0 -1>}
 deform_groin_r {parent=deform_groin_base_r, loc=<0 0 0>, rot=<0 0 0 -1>}
 dyn_skirt_fr_r_base {parent=thigh_r, loc=<7.6722 -9.2159 -0.2959>, rot=<-0.0398 -0.997 0.0623 -0.0243>}
 dyn_skirt_fr_r_01 {parent=dyn_skirt_fr_r_base, loc=<0 0 0>, rot=<0 0 0 -1>}
 dyn_skirt_fr_r_02 {parent=dyn_skirt_fr_r_01, loc=<-9.1982 0 0>, rot=<0 0 0 -1>}
 dyn_skirt_fr_r_03 {parent=dyn_skirt_fr_r_01, loc=<-15.4523 -0.2775 2.2325>, rot=<0 0 0 -1>}
 dyn_skirt_fr_strap_support {parent=dyn_skirt_fr_r_base, loc=<-2.4556 -1.3745 -0.0182>, rot=<0 0 0 1>}
 deform_glute_base_l {parent=pelvis, loc=<-7.3466 -9.0213 -9.1245>, rot=<0.0172 0.031 -0.4477 -0.8935>}
 deform_glute_l {parent=deform_glute_base_l, loc=<0 0 0>, rot=<0 0 0 1>}
 deform_glute_base_r {parent=pelvis, loc=<-7.3464 -9.0213 9.1245>, rot=<0.031 -0.0172 0.8935 -0.4477>}
 deform_glute_r {parent=deform_glute_base_r, loc=<0 0 0>, rot=<0 0 0 1>}
 dyn_skirt_fr_l_base {parent=pelvis, loc=<-4.817 8.3281 -10.3591>, rot=<-0.0684 -0.981 0.166 -0.0729>}
 dyn_skirt_fr_l_01 {parent=dyn_skirt_fr_l_base, loc=<0 0 0>, rot=<0 0 0 1>}
 dyn_skirt_fr_l_02 {parent=dyn_skirt_fr_l_01, loc=<9.8343 0 0>, rot=<0 0 0 1>}
 dyn_skirt_fr_l_03 {parent=dyn_skirt_fr_l_02, loc=<9.1141 0 0>, rot=<0 0 0 1>}
 dyn_skirt_bow_r_base {parent=dyn_skirt_fr_l_base, loc=<0.1727 0.9119 2.611>, rot=<-0.2334 0.1139 -0.063 -0.9636>}
 dyn_skirt_bow_r_01 {parent=dyn_skirt_bow_r_base, loc=<0 0 0>, rot=<0 0 0 1>}
 dyn_skirt_bow_r_02 {parent=dyn_skirt_bow_r_01, loc=<7.5339 0 0>, rot=<0 0 0 1>}
 dyn_skirt_bow_l_base {parent=dyn_skirt_fr_l_base, loc=<-1.2139 1.6188 -2.5812>, rot=<0.0049 -0.0672 0.0126 -0.9976>}
 dyn_skirt_bow_l_01 {parent=dyn_skirt_bow_l_base, loc=<0 0 0>, rot=<0 0 0 1>}
 dyn_skirt_bow_l_02 {parent=dyn_skirt_bow_l_01, loc=<7.5339 0 0>, rot=<0 0 0 1>}
 dyn_skirt_bk_l_base {parent=pelvis, loc=<-7.2014 -12.5389 -9.4069>, rot=<0.0219 -0.9793 -0.1899 -0.0668>}
 dyn_skirt_bk_l_01 {parent=dyn_skirt_bk_l_base, loc=<0 0 0>, rot=<0 0 0 1>}
 dyn_skirt_bk_l_02 {parent=dyn_skirt_bk_l_01, loc=<10.1001 0 0>, rot=<0 0 0 1>}
 dyn_skirt_bk_l_03 {parent=dyn_skirt_bk_l_02, loc=<10.6954 0 0>, rot=<0 0 0 1>}
 dyn_skirt_bk_mid_base {parent=pelvis, loc=<-9.5621 -13.5309 1.5593>, rot=<0.0468 -0.9987 0.0131 0.0154>}
 dyn_skirt_bk_mid_01 {parent=dyn_skirt_bk_mid_base, loc=<0 0 0>, rot=<0 0 0 1>}
 dyn_skirt_bk_mid_02 {parent=dyn_skirt_bk_mid_01, loc=<9.8397 0 0>, rot=<0 0 0 1>}
 dyn_skirt_bk_mid_03 {parent=dyn_skirt_bk_mid_02, loc=<9.7004 0 0>, rot=<0 0 0 1>}
 dyn_skirt_bk_mid_04 {parent=dyn_skirt_bk_mid_03, loc=<9.268 0 0>, rot=<0 0 0 1>}
 dyn_skirt_bk_mid_05 {parent=dyn_skirt_bk_mid_04, loc=<9.068 0 0>, rot=<0 0 0 1>}
 dyn_skirt_bk_07 {parent=dyn_skirt_bk_mid_02, loc=<4.0415 2.271 9.357>, rot=<0.1603 0.0456 -0.0303 -0.9855>}
 dyn_skirt_bk_06 {parent=dyn_skirt_bk_mid_02, loc=<8.3244 2.5655 5.6748>, rot=<0.0129 0.0156 -0.0319 -0.9993>}
 dyn_skirt_bk_09 {parent=dyn_skirt_bk_mid_02, loc=<4.9257 1.7612 -0.3042>, rot=<-0.0264 0.0098 -0.0326 -0.9991>}
 dyn_skirt_bk_10 {parent=dyn_skirt_bk_mid_02, loc=<4.495 1.9075 -7.3714>, rot=<-0.0258 -0.0081 -0.0331 -0.9991>}
 dyn_skirt_bk_08 {parent=pelvis, loc=<-18.8066 -11.5033 11.4988>, rot=<0.0427 -0.9618 0.2603 0.0729>}
 dyn_skirt_side_l_frnt_base {parent=pelvis, loc=<-22.026 4.2457 -19.0975>, rot=<-0.0112 -0.9829 -0.1477 -0.1098>}
 dyn_skirt_side_l_frnt_01 {parent=dyn_skirt_side_l_frnt_base, loc=<0 0 0>, rot=<0 0 0 1>}
 dyn_skirt_side_l_frnt_02 {parent=dyn_skirt_side_l_frnt_01, loc=<9.5765 0 0>, rot=<0 0 0 1>}
 dyn_skirt_side_l_frnt_03 {parent=dyn_skirt_side_l_frnt_02, loc=<8.8718 0 0>, rot=<0 0 0 1>}
 dyn_skirt_side_l_bk_base {parent=pelvis, loc=<-22.4108 -6.307 -16.4371>, rot=<-0.0152 -0.9572 0.2701 -0.1025>}
 dyn_skirt_side_l_bk_01 {parent=dyn_skirt_side_l_bk_base, loc=<0 0 0>, rot=<0 0 0 1>}
 dyn_skirt_side_l_bk_02 {parent=dyn_skirt_side_l_bk_01, loc=<9.836 0 0>, rot=<0 0 0 1>}
 dyn_skirt_side_l_bk_03 {parent=dyn_skirt_side_l_bk_02, loc=<10.4007 0 0>, rot=<0 0 0 1>}
 dyn_skirt_fr_mid_base {parent=pelvis, loc=<-15.9966 10.5425 -1.5795>, rot=<-0.0016 -0.0009 0.9987 0.051>}
 dyn_skirt_fr_mid_01 {parent=dyn_skirt_fr_mid_base, loc=<0 0 0>, rot=<0 0 0 1>}
 dyn_skirt_fr_mid_02 {parent=dyn_skirt_fr_mid_01, loc=<11.0982 0.2436 -5.1582>, rot=<0.0919 -0.0382 0.0044 -0.995>}
 dyn_skirt_fr_mid_03 {parent=dyn_skirt_fr_mid_01, loc=<11.0841 0.1846 3.2317>, rot=<-0.0026 -0.9994 -0.0191 -0.0279>}
 dyn_hair_bk_01_null {type=Null, loc=<-3.5363 -11.9746 151.8473>, rot=<0.0916 -0.6832 0.0921 -0.7186>}
 dyn_hair_bk_01_ctrl {controlType=EULER_TRANSFORM, parent=dyn_hair_bk_01_null, type=Control, loc=<0 0 0>, rot=<0 0 0 1>}
 dyn_hair_bk_02_null {type=Null, loc=<-4.2007 -13.597 142.3015>, rot=<-0.0699 0.6857 -0.0692 0.7212>}
 dyn_hair_bk_02_ctrl {controlType=EULER_TRANSFORM, parent=dyn_hair_bk_02_null, type=Control, loc=<0 0 0>, rot=<0 0 0 1>}
 dyn_hair_bk_03_null {type=Null, loc=<-4.3224 -15.4694 133.5258>, rot=<-0.0196 -0.689 -0.0248 -0.7241>}
 dyn_hair_bk_03_ctrl {controlType=EULER_TRANSFORM, parent=dyn_hair_bk_03_null, type=Control, loc=<0 0 0>, rot=<0 0 0 1>}
 dyn_hair_bk_04_null {type=Null, loc=<-3.97 -15.0202 126.4052>, rot=<0.0196 0.689 0.0248 0.7241>}
 dyn_hair_bk_04_ctrl {controlType=EULER_TRANSFORM, parent=dyn_hair_bk_04_null, type=Control, loc=<0 0 0>, rot=<0 0 0 1>}
]
variables [
 Crouching {type=bool, default=False}
 Dyn Hair Ctls {type="TArray<FRigElementKey>"}
 Is Backbling {type=bool, default=False}
 Is Skydive Floating {type=bool, default=False}
]
