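# Physics asset for the F_MED_WriteWalrus female FaceAcc mesh part, setting up physics for its hair.
# Four chains (back hair, pigtail, right side hair, right curl) hang from the head through zero-size base bodies, with mostly tight angular limits shaping their bend.
# Capsules and a sphere cover torso, neck, head, shoulders and arms, with left forearm only.
# An extra angularly free constraint cross-links the pigtail to the back hair mid-chain.
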PhysicsAsset /BRCosmetics/Characters/Player/Female/Medium/Bodies/F_MED_WriteWalrus/Meshes/Parts/F_MED_WriteWalrus_FaceAcc_Physics
bodies {
 clavicle_l [{shapeType=Capsule, length=8.7997, radius=7.5679, shapeName=clavicle_l_capsule, center=<6.0138 0 0>, rotation=<rot 90 0 0>}]
 clavicle_r [{shapeType=Capsule, length=7.9843, radius=7.0691, shapeName=clavicle_r_capsule, center=<-6.0138 0 0>, rotation=<rot 90 180 0>}]
 dyn_curl__side_r_01 [{shapeType=Capsule, length=3.2399, radius=2.8519, shapeName=dyn_curl__side_r_01_capsule, center=<3.1373 0 0>, rotation=<rot 90 0 0>}]
 dyn_curl__side_r_02 [{
   shapeType=Capsule
   length=6.9811
   radius=3.2956
   shapeName=dyn_curl__side_r_02_capsule
   center=<5.5898 1.6362 -2.8625>
   rotation=<rot 41.9538 0.0016 -15.1787>
  }]
 dyn_curl__side_r_base [{shapeType=Capsule, length=0, radius=0, shapeName=dyn_curl__side_r_base_capsule}]
 dyn_hair_bk_01 [{shapeType=Capsule, length=15.3602, radius=9.416, shapeName=dyn_hair_bk_01_capsule, center=<6.8818 -2.9424 -2.4844>}]
 dyn_hair_bk_02 [{shapeType=Capsule, length=12.1476, radius=7.5262, shapeName=dyn_hair_bk_02_capsule, center=<6.8818 -5.5809 -2.1188>}]
 dyn_hair_bk_03 [{shapeType=Capsule, length=14.3874, radius=7.1999, shapeName=dyn_hair_bk_03_capsule, center=<6.8818 -5.2852 0>}]
 dyn_hair_bk_04 [{shapeType=Capsule, length=13.5689, radius=7.5209, shapeName=dyn_hair_bk_04_capsule, center=<10.3531 -4.5758 0>}]
 dyn_hair_bk_base [{shapeType=Capsule, length=0, radius=0, shapeName=dyn_hair_bk_base_capsule}]
 dyn_hair_side_r_01 [{shapeType=Capsule, length=5.4522, radius=2.4985, shapeName=dyn_hair_side_r_01_capsule, center=<3.2769 0 0>, rotation=<rot 90 0 0>}]
 dyn_hair_side_r_02 [{shapeType=Capsule, length=5.3804, radius=2.3468, shapeName=dyn_hair_side_r_02_capsule, center=<2.7319 0 0>, rotation=<rot 90 0 0>}]
 dyn_hair_side_r_04 [{shapeType=Capsule, length=4.8005, radius=2.0773, shapeName=dyn_hair_side_r_04_capsule, center=<6.3867 0.7901 0>, rotation=<rot 90 19.6992 0>}]
 dyn_hair_side_r_05 [{shapeType=Capsule, length=7.1135, radius=1.7833, shapeName=dyn_hair_side_r_05_capsule, center=<4.8565 0 0>, rotation=<rot 90 0 0>}]
 dyn_hair_side_r_06 [{shapeType=Capsule, length=4.2308, radius=1.7692, shapeName=dyn_hair_side_r_06_capsule, center=<3.5003 0 0>, rotation=<rot 90 0 0>}]
 dyn_hair_side_r_base [{shapeType=Capsule, length=0, radius=0, shapeName=dyn_hair_side_r_base_capsule}]
 dyn_pigtail_01 [{shapeType=Capsule, length=3.0362, radius=2.9869, shapeName=dyn_pigtail_01_capsule, center=<4.6706 0 0>, rotation=<rot 90 0 0>}]
 dyn_pigtail_02 [{shapeType=Capsule, length=2.555, radius=3.4682, shapeName=dyn_pigtail_02_capsule, center=<4.6706 0 0>, rotation=<rot 90 0 0>}]
 dyn_pigtail_03 [{shapeType=Capsule, length=2.555, radius=3.4682, shapeName=dyn_pigtail_03_capsule, center=<4.6706 0 0>, rotation=<rot 90 0 0>}]
 dyn_pigtail_04 [{shapeType=Capsule, length=2.555, radius=3.4682, shapeName=dyn_pigtail_04_capsule, center=<4.6706 0 0>, rotation=<rot 90 0 0>}]
 dyn_pigtail_05 [{shapeType=Capsule, length=4.9768, radius=5.3718, shapeName=dyn_pigtail_05_capsule, center=<4.6706 0 0>, rotation=<rot 90 0 0>}]
 dyn_pigtail_06 [{shapeType=Capsule, length=3.4111, radius=5.8388, shapeName=dyn_pigtail_06_capsule, center=<4.6706 0 0>, rotation=<rot 90 0 0>}]
 dyn_pigtail_07 [{shapeType=Capsule, length=4.1927, radius=6.2163, shapeName=dyn_pigtail_07_capsule, center=<6.0038 0 0>, rotation=<rot 90 0 0>}]
 dyn_pigtail_base [{shapeType=Capsule, length=0, radius=0, shapeName=dyn_pigtail_base_capsule}]
 head [{shapeType=Capsule, length=0, radius=9.715, shapeName=head_capsule, center=<7.1056 -1.1755 0>}]
 lowerarm_l [{shapeType=Capsule, length=16.2939, radius=5.7823, shapeName=lowerarm_l_capsule, center=<12.8812 0 0>, rotation=<rot 90 0 0>}]
 neck_01 [{shapeType=Capsule, length=8.7325, radius=3.9901, shapeName=neck_01_capsule, rotation=<rot 90 0 0>}]
 neck_02 [{shapeType=Capsule, length=10.9414, radius=4.5611, shapeName=neck_02_capsule, rotation=<rot 90 0 0>}]
 pelvis [{shapeType=Capsule, length=10.5251, radius=11.2551, shapeName=pelvis_capsule}]
 spine_01 [{shapeType=Capsule, length=14.995, radius=8.4259, shapeName=spine_01_capsule}]
 spine_02 [{shapeType=Capsule, length=18.7281, radius=8.7312, shapeName=spine_02_capsule}]
 spine_03 [{shapeType=Capsule, length=18.9009, radius=8.5398, shapeName=spine_03_capsule}]
 spine_04 [{shapeType=Capsule, length=19.5443, radius=8.0163, shapeName=spine_04_capsule}]
 spine_05 [
  {shapeType=Capsule, length=19.4898, radius=7.5854, shapeName=spine_05_capsule}
  {shapeType=Capsule, length=11.1558, radius=7.9641, shapeName=spine_05_capsule, center=<6.3196 -1.212 0>}
 ]
 support_shoulder_r_rotator [{shapeType=Capsule, length=0, radius=8.4701, shapeName=support_shoulder_r_rotator_capsule, center=<-2.2389 0 0>, rotation=<rot 0 -90 0>}]
 upperarm_l [
  {shapeType=Sphere, radius=9.1089, shapeName=upperarm_l_sphere, center=<2.9251 0.8641 2.3403>}
  {shapeType=Capsule, length=19.2383, radius=5.2847, shapeName=upperarm_l_capsule, center=<13.8803 0 0>, rotation=<rot 90 0 0>}
 ]
 upperarm_r [{shapeType=Capsule, length=19.8969, radius=4.6261, shapeName=upperarm_r_capsule, center=<-13.8803 0 0>, rotation=<rot 90 180 0>}]
}
constraints [
 {bone1Name=spine_04, bone2Name=spine_03, swing1LimitDegrees=45, swing2LimitDegrees=45, twistLimitDegrees=45}
 {bone1Name=spine_05, bone2Name=spine_04, swing1LimitDegrees=45, swing2LimitDegrees=45, twistLimitDegrees=45}
 {bone1Name=clavicle_l, bone2Name=spine_05, swing1LimitDegrees=45, swing2LimitDegrees=45, twistLimitDegrees=45}
 {bone1Name=upperarm_l, bone2Name=clavicle_l, swing1LimitDegrees=45, swing2LimitDegrees=45, twistLimitDegrees=45}
 {bone1Name=lowerarm_l, bone2Name=upperarm_l, swing1LimitDegrees=45, swing2LimitDegrees=45, twistLimitDegrees=45}
 {bone1Name=clavicle_r, bone2Name=spine_05, swing1LimitDegrees=45, swing2LimitDegrees=45, twistLimitDegrees=45}
 {bone1Name=upperarm_r, bone2Name=clavicle_r, swing1LimitDegrees=45, swing2LimitDegrees=45, twistLimitDegrees=45}
 {bone1Name=neck_01, bone2Name=spine_05, swing1LimitDegrees=45, swing2LimitDegrees=45, twistLimitDegrees=45}
 {bone1Name=neck_02, bone2Name=neck_01, swing1LimitDegrees=45, swing2LimitDegrees=45, twistLimitDegrees=45}
 {bone1Name=head, bone2Name=neck_02, swing1LimitDegrees=45, swing2LimitDegrees=45, twistLimitDegrees=45}
 {bone1Name=dyn_hair_side_r_base, bone2Name=head, swing1LimitDegrees=45, swing2LimitDegrees=45, twistLimitDegrees=45}
 {bone1Name=dyn_hair_side_r_01, bone2Name=dyn_hair_side_r_base, swing1LimitDegrees=5, swing2LimitDegrees=20, twistLimitDegrees=20}
 {bone1Name=dyn_hair_side_r_02, bone2Name=dyn_hair_side_r_01, swing1LimitDegrees=20, swing2LimitDegrees=10, twistLimitDegrees=20}
 {bone1Name=dyn_hair_side_r_04, bone2Name=dyn_hair_side_r_02, swing1LimitDegrees=10, swing2LimitDegrees=20, twistLimitDegrees=20}
 {bone1Name=dyn_hair_side_r_05, bone2Name=dyn_hair_side_r_02, swing1LimitDegrees=10, swing2LimitDegrees=10, twistLimitDegrees=20}
 {bone1Name=dyn_hair_side_r_06, bone2Name=dyn_hair_side_r_02, swing1LimitDegrees=10, swing2LimitDegrees=10, twistLimitDegrees=20}
 {bone1Name=dyn_pigtail_base, bone2Name=head, swing1LimitDegrees=45, swing2LimitDegrees=45, twistLimitDegrees=45}
 {bone1Name=dyn_pigtail_01, bone2Name=dyn_pigtail_base, swing1LimitDegrees=5, swing2LimitDegrees=5, twistLimitDegrees=0}
 {bone1Name=dyn_pigtail_02, bone2Name=dyn_pigtail_01, swing1LimitDegrees=10, swing2LimitDegrees=20, twistLimitDegrees=1}
 {bone1Name=dyn_pigtail_03, bone2Name=dyn_pigtail_02, swing1LimitDegrees=10, swing2LimitDegrees=20, twistLimitDegrees=5}
 {bone1Name=dyn_pigtail_04, bone2Name=dyn_pigtail_03, swing1LimitDegrees=10, swing2LimitDegrees=20, twistLimitDegrees=5}
 {bone1Name=dyn_pigtail_05, bone2Name=dyn_pigtail_04, swing1LimitDegrees=10, swing2LimitDegrees=20, twistLimitDegrees=5}
 {bone1Name=dyn_pigtail_06, bone2Name=dyn_pigtail_05, swing1LimitDegrees=20, swing2LimitDegrees=20, twistLimitDegrees=5}
 {bone1Name=dyn_pigtail_07, bone2Name=dyn_pigtail_06, swing1LimitDegrees=10, swing2LimitDegrees=20, twistLimitDegrees=10}
 {
  bone1Name=dyn_curl__side_r_base
  bone2Name=head
  swing1LimitDegrees=45
  swing1Motion=Free
  swing2LimitDegrees=45
  swing2Motion=Free
  twistLimitDegrees=45
  twistMotion=Free
 }
 {bone1Name=dyn_curl__side_r_01, bone2Name=dyn_curl__side_r_base, swing1LimitDegrees=5, swing2LimitDegrees=20, twistLimitDegrees=5}
 {bone1Name=dyn_curl__side_r_02, bone2Name=dyn_curl__side_r_01, swing1LimitDegrees=20, swing2LimitDegrees=30, twistLimitDegrees=5}
 {
  bone1Name=spine_03
  bone2Name=spine_02
  swing1LimitDegrees=45
  swing1Motion=Free
  swing2LimitDegrees=45
  swing2Motion=Free
  twistLimitDegrees=45
  twistMotion=Free
 }
 {
  bone1Name=spine_02
  bone2Name=spine_01
  swing1LimitDegrees=45
  swing1Motion=Free
  swing2LimitDegrees=45
  swing2Motion=Free
  twistLimitDegrees=45
  twistMotion=Free
 }
 {
  bone1Name=spine_01
  bone2Name=pelvis
  swing1LimitDegrees=45
  swing1Motion=Free
  swing2LimitDegrees=45
  swing2Motion=Free
  twistLimitDegrees=45
  twistMotion=Free
 }
 {
  bone1Name=support_shoulder_r_rotator
  bone2Name=upperarm_r
  swing1LimitDegrees=45
  swing1Motion=Free
  swing2LimitDegrees=45
  swing2Motion=Free
  twistLimitDegrees=45
  twistMotion=Free
 }
 {
  bone1Name=dyn_hair_bk_base
  bone2Name=head
  swing1LimitDegrees=45
  swing1Motion=Free
  swing2LimitDegrees=45
  swing2Motion=Free
  twistLimitDegrees=45
  twistMotion=Free
 }
 {bone1Name=dyn_hair_bk_01, bone2Name=dyn_hair_bk_base, swing1LimitDegrees=45, swing2LimitDegrees=10, twistLimitDegrees=10}
 {bone1Name=dyn_hair_bk_02, bone2Name=dyn_hair_bk_01, swing1LimitDegrees=45, swing2LimitDegrees=10, twistLimitDegrees=10}
 {bone1Name=dyn_hair_bk_03, bone2Name=dyn_hair_bk_02, swing1LimitDegrees=45, swing2LimitDegrees=10, twistLimitDegrees=10}
 {bone1Name=dyn_hair_bk_04, bone2Name=dyn_hair_bk_03, swing1LimitDegrees=45, swing2LimitDegrees=10, twistLimitDegrees=10}
 {
  bone1Name=dyn_pigtail_05
  bone2Name=dyn_hair_bk_03
  swing1LimitDegrees=45
  swing1Motion=Free
  swing2LimitDegrees=45
  swing2Motion=Free
  twistLimitDegrees=45
  twistMotion=Free
 }
]
