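# Post-process style AnimBP for the WriteWalrus female cosmetic FaceAcc part, adding secondary motion to hair.
# Parent and Body poses, blended at the clavicles, feed a Control Rig driven by crouch, backbling and skydive-float flags.
# A backbling bool selects between two RigidBody physics assets, then chained AnimDynamics nodes swing bangs, curls and pigtails before a head Z stretch.
# The EventGraph updates those flags, gravity override, LOD state and Body reference.
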AnimBlueprint /BRCosmetics/Characters/Player/Female/Medium/Bodies/F_MED_WriteWalrus/Meshes/Parts/F_MED_WriteWalrus_FaceAcc_AnimBP
target_skeleton=&/F_MED_WriteWalrus_FaceAcc_Skeleton
graphs [
 AnimGraph {nodes [
   n1 = Anim Dynamics - Bone: dyn_bang_01(ComponentPose=Local To Component(LocalPose=Use cached pose 'toAnimDyn'()), Alpha=1, ExternalForce="0,0,0") {
    angularDampingOverride=0.99
    angularSpringConstant=100
    bAngularSpring=true
    bOverrideAngularDamping=true
    bUseGravityOverride=true
    simulationSpace=RootRelative
    boundBone {boneName=dyn_bang_01}
    componentAppliedLinearAccClamp=<3500 3500 3500>
    componentLinearAccScale=<1 1 1>
    componentLinearVelScale=<1 1 1>
    gravityOverride {z=-980}
    physicsBodyDefinitions [{
      boundBone {boneName=dyn_bang_01}
      boxExtents=<4 4 4>
      constraintSetup {angularLimitsMax=<0 0 10>, angularLimitsMin=<0 -20 -10>, angularTarget=<1 0 0>}
      localJointOffset=<0.1 0 0>
     }]
   }
   n2 = Anim Dynamics - Bone: dyn_bang_02(ComponentPose=@n1, Alpha=1, ExternalForce="0,0,0") {
    angularDampingOverride=0.99
    angularSpringConstant=100
    bAngularSpring=true
    bOverrideAngularDamping=true
    bUseGravityOverride=true
    simulationSpace=RootRelative
    boundBone {boneName=dyn_bang_02}
    componentAppliedLinearAccClamp=<3500 3500 3500>
    componentLinearAccScale=<1 1 1>
    componentLinearVelScale=<1 1 1>
    gravityOverride {z=-980}
    physicsBodyDefinitions [{
      boundBone {boneName=dyn_bang_02}
      boxExtents=<4 4 4>
      constraintSetup {angularLimitsMax=<0 0 20>, angularLimitsMin=<0 -20 -10>, angularTarget=<1 0 0>}
      localJointOffset=<0.1 0 0>
     }]
   }
   n3 = Anim Dynamics - Bone: dyn_hair_frnt_02(ComponentPose=@n2, Alpha=1, ExternalForce="0,0,0") {
    angularDampingOverride=0.99
    angularSpringConstant=100
    bAngularSpring=true
    bOverrideAngularDamping=true
    bUseGravityOverride=true
    simulationSpace=RootRelative
    boundBone {boneName=dyn_hair_frnt_02}
    componentAppliedLinearAccClamp=<3500 3500 3500>
    componentLinearAccScale=<1 1 1>
    componentLinearVelScale=<1 1 1>
    gravityOverride {z=-980}
    physicsBodyDefinitions [{
      boundBone {boneName=dyn_hair_frnt_02}
      boxExtents=<4 4 4>
      constraintSetup {angularLimitsMax=<0 0 20>, angularLimitsMin=<0 -20 -10>, angularTarget=<1 0 0>}
      localJointOffset=<0.1 0 0>
     }]
   }
   n4 = Anim Dynamics - Bone: dyn_hair_frnt_04(ComponentPose=@n3, Alpha=1, ExternalForce="0,0,0") {
    angularDampingOverride=0.99
    angularSpringConstant=100
    bAngularSpring=true
    bOverrideAngularDamping=true
    bUseGravityOverride=true
    simulationSpace=RootRelative
    boundBone {boneName=dyn_hair_frnt_04}
    componentAppliedLinearAccClamp=<3500 3500 3500>
    componentLinearAccScale=<1 1 1>
    componentLinearVelScale=<1 1 1>
    gravityOverride {z=-980}
    physicsBodyDefinitions [{
      boundBone {boneName=dyn_hair_frnt_04}
      boxExtents=<4 4 4>
      constraintSetup {angularLimitsMax=<0 30 20>, angularLimitsMin=<0 0 -10>, angularTarget=<1 0 0>}
      localJointOffset=<0.1 0 0>
     }]
   }
   n5 = Anim Dynamics - Bone: dyn_hair_frnt_03(ComponentPose=@n4, Alpha=1, ExternalForce="0,0,0") {
    angularDampingOverride=0.99
    angularSpringConstant=100
    bAngularSpring=true
    bOverrideAngularDamping=true
    bUseGravityOverride=true
    simulationSpace=RootRelative
    boundBone {boneName=dyn_hair_frnt_03}
    componentAppliedLinearAccClamp=<3500 3500 3500>
    componentLinearAccScale=<1 1 1>
    componentLinearVelScale=<1 1 1>
    gravityOverride {z=-980}
    physicsBodyDefinitions [{
      boundBone {boneName=dyn_hair_frnt_03}
      boxExtents=<4 4 4>
      constraintSetup {angularLimitsMax=<0 40 20>, angularLimitsMin=<0 0 -10>, angularTarget=<1 0 0>}
      localJointOffset=<0.1 0 0>
     }]
   }
   n6 = Anim Dynamics - Bone: dyn_curl_side_r_04(ComponentPose=@n5, Alpha=1, ExternalForce="0,0,0") {
    angularDampingOverride=0.99
    angularSpringConstant=100
    bAngularSpring=true
    bOverrideAngularDamping=true
    bUseGravityOverride=true
    simulationSpace=RootRelative
    boundBone {boneName=dyn_curl_side_r_04}
    componentAppliedLinearAccClamp=<3500 3500 3500>
    componentLinearAccScale=<1 1 1>
    componentLinearVelScale=<1 1 1>
    gravityOverride {z=-980}
    physicsBodyDefinitions [{
      boundBone {boneName=dyn_curl_side_r_04}
      boxExtents=<4 4 4>
      constraintSetup {angularLimitsMax=<0 0 20>, angularLimitsMin=<0 -20 -10>, angularTarget=<1 0 0>}
      localJointOffset=<0.1 0 0>
     }]
   }
   n7 = Anim Dynamics - Bone: dyn_curl_side_r_01(ComponentPose=@n6, Alpha=1, ExternalForce="0,0,0") {
    angularDampingOverride=0.99
    angularSpringConstant=100
    bAngularSpring=true
    bOverrideAngularDamping=true
    bUseGravityOverride=true
    simulationSpace=RootRelative
    boundBone {boneName=dyn_curl_side_r_01}
    componentAppliedLinearAccClamp=<3500 3500 3500>
    componentLinearAccScale=<1 1 1>
    componentLinearVelScale=<1 1 1>
    gravityOverride {z=-980}
    physicsBodyDefinitions [{
      boundBone {boneName=dyn_curl_side_r_01}
      boxExtents=<4 4 4>
      constraintSetup {angularLimitsMax=<0 0 20>, angularLimitsMin=<0 -20 -10>, angularTarget=<1 0 0>}
      localJointOffset=<0.1 0 0>
     }]
   }
   n8 = Anim Dynamics - Bone: dyn_curl_bk_01(ComponentPose=@n7, Alpha=1, ExternalForce="0,0,0") {
    angularDampingOverride=0.99
    angularSpringConstant=100
    bAngularSpring=true
    bOverrideAngularDamping=true
    bUseGravityOverride=true
    simulationSpace=RootRelative
    boundBone {boneName=dyn_curl_bk_01}
    componentAppliedLinearAccClamp=<3500 3500 3500>
    componentLinearAccScale=<1 1 1>
    componentLinearVelScale=<1 1 1>
    gravityOverride {z=-980}
    physicsBodyDefinitions [{
      boundBone {boneName=dyn_curl_bk_01}
      boxExtents=<4 4 4>
      constraintSetup {angularLimitsMax=<0 0 20>, angularLimitsMin=<0 -20 -10>, angularTarget=<1 0 0>}
      localJointOffset=<0.1 0 0>
     }]
   }
   n9 = Anim Dynamics - Bone: dyn_curl_bk_02(ComponentPose=@n8, Alpha=1, ExternalForce="0,0,0") {
    angularDampingOverride=0.99
    angularSpringConstant=100
    bAngularSpring=true
    bOverrideAngularDamping=true
    bUseGravityOverride=true
    simulationSpace=RootRelative
    boundBone {boneName=dyn_curl_bk_02}
    componentAppliedLinearAccClamp=<3500 3500 3500>
    componentLinearAccScale=<1 1 1>
    componentLinearVelScale=<1 1 1>
    gravityOverride {z=-980}
    physicsBodyDefinitions [{
      boundBone {boneName=dyn_curl_bk_02}
      boxExtents=<4 4 4>
      constraintSetup {angularLimitsMax=<0 0 20>, angularLimitsMin=<0 -20 -10>, angularTarget=<1 0 0>}
      localJointOffset=<0.1 0 0>
     }]
   }
   n10 = Anim Dynamics - Bone: dyn_curl_bk_03(ComponentPose=@n9, Alpha=1, ExternalForce="0,0,0") {
    angularDampingOverride=0.99
    angularSpringConstant=100
    bAngularSpring=true
    bOverrideAngularDamping=true
    bUseGravityOverride=true
    simulationSpace=RootRelative
    boundBone {boneName=dyn_curl_bk_03}
    componentAppliedLinearAccClamp=<3500 3500 3500>
    componentLinearAccScale=<1 1 1>
    componentLinearVelScale=<1 1 1>
    gravityOverride {z=-980}
    physicsBodyDefinitions [{
      boundBone {boneName=dyn_curl_bk_03}
      boxExtents=<4 4 4>
      constraintSetup {angularLimitsMax=<0 0 20>, angularLimitsMin=<0 -20 -10>, angularTarget=<1 0 0>}
      localJointOffset=<0.1 0 0>
     }]
   }
   n11 = Anim Dynamics - Bone: dyn_curl_bk_04(ComponentPose=@n10, Alpha=1, ExternalForce="0,0,0") {
    angularDampingOverride=0.99
    angularSpringConstant=100
    bAngularSpring=true
    bOverrideAngularDamping=true
    bUseGravityOverride=true
    simulationSpace=RootRelative
    boundBone {boneName=dyn_curl_bk_04}
    componentAppliedLinearAccClamp=<3500 3500 3500>
    componentLinearAccScale=<1 1 1>
    componentLinearVelScale=<1 1 1>
    gravityOverride {z=-980}
    physicsBodyDefinitions [{
      boundBone {boneName=dyn_curl_bk_04}
      boxExtents=<4 4 4>
      constraintSetup {angularLimitsMax=<0 0 0>, angularLimitsMin=<0 -20 -30>, angularTarget=<1 0 0>}
      localJointOffset=<0.1 0 0>
     }]
   }
   n12 = Anim Dynamics - Bone: dyn_curl_bk_05(ComponentPose=@n11, Alpha=1, ExternalForce="0,0,0") {
    angularDampingOverride=0.99
    angularSpringConstant=100
    bAngularSpring=true
    bOverrideAngularDamping=true
    bUseGravityOverride=true
    simulationSpace=RootRelative
    boundBone {boneName=dyn_curl_bk_05}
    componentAppliedLinearAccClamp=<3500 3500 3500>
    componentLinearAccScale=<1 1 1>
    componentLinearVelScale=<1 1 1>
    gravityOverride {z=-980}
    physicsBodyDefinitions [{
      boundBone {boneName=dyn_curl_bk_05}
      boxExtents=<4 4 4>
      constraintSetup {angularLimitsMax=<0 0 20>, angularLimitsMin=<0 -20 -10>, angularTarget=<1 0 0>}
      localJointOffset=<0.1 0 0>
     }]
   }
   n13 = Anim Dynamics - Bone: dyn_curl_bk_06(ComponentPose=@n12, Alpha=1, ExternalForce="0,0,0") {
    angularDampingOverride=0.99
    angularSpringConstant=100
    bAngularSpring=true
    bOverrideAngularDamping=true
    bUseGravityOverride=true
    simulationSpace=RootRelative
    boundBone {boneName=dyn_curl_bk_06}
    componentAppliedLinearAccClamp=<3500 3500 3500>
    componentLinearAccScale=<1 1 1>
    componentLinearVelScale=<1 1 1>
    gravityOverride {z=-980}
    physicsBodyDefinitions [{
      boundBone {boneName=dyn_curl_bk_06}
      boxExtents=<4 4 4>
      constraintSetup {angularLimitsMax=<0 0 20>, angularLimitsMin=<0 -20 -10>, angularTarget=<1 0 0>}
      localJointOffset=<0.1 0 0>
     }]
   }
   n14 = Anim Dynamics - Bone: dyn_curl_bk_07(ComponentPose=@n13, Alpha=1, ExternalForce="0,0,0") {
    angularDampingOverride=0.99
    angularSpringConstant=100
    bAngularSpring=true
    bOverrideAngularDamping=true
    bUseGravityOverride=true
    simulationSpace=RootRelative
    boundBone {boneName=dyn_curl_bk_07}
    componentAppliedLinearAccClamp=<3500 3500 3500>
    componentLinearAccScale=<1 1 1>
    componentLinearVelScale=<1 1 1>
    gravityOverride {z=-980}
    physicsBodyDefinitions [{
      boundBone {boneName=dyn_curl_bk_07}
      boxExtents=<4 4 4>
      constraintSetup {angularLimitsMax=<0 0 20>, angularLimitsMin=<0 -20 -10>, angularTarget=<1 0 0>}
      localJointOffset=<0.1 0 0>
     }]
   }
   n15 = Anim Dynamics - Bone: dyn_curl_bk_08(ComponentPose=@n14, Alpha=1, ExternalForce="0,0,0") {
    angularDampingOverride=0.99
    angularSpringConstant=100
    bAngularSpring=true
    bOverrideAngularDamping=true
    bUseGravityOverride=true
    simulationSpace=RootRelative
    boundBone {boneName=dyn_curl_bk_08}
    componentAppliedLinearAccClamp=<3500 3500 3500>
    componentLinearAccScale=<1 1 1>
    componentLinearVelScale=<1 1 1>
    gravityOverride {z=-980}
    physicsBodyDefinitions [{
      boundBone {boneName=dyn_curl_bk_08}
      boxExtents=<4 4 4>
      constraintSetup {angularLimitsMax=<0 0 20>, angularLimitsMin=<0 -20 -10>, angularTarget=<1 0 0>}
      localJointOffset=<0.1 0 0>
     }]
   }
   n16 = Anim Dynamics - Bone: dyn_curl_bk_09(ComponentPose=@n15, Alpha=1, ExternalForce="0,0,0") {
    angularDampingOverride=0.99
    angularSpringConstant=100
    bAngularSpring=true
    bOverrideAngularDamping=true
    bUseGravityOverride=true
    simulationSpace=RootRelative
    boundBone {boneName=dyn_curl_bk_09}
    componentAppliedLinearAccClamp=<3500 3500 3500>
    componentLinearAccScale=<1 1 1>
    componentLinearVelScale=<1 1 1>
    gravityOverride {z=-980}
    physicsBodyDefinitions [{
      boundBone {boneName=dyn_curl_bk_09}
      boxExtents=<4 4 4>
      constraintSetup {angularLimitsMax=<0 0 20>, angularLimitsMin=<0 -20 -10>, angularTarget=<1 0 0>}
      localJointOffset=<0.1 0 0>
     }]
   }
   n17 = Anim Dynamics - Bone: dyn_curl_bk_10(ComponentPose=@n16, Alpha=1, ExternalForce="0,0,0") {
    angularDampingOverride=0.99
    angularSpringConstant=100
    bAngularSpring=true
    bOverrideAngularDamping=true
    bUseGravityOverride=true
    simulationSpace=RootRelative
    boundBone {boneName=dyn_curl_bk_10}
    componentAppliedLinearAccClamp=<3500 3500 3500>
    componentLinearAccScale=<1 1 1>
    componentLinearVelScale=<1 1 1>
    gravityOverride {z=-980}
    physicsBodyDefinitions [{
      boundBone {boneName=dyn_curl_bk_10}
      boxExtents=<4 4 4>
      constraintSetup {angularLimitsMax=<0 0 20>, angularLimitsMin=<0 -20 -10>, angularTarget=<1 0 0>}
      localJointOffset=<0.1 0 0>
     }]
   }
   n18 = Anim Dynamics - Bone: dyn_curl_bk_11(ComponentPose=@n17, Alpha=1, ExternalForce="0,0,0") {
    angularDampingOverride=0.99
    angularSpringConstant=100
    bAngularSpring=true
    bOverrideAngularDamping=true
    bUseGravityOverride=true
    simulationSpace=RootRelative
    boundBone {boneName=dyn_curl_bk_11}
    componentAppliedLinearAccClamp=<3500 3500 3500>
    componentLinearAccScale=<1 1 1>
    componentLinearVelScale=<1 1 1>
    gravityOverride {z=-980}
    physicsBodyDefinitions [{
      boundBone {boneName=dyn_curl_bk_11}
      boxExtents=<4 4 4>
      constraintSetup {angularLimitsMax=<0 0 20>, angularLimitsMin=<0 -20 -10>, angularTarget=<1 0 0>}
      localJointOffset=<0.1 0 0>
     }]
   }
   n19 = Anim Dynamics - Bone: dyn_curl_bk_12(ComponentPose=@n18, Alpha=1, ExternalForce="0,0,0") {
    angularDampingOverride=0.99
    angularSpringConstant=100
    bAngularSpring=true
    bOverrideAngularDamping=true
    bUseGravityOverride=true
    simulationSpace=RootRelative
    boundBone {boneName=dyn_curl_bk_12}
    componentAppliedLinearAccClamp=<3500 3500 3500>
    componentLinearAccScale=<1 1 1>
    componentLinearVelScale=<1 1 1>
    gravityOverride {z=-980}
    physicsBodyDefinitions [{
      boundBone {boneName=dyn_curl_bk_12}
      boxExtents=<4 4 4>
      constraintSetup {angularLimitsMax=<0 0 20>, angularLimitsMin=<0 -20 -10>, angularTarget=<1 0 0>}
      localJointOffset=<0.1 0 0>
     }]
   }
   n20 = Anim Dynamics - Bone: dyn_curl_bk_13(ComponentPose=@n19, Alpha=1, ExternalForce="0,0,0") {
    angularDampingOverride=0.99
    angularSpringConstant=100
    bAngularSpring=true
    bOverrideAngularDamping=true
    bUseGravityOverride=true
    simulationSpace=RootRelative
    boundBone {boneName=dyn_curl_bk_13}
    componentAppliedLinearAccClamp=<3500 3500 3500>
    componentLinearAccScale=<1 1 1>
    componentLinearVelScale=<1 1 1>
    gravityOverride {z=-980}
    physicsBodyDefinitions [{
      boundBone {boneName=dyn_curl_bk_13}
      boxExtents=<4 4 4>
      constraintSetup {angularLimitsMax=<0 0 20>, angularLimitsMin=<0 -20 -10>, angularTarget=<1 0 0>}
      localJointOffset=<0.1 0 0>
     }]
   }
   n21 = Anim Dynamics - Bone: dyn_curl_bk_14(ComponentPose=@n20, Alpha=1, ExternalForce="0,0,0") {
    angularDampingOverride=0.99
    angularSpringConstant=100
    bAngularSpring=true
    bOverrideAngularDamping=true
    bUseGravityOverride=true
    simulationSpace=RootRelative
    boundBone {boneName=dyn_curl_bk_14}
    componentAppliedLinearAccClamp=<3500 3500 3500>
    componentLinearAccScale=<1 1 1>
    componentLinearVelScale=<1 1 1>
    gravityOverride {z=-980}
    physicsBodyDefinitions [{
      boundBone {boneName=dyn_curl_bk_14}
      boxExtents=<4 4 4>
      constraintSetup {angularLimitsMax=<0 0 20>, angularLimitsMin=<0 -20 -10>, angularTarget=<1 0 0>}
      localJointOffset=<0.1 0 0>
     }]
   }
   n22 = Anim Dynamics - Bone: dyn_curl_bk_15(ComponentPose=@n21, Alpha=1, ExternalForce="0,0,0") {
    angularDampingOverride=0.99
    angularSpringConstant=100
    bAngularSpring=true
    bOverrideAngularDamping=true
    bUseGravityOverride=true
    simulationSpace=RootRelative
    boundBone {boneName=dyn_curl_bk_15}
    componentAppliedLinearAccClamp=<3500 3500 3500>
    componentLinearAccScale=<1 1 1>
    componentLinearVelScale=<1 1 1>
    gravityOverride {z=-980}
    physicsBodyDefinitions [{
      boundBone {boneName=dyn_curl_bk_15}
      boxExtents=<4 4 4>
      constraintSetup {angularLimitsMax=<0 0 20>, angularLimitsMin=<0 -20 -10>, angularTarget=<1 0 0>}
      localJointOffset=<0.1 0 0>
     }]
   }
   n23 = Anim Dynamics - Bone: dyn_curl_bk_16(ComponentPose=@n22, Alpha=1, ExternalForce="0,0,0") {
    angularDampingOverride=0.99
    angularSpringConstant=100
    bAngularSpring=true
    bOverrideAngularDamping=true
    bUseGravityOverride=true
    simulationSpace=RootRelative
    boundBone {boneName=dyn_curl_bk_16}
    componentAppliedLinearAccClamp=<3500 3500 3500>
    componentLinearAccScale=<1 1 1>
    componentLinearVelScale=<1 1 1>
    gravityOverride {z=-980}
    physicsBodyDefinitions [{
      boundBone {boneName=dyn_curl_bk_16}
      boxExtents=<4 4 4>
      constraintSetup {angularLimitsMax=<0 0 20>, angularLimitsMin=<0 -20 -10>, angularTarget=<1 0 0>}
      localJointOffset=<0.1 0 0>
     }]
   }
   n24 = Anim Dynamics - Bone: dyn_curl_bk_17(ComponentPose=@n23, Alpha=1, ExternalForce="0,0,0") {
    angularDampingOverride=0.99
    angularSpringConstant=100
    bAngularSpring=true
    bOverrideAngularDamping=true
    bUseGravityOverride=true
    simulationSpace=RootRelative
    boundBone {boneName=dyn_curl_bk_17}
    componentAppliedLinearAccClamp=<3500 3500 3500>
    componentLinearAccScale=<1 1 1>
    componentLinearVelScale=<1 1 1>
    gravityOverride {z=-980}
    physicsBodyDefinitions [{
      boundBone {boneName=dyn_curl_bk_17}
      boxExtents=<4 4 4>
      constraintSetup {angularLimitsMax=<0 0 20>, angularLimitsMin=<0 -20 -10>, angularTarget=<1 0 0>}
      localJointOffset=<0.1 0 0>
     }]
   }
   n25 = Anim Dynamics - Bone: dyn_curl_bk_18(ComponentPose=@n24, Alpha=1, ExternalForce="0,0,0") {
    angularDampingOverride=0.99
    angularSpringConstant=100
    bAngularSpring=true
    bOverrideAngularDamping=true
    bUseGravityOverride=true
    simulationSpace=RootRelative
    boundBone {boneName=dyn_curl_bk_18}
    componentAppliedLinearAccClamp=<3500 3500 3500>
    componentLinearAccScale=<1 1 1>
    componentLinearVelScale=<1 1 1>
    gravityOverride {z=-980}
    physicsBodyDefinitions [{
      boundBone {boneName=dyn_curl_bk_18}
      boxExtents=<4 4 4>
      constraintSetup {angularLimitsMax=<0 0 20>, angularLimitsMin=<0 -20 -10>, angularTarget=<1 0 0>}
      localJointOffset=<0.1 0 0>
     }]
   }
   n26 = Anim Dynamics - Bone: dyn_curl_bk_19(ComponentPose=@n25, Alpha=1, ExternalForce="0,0,0") {
    angularDampingOverride=0.99
    angularSpringConstant=100
    bAngularSpring=true
    bOverrideAngularDamping=true
    bUseGravityOverride=true
    simulationSpace=RootRelative
    boundBone {boneName=dyn_curl_bk_19}
    componentAppliedLinearAccClamp=<3500 3500 3500>
    componentLinearAccScale=<1 1 1>
    componentLinearVelScale=<1 1 1>
    gravityOverride {z=-980}
    physicsBodyDefinitions [{
      boundBone {boneName=dyn_curl_bk_19}
      boxExtents=<4 4 4>
      constraintSetup {angularLimitsMax=<0 0 20>, angularLimitsMin=<0 -20 -10>, angularTarget=<1 0 0>}
      localJointOffset=<0.1 0 0>
     }]
   }
   n27 = Anim Dynamics - Bone: dyn_curl_bk_20(ComponentPose=@n26, Alpha=1, ExternalForce="0,0,0") {
    angularDampingOverride=0.99
    angularSpringConstant=100
    bAngularSpring=true
    bOverrideAngularDamping=true
    bUseGravityOverride=true
    simulationSpace=RootRelative
    boundBone {boneName=dyn_curl_bk_20}
    componentAppliedLinearAccClamp=<3500 3500 3500>
    componentLinearAccScale=<1 1 1>
    componentLinearVelScale=<1 1 1>
    gravityOverride {z=-980}
    physicsBodyDefinitions [{
      boundBone {boneName=dyn_curl_bk_20}
      boxExtents=<4 4 4>
      constraintSetup {angularLimitsMax=<0 0 20>, angularLimitsMin=<0 -20 -10>, angularTarget=<1 0 0>}
      localJointOffset=<0.1 0 0>
     }]
   }
   n28 = Anim Dynamics - Bone: dyn_curl_pigtail_01(ComponentPose=@n27, Alpha=1, ExternalForce="0,0,0") {
    angularDampingOverride=0.99
    angularSpringConstant=100
    bAngularSpring=true
    bOverrideAngularDamping=true
    bUseGravityOverride=true
    simulationSpace=RootRelative
    boundBone {boneName=dyn_curl_pigtail_01}
    componentAppliedLinearAccClamp=<3500 3500 3500>
    componentLinearAccScale=<1 1 1>
    componentLinearVelScale=<1 1 1>
    gravityOverride {z=-980}
    physicsBodyDefinitions [{
      boundBone {boneName=dyn_curl_pigtail_01}
      boxExtents=<4 4 4>
      constraintSetup {angularLimitsMax=<0 0 20>, angularLimitsMin=<0 -20 -10>, angularTarget=<1 0 0>}
      localJointOffset=<0.1 0 0>
     }]
   }
   n29 = Anim Dynamics - Bone: dyn_curl_pigtail_02(ComponentPose=@n28, Alpha=1, ExternalForce="0,0,0") {
    angularDampingOverride=0.99
    angularSpringConstant=100
    bAngularSpring=true
    bOverrideAngularDamping=true
    bUseGravityOverride=true
    simulationSpace=RootRelative
    boundBone {boneName=dyn_curl_pigtail_02}
    componentAppliedLinearAccClamp=<3500 3500 3500>
    componentLinearAccScale=<1 1 1>
    componentLinearVelScale=<1 1 1>
    gravityOverride {z=-980}
    physicsBodyDefinitions [{
      boundBone {boneName=dyn_curl_pigtail_02}
      boxExtents=<4 4 4>
      constraintSetup {angularLimitsMax=<0 0 20>, angularLimitsMin=<0 -20 -10>, angularTarget=<1 0 0>}
      localJointOffset=<0.1 0 0>
     }]
   }
   n30 = Anim Dynamics - Bone: dyn_curl_pigtail_03(ComponentPose=@n29, Alpha=1, ExternalForce="0,0,0") {
    angularDampingOverride=0.99
    angularSpringConstant=100
    bAngularSpring=true
    bOverrideAngularDamping=true
    bUseGravityOverride=true
    simulationSpace=RootRelative
    boundBone {boneName=dyn_curl_pigtail_03}
    componentAppliedLinearAccClamp=<3500 3500 3500>
    componentLinearAccScale=<1 1 1>
    componentLinearVelScale=<1 1 1>
    gravityOverride {z=-980}
    physicsBodyDefinitions [{
      boundBone {boneName=dyn_curl_pigtail_03}
      boxExtents=<4 4 4>
      constraintSetup {angularLimitsMax=<0 0 20>, angularLimitsMin=<0 -20 -10>, angularTarget=<1 0 0>}
      localJointOffset=<0.1 0 0>
     }]
   }
   n31 = Anim Dynamics - Bone: dyn_curl_pigtail_04(ComponentPose=@n30, Alpha=1, ExternalForce="0,0,0") {
    angularDampingOverride=0.99
    angularSpringConstant=100
    bAngularSpring=true
    bOverrideAngularDamping=true
    bUseGravityOverride=true
    simulationSpace=RootRelative
    boundBone {boneName=dyn_curl_pigtail_04}
    componentAppliedLinearAccClamp=<3500 3500 3500>
    componentLinearAccScale=<1 1 1>
    componentLinearVelScale=<1 1 1>
    gravityOverride {z=-980}
    physicsBodyDefinitions [{
      boundBone {boneName=dyn_curl_pigtail_04}
      boxExtents=<4 4 4>
      constraintSetup {angularLimitsMax=<0 0 20>, angularLimitsMin=<0 -20 -10>, angularTarget=<1 0 0>}
      localJointOffset=<0.1 0 0>
     }]
   }
   Output Pose(Result=Component To Local(ComponentPose="Transform (Modify) Bone - Bone: head"(ComponentPose=@n31, Alpha=1, Translation="0,0,0", Rotation="0,0,0", Scale="1,1,1.1") {scaleMode=BMM_Additive, boneToModify {boneName=head}, scale {z=1.1}})) {layerGroup=None}
   n32 = Get Body() -> (Body:Skeletal Mesh Component Budgeted Object Reference)
   n33 = Control Rig!(
    Source=Layered blend per bone(
     BasePose=Copy Pose From Mesh() {bUseAttachedParent=true}
     BlendPoses_0=Copy Pose From Mesh(SourceMeshComponent=@n32) {bUseAttachedParent=true}
    ) {layerSetup [{branchFilters [{boneName=clavicle_l}, {boneName=clavicle_r}]}]}
    Alpha
    Crouching:bool=Get Crouching() -> (Crouching:bool)
    Is Backbling:bool=Get Is Backbling() -> (Is Backbling:bool)
    Is Skydive Floating:bool=Get bIsSkydiveFloating()
   ) -> (Pose) {controlRigAssetReference {blueprintRigClass=&/F_MED_WriteWalrus_FaceAcc_CtrlRig.F_MED_WriteWalrus_FaceAcc_CtrlRig_C}}
   Save cached pose 'toPhys'(Pose=@n33) {cacheName=toPhys}
   n34 = RigidBody!(ComponentPose=Local To Component(LocalPose=Use cached pose 'toPhys'()), Alpha, OverrideWorldGravity:Vector="0,0,0", ExternalForce) -> (Pose) {
    bClampLinearTranslationLimitToRefPose=true
    bOverrideWorldGravity=true
    overridePhysicsAsset=&/F_MED_WriteWalrus_FaceAcc_Physics
    simulationSpace=BaseBoneSpace
    baseBoneRef {boneName=spine_05}
    componentAppliedLinearAccClamp=<3000 3000 3000>
    simSpaceSettings {
     maxAngularAcceleration=32
     maxAngularVelocity=8
     maxLinearAcceleration=2400
     maxLinearVelocity=800
     velocityScaleZ=0.5
     worldAlpha=1
     externalLinearDragV=<0.25 0.25 0.25>
    }
   }
   n35 = RigidBody(ComponentPose=Local To Component(LocalPose=Use cached pose 'toPhys'()), OverrideWorldGravity="0,0,0") {
    bClampLinearTranslationLimitToRefPose=true
    bOverrideWorldGravity=true
    overridePhysicsAsset=&/F_MED_WriteWalrus_FaceAcc_Physics_IK
    simulationSpace=BaseBoneSpace
    baseBoneRef {boneName=spine_05}
    componentAppliedLinearAccClamp=<3000 3000 3000>
    simSpaceSettings {
     maxAngularAcceleration=32
     maxAngularVelocity=8
     maxLinearAcceleration=2400
     maxLinearVelocity=800
     velocityScaleZ=0.5
     worldAlpha=1
     externalLinearDragV=<0.25 0.25 0.25>
    }
   }
   Save cached pose 'toAnimDyn'(Pose=Blend Poses by bool(BlendPose_0=Component To Local(ComponentPose=@n35), BlendPose_1=Component To Local(ComponentPose=@n34), bActiveValue=Get Is Backbling())) {cacheName=toAnimDyn}
  ]}
 EventGraph {nodes [
   Event BlueprintUpdateAnimation()
   > n1 = Sequence()
   Set Is LOD1(execute=@n1.then_1, Is LOD1:bool="integer > integer"(A=Get CurrentLOD())) -> (Output_Get:bool)
   n2 = "Gravity Override State Function (Parameter Asset) (Nativized)"(
    execute=@n1.then_0
    CharacterPartAnimInstance=Self-Reference()
    GameTimeInSeconds=GetGameTimeInSeconds()
    Parameters=/BRCosmetics/Characters/Player/Female/Medium/Bodies/F_MED_WriteWalrus/Meshes/M_MED_WriteWalrus_Parameters
   )
   > Set Gravity Override(Gravity Override=@n2)
   > Set RBAN Settings Out(RBAN Settings Out=@n2.OutRigidBodyAnimNodeParameters)
   n3 = CheckForBackbling()
   > Set Is Backbling(Is Backbling:bool=OR Boolean!(A=@n3, B=@n3.bIsBackblingACape, C:bool=@n3.bIsBackblingLong) -> (ReturnValue)) -> (Output_Get:bool)
   Event OnEndCrouching()
   > Set Crouching(Crouching:bool=false) -> (Output_Get:bool)
   Event OnBeginCrouching()
   > Set Crouching(Crouching=true)
   n4 = Event OnCharacterCustomizationLoaded()
   > Set Body(Body:Skeletal Mesh Component Budgeted Object Reference=Get Array Item with EFort Custom Part Type(Array=@n4.CustomMeshParts)) -> (Output_Get:Skeletal Mesh Component Budgeted Object Reference)
  ]}
]
variables [
 Body {type=SkeletalMeshComponentBudgeted}
 Crouching {type=boolean}
 Gravity Override {type=Vector}
 Is Backbling {type=boolean}
 Is LOD1 {type=boolean}
 RBAN Settings Out {type=FortRigidBodyAnimNodeParameters}
]
